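# Humanoid AI behavior tree for the shoot-spot phase of cover-and-shoot combat, run on the CoverAndShoot blackboard.
# A sequence flags line of fire as acquired, then a selector tries the gameplay ability named by the shoot-spot tag key, if present and activatable.
# Otherwise it waits for a blackboard-driven shoot time with random deviation, using separate timing values when the AI was being aimed at.
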BehaviorTree /AICore/Humanoids/BehaviorTrees/BehaviorTrees/BT_CoverAndShoot_AtShootSpot
blackboard=/AICore/Humanoids/BehaviorTrees/Blackboards/BB_CoverAndShoot
root {
 name=Sequence
 class=BTComposite_Sequence
 children [
  Set Has Acquired LOF {
   class=BTTask_SetKeyValueBool
   children []
   decorators []
   properties {
    bIgnoreRestartSelf=false
    nodeName=Set Has Acquired LOF
    blackboardKey {bNoneIsAllowedValue=false, selectedKeyName=Cover_HasAcquiredLOF}
    value {defaultValue=true, key=None}
   }
   services []
  }
  Selector {
   class=BTComposite_Selector
   children [
    GameplayAbility {
     class=FortBTTask_ExecuteGameplayAbility
     decorators [
      Has Gameplay Ability {
       class=FortBTDecorator_GameplayAbility_HasGameplayAbility
       properties {gameplayAbilityTagContainer {key=Cover_ShootSpotAbilityTag, defaultValue {gameplayTags []}}}
      }
      Can Activate Gameplay Ability {
       class=FortBTDecorator_GameplayAbility_CanActivate
       properties {flowAbortMode=None, gameplayAbilityTagContainer {key=Cover_ShootSpotAbilityTag, defaultValue {gameplayTags []}}, targetValue {key=None}}
      }
     ]
     properties {nodeName=GameplayAbility, gameplayAbilityTagContainer {key=Cover_ShootSpotAbilityTag, defaultValue {gameplayTags []}}}
    }
    Wait {
     class=BTTask_Wait
     decorators [Was Being Aimed At {
       class=BTDecorator_Blackboard
       properties {
        arithmeticOperation=Equal
        flowAbortMode=None
        intValue=0
        nodeName=Was Being Aimed At
        stringValue=""
        blackboardKey {selectedKeyName=Cover_WasBeingAimedAt}
       }
      }]
     properties {randomDeviation {key=Cover_AimedAtShootTimeDeviation}, waitTime {defaultValue=5, key=Cover_AimedAtShootTime}}
    }
    Wait {class=BTTask_Wait, properties {randomDeviation {key=Cover_ShootTimeDeviation}, waitTime {defaultValue=5, key=Cover_ShootTime}}}
   ]
   decorators []
   properties {
    children [
     {
      childComposite=null
      childTask=&self:FortBTTask_ExecuteGameplayAbility_0
      decorators [
       &self:FortBTDecorator_GameplayAbility_HasGameplayAbility_0
       &self:FortBTDecorator_GameplayAbility_CanActivate_0
      ]
     }
     {
      childComposite=null
      childTask=&self:BTTask_Wait_0
      decorators [&self:BTDecorator_Blackboard_16]
     }
     {childComposite=null, childTask=&self:BTTask_Wait_1, decorators []}
    ]
    services []
   }
   services []
  }
 ]
 decorators []
 properties {children [
   {childTask=&self:BTTask_SetKeyValueBool_1}
   {childComposite=&self:BTComposite_Selector_19, childTask=null}
  ]}
}
