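# Behavior tree for humanoid AI that fights from cover by cycling between hiding and peeking out to shoot.
# It equips a weapon, moves to the defend spot, crouches and waits, then moves to a shoot spot unless line of fire exists.
# With line of fire it runs BT_CoverAndShoot_AtShootSpot, otherwise an EQS query repositions it.
# Blackboard flags let incoming aim delay peeking or interrupt shooting, and destination comparisons distinguish same from new cover.
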BehaviorTree /AICore/Humanoids/BehaviorTrees/BehaviorTrees/BT_CoverAndShoot
blackboard=/AICore/Humanoids/BehaviorTrees/Blackboards/BB_CoverAndShoot
root {
 name=Sequence
 class=BTComposite_Sequence
 children [Sequence {
   class=BTComposite_Sequence
   children [
    Equip Item {
     class=FortAthenaBTTask_EquipItem
     children []
     decorators [Force Success {class=BTDecorator_ForceSuccess}]
     properties {
      bIgnoreRestartSelf=false
      equipItemKeyName=AIEvaluator_Injection_Item
      nodeName=Equip Item
      weaponKeyName=AIEvaluator_Global_DesiredWeapon
      disableEquipAnimation {defaultValue=false, key=None}
      itemTagToEquip {key=None, defaultValue {gameplayTags []}}
      waitForEquipToFinish {defaultValue=false, key=None}
     }
     services []
    }
    FortAthenaBTTask_GetDefendSpotShootSpot_1 {
     class=FortAthenaBTTask_GetDefendSpotShootSpot
     children []
     decorators []
     properties {
      bIgnoreRestartSelf=false
      nodeName=""
      allowCrouchBlackboardKey {bNoneIsAllowedValue=false, selectedKeyName=Cover_ShootPositionAllowCrouch}
      defendSpotBlackboardKey {bNoneIsAllowedValue=false, selectedKeyName=AIEvaluator_Positioning_DefendSpot}
      shootSpotPositionBlackboardKey {bNoneIsAllowedValue=false, selectedKeyName=Cover_ShootPosition}
     }
     services []
    }
    FortAthenaBTTask_GetDefendSpotShouldCrouch_1 {
     class=FortAthenaBTTask_GetDefendSpotShouldCrouch
     children []
     decorators []
     properties {
      bIgnoreRestartSelf=false
      nodeName=""
      allowCrouchBlackboardKey {bNoneIsAllowedValue=false, selectedKeyName=Cover_DefendSpotShouldCrouch}
      defendSpotBlackboardKey {bNoneIsAllowedValue=false, selectedKeyName=AIEvaluator_Positioning_DefendSpot}
     }
     services []
    }
    Selector {
     class=BTComposite_Selector
     children [Sequence {
       class=BTComposite_Sequence
       children [
        FortBTTask_Bark_0 {
         class=FortBTTask_Bark
         children []
         decorators [Force Success {class=BTDecorator_ForceSuccess}]
         properties {
          bIgnoreRestartSelf=false
          nodeName=""
          barkTag {bCanEditDefaultValueType=false, key=Ranged_RepositioningBarkTag, defaultValue {_structType=&GameplayTag}}
         }
         services []
        }
        Reposition {
         class=FortAthenaBTTask_BotMoveTo
         decorators []
         properties {
          executionStatusKeyName=AIEvaluator_TokenPosition_ExecutionStatus
          movementResultKeyName=AIEvaluator_TokenPosition_MovementState
          nodeName=Reposition
          acceptableRadius {defaultValue=5}
          allowRandomWobble {defaultValue=true}
          bAllowPartialPath {defaultValue=false}
          bAllowStrafe {defaultValue=true}
          bReachTestIncludesAgentRadius {defaultValue=false}
          bReachTestIncludesGoalRadius {defaultValue=false}
          blackboardKey {selectedKeyName=AIEvaluator_Positioning_Destination}
          filterClass {key=Ranged_NavigationFilter}
          observedBlackboardValueTolerance {defaultValue=2.75}
         }
        }
        Set Previous Destination {
         class=BTTask_SetKeyValueVector
         children []
         decorators []
         properties {
          bIgnoreRestartSelf=false
          nodeName=Set Previous Destination
          blackboardKey {bNoneIsAllowedValue=false, selectedKeyName=Cover_PreviousDefendSpotPosition}
          value {key=AIEvaluator_Positioning_Destination, defaultValue=<0 0 0>}
         }
         services []
        }
        Sequence {
         class=BTComposite_Sequence
         children [
          Sequence {
           class=BTComposite_Sequence
           children [
            FortAthenaBTTask_GetDefendSpotShouldCrouch_0 {
             class=FortAthenaBTTask_GetDefendSpotShouldCrouch
             children []
             decorators []
             properties {
              bIgnoreRestartSelf=false
              nodeName=""
              allowCrouchBlackboardKey {bNoneIsAllowedValue=false, selectedKeyName=Cover_DefendSpotShouldCrouch}
              defendSpotBlackboardKey {bNoneIsAllowedValue=false, selectedKeyName=AIEvaluator_Positioning_DefendSpot}
             }
             services []
            }
            Selector {
             class=BTComposite_Selector
             children [
              Always crouch {
               class=FortBTTask_Crouch
               children []
               decorators [Blackboard Based Condition {
                 class=BTDecorator_Blackboard
                 properties {
                  arithmeticOperation=Equal
                  flowAbortMode=None
                  intValue=0
                  nodeName=Blackboard Based Condition
                  stringValue=""
                  blackboardKey {selectedKeyName=Cover_ForceCrouchAtCover}
                 }
                }]
               properties {bIgnoreRestartSelf=false, nodeName=Always crouch, shouldCrouch {defaultValue=true, key=None}}
               services []
              }
              Crouch if Crouch Cover {
               class=FortBTTask_Crouch
               children []
               decorators [Blackboard Based Condition {
                 class=BTDecorator_Blackboard
                 properties {
                  arithmeticOperation=Equal
                  basicOperation=NotSet
                  flowAbortMode=None
                  intValue=0
                  nodeName=Blackboard Based Condition
                  stringValue=""
                  blackboardKey {selectedKeyName=Cover_ForceCrouchAtCover}
                 }
                }]
               properties {bIgnoreRestartSelf=false, nodeName=Crouch if Crouch Cover, shouldCrouch {defaultValue=true, key=Cover_DefendSpotShouldCrouch}}
               services []
              }
             ]
             decorators []
             properties {
              children [
               {
                childComposite=null
                childTask=&self:FortBTTask_Crouch_1
                decorators [&self:BTDecorator_Blackboard_9]
               }
               {
                childComposite=null
                childTask=&self:FortBTTask_Crouch_0
                decorators [&self:BTDecorator_Blackboard_6]
               }
              ]
              services []
             }
             services []
            }
            Wait {
             class=BTTask_Wait
             properties {randomDeviation {defaultValue=0.5, key=Cover_HideTimeDeviation}, waitTime {defaultValue=3, key=Cover_HideTime}}
            }
           ]
           decorators []
           properties {
            children [
             {childTask=&self:FortAthenaBTTask_GetDefendSpotShouldCrouch_0}
             {childComposite=&self:BTComposite_Selector_6, childTask=null}
             {childTask=&self:BTTask_Wait_6}
            ]
            services [
             &self:FortAthenaBTService_SetInTacticalAction_11
             &self:FortAthenaBTService_ApplyGameplayTags_0
             &self:FortAthenaBTService_AIEvaluator_9
            ]
           }
           services [
            FortAthenaBTService_SetInTacticalAction_11 {
             class=FortAthenaBTService_SetInTacticalAction
             properties {
              shouldSetInTacticalAction {key=Cover_DefendSpotSetInTacticalAction}
              tacticalActionIdentifierTag {defaultValue {_structType=&GameplayTag}}
             }
            }
            Set Is At Cover Tag {
             class=FortAthenaBTService_ApplyGameplayTags
             properties {interval=0.5, nodeName=Set Is At Cover Tag, gameplayTagsToApply {defaultValue {gameplayTags [{tagName=AI.Is.Cover.Hiding}]}}}
            }
            Focus on target {
             class=FortAthenaBTService_AIEvaluator
             properties {interval=0.5, nodeName=Focus on target, aIEvaluatorClass {defaultValue=&FortAthenaAIBotEvaluator_HandleFocusing}}
            }
           ]
          }
          Selector {
           class=BTComposite_Selector
           children [
            Wait {
             class=BTTask_Wait
             decorators [
              Can Block Peek From Cover {
               class=BTDecorator_Blackboard
               properties {
                arithmeticOperation=Equal
                flowAbortMode=None
                intValue=0
                nodeName=Can Block Peek From Cover
                stringValue=""
                blackboardKey {selectedKeyName=Cover_AimedAt_CanBlockPeekFromCover}
               }
              }
              Is Being Aimed At {
               class=BTDecorator_Blackboard
               properties {
                arithmeticOperation=Equal
                flowAbortMode=None
                intValue=0
                nodeName=Is Being Aimed At
                stringValue=""
                blackboardKey {selectedKeyName=Cover_IsBeingAimedAt}
               }
              }
             ]
             properties {waitTime {defaultValue=1}}
            }
            Selector {
             class=BTComposite_Selector
             children [
              Always Succeed {
               class=FortBTTask_AlwaysSucceed
               children []
               decorators [Has Line Of Fire {
                 class=FortBTDecorator_HasLineOfFire
                 properties {
                  bInverseCondition=false
                  flowAbortMode=None
                  nodeName=Has Line Of Fire
                  targetActorKey {bNoneIsAllowedValue=false, selectedKeyName=AIEvaluator_Global_TargetActor}
                 }
                }]
               properties {bIgnoreRestartSelf=false, nodeName=Always Succeed}
               services []
              }
              Sequence {
               class=BTComposite_Sequence
               children [
                FortAthenaBTTask_GetDefendSpotShootSpot_0 {
                 class=FortAthenaBTTask_GetDefendSpotShootSpot
                 children []
                 decorators []
                 properties {
                  bIgnoreRestartSelf=false
                  nodeName=""
                  allowCrouchBlackboardKey {bNoneIsAllowedValue=false, selectedKeyName=Cover_ShootPositionAllowCrouch}
                  defendSpotBlackboardKey {bNoneIsAllowedValue=false, selectedKeyName=AIEvaluator_Positioning_DefendSpot}
                  shootSpotPositionBlackboardKey {bNoneIsAllowedValue=false, selectedKeyName=Cover_ShootPosition}
                 }
                 services []
                }
                Selector {
                 class=BTComposite_Selector
                 children [
                  Sequence {
                   class=BTComposite_Sequence
                   children [
                    Selector {
                     class=BTComposite_Selector
                     children [
                      Standing Shoot Position {
                       class=FortBTTask_Crouch
                       children []
                       decorators [Blackboard Based Condition {
                         class=BTDecorator_Blackboard
                         properties {
                          arithmeticOperation=Equal
                          flowAbortMode=None
                          intValue=0
                          nodeName=Blackboard Based Condition
                          stringValue=""
                          blackboardKey {selectedKeyName=Cover_AlwaysStandWhenPeeking}
                         }
                        }]
                       properties {bIgnoreRestartSelf=false, nodeName=Standing Shoot Position, shouldCrouch {defaultValue=false, key=None}}
                       services []
                      }
                      Crouch if shoot position is Crouch {
                       class=FortBTTask_Crouch
                       children []
                       decorators [Blackboard Based Condition {
                         class=BTDecorator_Blackboard
                         properties {
                          arithmeticOperation=Equal
                          basicOperation=NotSet
                          flowAbortMode=None
                          intValue=0
                          nodeName=Blackboard Based Condition
                          stringValue=""
                          blackboardKey {selectedKeyName=Cover_AlwaysStandWhenPeeking}
                         }
                        }]
                       properties {bIgnoreRestartSelf=false, nodeName=Crouch if shoot position is Crouch, shouldCrouch {defaultValue=false, key=Cover_ShootPositionAllowCrouch}}
                       services []
                      }
                     ]
                     decorators []
                     properties {
                      children [
                       {
                        childComposite=null
                        childTask=&self:FortBTTask_Crouch_3
                        decorators [&self:BTDecorator_Blackboard_2]
                       }
                       {
                        childComposite=null
                        childTask=&self:FortBTTask_Crouch_4
                        decorators [&self:BTDecorator_Blackboard_3]
                       }
                      ]
                      services []
                     }
                     services []
                    }
                    Bot Move To {
                     class=FortAthenaBTTask_BotMoveTo
                     decorators []
                     properties {
                      movementResultKeyName=None
                      nodeName=Bot Move To
                      acceptableRadius {defaultValue=5}
                      allowRandomWobble {defaultValue=true}
                      bAllowPartialPath {defaultValue=false}
                      bAllowStrafe {defaultValue=true}
                      bReachTestIncludesAgentRadius {defaultValue=false}
                      bStartFromPreviousPath {defaultValue=true}
                      blackboardKey {selectedKeyName=Cover_ShootPosition}
                      filterClass {
                       defaultValue=&NavFilter_Phoebe_NoSmashableNoWater_C
                       key=Ranged_NavigationFilter
                      }
                      observedBlackboardValueTolerance {defaultValue=4.75}
                     }
                    }
                   ]
                   decorators [Shoot spot from same cover {
                     class=BTDecorator_CompareBBEntries
                     properties {
                      bInverseCondition=false
                      flowAbortMode=Self
                      nodeName=Shoot spot from same cover
                      operator=Equal
                      blackboardKeyA {bNoneIsAllowedValue=false, selectedKeyName=AIEvaluator_Positioning_Destination}
                      blackboardKeyB {bNoneIsAllowedValue=false, selectedKeyName=Cover_PreviousDefendSpotPosition}
                     }
                    }]
                   properties {
                    children [
                     {childComposite=&self:BTComposite_Selector_7, childTask=null}
                     {childTask=&self:FortAthenaBTTask_BotMoveTo_2}
                    ]
                    services [&self:FortAthenaBTService_AIEvaluator_4]
                   }
                   services [Focus on target {
                     class=FortAthenaBTService_AIEvaluator
                     properties {
                      interval=0.5
                      nodeName=Focus on target
                      aIEvaluatorClass {defaultValue=&FortAthenaAIBotEvaluator_HandleFocusing, key=Ranged_RepositionFocusEvaluator}
                     }
                    }]
                  }
                  Sequence {
                   class=BTComposite_Sequence
                   children [
                    Stand Up {
                     class=FortBTTask_Crouch
                     children []
                     decorators []
                     properties {bIgnoreRestartSelf=false, nodeName=Stand Up, shouldCrouch {defaultValue=false, key=None}}
                     services []
                    }
                    Bot Move To {
                     class=FortAthenaBTTask_BotMoveTo
                     decorators []
                     properties {
                      movementResultKeyName=None
                      nodeName=Bot Move To
                      acceptableRadius {defaultValue=5}
                      allowRandomWobble {defaultValue=true}
                      bAllowPartialPath {defaultValue=false}
                      bAllowStrafe {defaultValue=true}
                      bReachTestIncludesAgentRadius {defaultValue=false}
                      bStartFromPreviousPath {defaultValue=true}
                      blackboardKey {selectedKeyName=Cover_ShootPosition}
                      filterClass {
                       defaultValue=&NavFilter_Phoebe_NoSmashableNoWater_C
                       key=Ranged_NavigationFilter
                      }
                      observedBlackboardValueTolerance {defaultValue=4.75}
                     }
                    }
                   ]
                   decorators [Shoot spot from different cover {
                     class=BTDecorator_CompareBBEntries
                     properties {
                      bInverseCondition=false
                      flowAbortMode=None
                      nodeName=Shoot spot from different cover
                      operator=NotEqual
                      blackboardKeyA {bNoneIsAllowedValue=false, selectedKeyName=AIEvaluator_Positioning_Destination}
                      blackboardKeyB {bNoneIsAllowedValue=false, selectedKeyName=Cover_PreviousDefendSpotPosition}
                     }
                    }]
                   properties {
                    children [
                     {childTask=&self:FortBTTask_Crouch_9}
                     {childTask=&self:FortAthenaBTTask_BotMoveTo_3}
                    ]
                    services [&self:FortAthenaBTService_AIEvaluator_3]
                   }
                   services [Focus on target {
                     class=FortAthenaBTService_AIEvaluator
                     properties {
                      interval=0.5
                      nodeName=Focus on target
                      aIEvaluatorClass {defaultValue=&FortAthenaAIBotEvaluator_HandleFocusing, key=Ranged_RepositionFocusEvaluator}
                     }
                    }]
                  }
                 ]
                 decorators []
                 properties {
                  children [
                   {
                    childComposite=&self:BTComposite_Sequence_10
                    childTask=null
                    decorators [&self:BTDecorator_CompareBBEntries_0]
                   }
                   {
                    childComposite=&self:BTComposite_Sequence_13
                    childTask=null
                    decorators [&self:BTDecorator_CompareBBEntries_5]
                   }
                  ]
                  services []
                 }
                 services []
                }
               ]
               decorators []
               properties {children [
                 {childTask=&self:FortAthenaBTTask_GetDefendSpotShootSpot_0}
                 {childComposite=&self:BTComposite_Selector_5, childTask=null}
                ]}
              }
             ]
             decorators []
             properties {
              children [
               {
                childComposite=null
                childTask=&self:FortBTTask_AlwaysSucceed_0
                decorators [&self:FortBTDecorator_HasLineOfFire_1]
               }
               {childComposite=&self:BTComposite_Sequence_3, childTask=null, decorators []}
              ]
              services []
             }
             services []
            }
           ]
           decorators []
           properties {
            children [
             {
              childComposite=null
              childTask=&self:BTTask_Wait_3
              decorators [
               &self:BTDecorator_Blackboard_12
               &self:BTDecorator_Blackboard_13
              ]
             }
             {childComposite=&self:BTComposite_Selector_3, childTask=null, decorators []}
            ]
            services [&self:FortAthenaBTService_SetInTacticalAction_12]
           }
           services [FortAthenaBTService_SetInTacticalAction_12 {
             class=FortAthenaBTService_SetInTacticalAction
             properties {shouldSetInTacticalAction {key=None}, tacticalActionIdentifierTag {defaultValue {_structType=&GameplayTag}}}
            }]
          }
          Sequence {
           class=BTComposite_Sequence
           children [
            Set was being aimed at {
             class=BTTask_SetKeyValueBool
             children []
             decorators []
             properties {
              bIgnoreRestartSelf=false
              nodeName=Set was being aimed at
              blackboardKey {bNoneIsAllowedValue=false, selectedKeyName=Cover_WasBeingAimedAt}
              value {defaultValue=false, key=Cover_IsBeingAimedAt}
             }
             services []
            }
            Set Has Acquired LOF {
             class=BTTask_SetKeyValueBool
             children []
             decorators []
             properties {
              bIgnoreRestartSelf=false
              nodeName=Set Has Acquired LOF
              blackboardKey {bNoneIsAllowedValue=false, selectedKeyName=Cover_HasAcquiredLOF}
              value {defaultValue=false, key=None}
             }
             services []
            }
            Selector {
             class=BTComposite_Selector
             children [
              Always Succeed {
               class=FortBTTask_AlwaysSucceed
               children []
               decorators [
                Can Interrupt Shooting {
                 class=BTDecorator_Blackboard
                 properties {
                  arithmeticOperation=Equal
                  flowAbortMode=None
                  intValue=0
                  nodeName=Can Interrupt Shooting
                  stringValue=""
                  blackboardKey {selectedKeyName=Cover_AimedAt_CanInterruptShooting}
                 }
                }
                Was Not Being Aimed At {
                 class=BTDecorator_Blackboard
                 properties {
                  arithmeticOperation=Equal
                  basicOperation=NotSet
                  flowAbortMode=None
                  intValue=0
                  nodeName=Was Not Being Aimed At
                  stringValue=""
                  blackboardKey {selectedKeyName=Cover_WasBeingAimedAt}
                 }
                }
                Is Being Aimed At {
                 class=BTDecorator_Blackboard
                 properties {arithmeticOperation=Equal, intValue=0, nodeName=Is Being Aimed At, stringValue="", blackboardKey {selectedKeyName=Cover_IsBeingAimedAt}}
                }
               ]
               properties {bIgnoreRestartSelf=false, nodeName=Always Succeed}
               services []
              }
              Selector {
               class=BTComposite_Selector
               children [
                Selector {
                 class=BTComposite_Selector
                 children [
                  Sequence {
                   class=BTComposite_Sequence
                   children [Run Behavior {
                     class=BTTask_RunBehavior
                     decorators []
                     properties {behaviorAsset=&/BT_CoverAndShoot_AtShootSpot, nodeName=Run Behavior}
                    }]
                   decorators [
                    Can Interrupt Shooting {
                     class=BTDecorator_Blackboard
                     properties {
                      arithmeticOperation=Equal
                      flowAbortMode=None
                      intValue=0
                      nodeName=Can Interrupt Shooting
                      stringValue=""
                      blackboardKey {selectedKeyName=Cover_AimedAt_CanInterruptShooting}
                     }
                    }
                    Has Line Of Fire {
                     class=FortBTDecorator_HasLineOfFire
                     properties {
                      bInverseCondition=false
                      flowAbortMode=Both
                      nodeName=Has Line Of Fire
                      targetActorKey {bNoneIsAllowedValue=false, selectedKeyName=AIEvaluator_Global_TargetActor}
                     }
                    }
                   ]
                   properties {children [{childTask=&self:BTTask_RunBehavior_1}]}
                  }
                  Sequence {
                   class=BTComposite_Sequence
                   children [Run Behavior {
                     class=BTTask_RunBehavior
                     decorators []
                     properties {behaviorAsset=&BT_CoverAndShoot_AtShootSpot, nodeName=Run Behavior}
                    }]
                   decorators [
                    Can't Interrupt Shooting {
                     class=BTDecorator_Blackboard
                     properties {
                      arithmeticOperation=Equal
                      basicOperation=NotSet
                      flowAbortMode=None
                      intValue=0
                      nodeName=Can't Interrupt Shooting
                      stringValue=""
                      blackboardKey {selectedKeyName=Cover_AimedAt_CanInterruptShooting}
                     }
                    }
                    Has Line Of Fire {
                     class=FortBTDecorator_HasLineOfFire
                     properties {
                      bInverseCondition=false
                      flowAbortMode=LowerPriority
                      nodeName=Has Line Of Fire
                      targetActorKey {bNoneIsAllowedValue=false, selectedKeyName=AIEvaluator_Global_TargetActor}
                     }
                    }
                   ]
                   properties {children [{childTask=&self:BTTask_RunBehavior_0}]}
                  }
                 ]
                 decorators []
                 properties {
                  children [
                   {
                    childComposite=&self:BTComposite_Sequence_20
                    childTask=null
                    decorators [
                     &self:BTDecorator_Blackboard_22
                     &self:FortBTDecorator_HasLineOfFire_8
                    ]
                   }
                   {
                    childComposite=&self:BTComposite_Sequence_21
                    childTask=null
                    decorators [
                     &self:BTDecorator_Blackboard_21
                     &self:FortBTDecorator_HasLineOfFire_7
                    ]
                   }
                  ]
                  services []
                 }
                 services []
                }
                Sequence {
                 class=BTComposite_Sequence
                 children [
                  Sequence {
                   class=BTComposite_Sequence
                   children [
                    Run EQS Query {
                     class=BTTask_RunEQSQuery
                     properties {
                      bUpdateBBOnFail=true
                      nodeName=Run EQS Query
                      blackboardKey {selectedKeyName=Cover_ShootReposition}
                      eQSRequest {bUseBBKeyForQueryTemplate=true, queryTemplate=null, runMode=SingleResult, eQSQueryBlackboardKey {selectedKeyName=Cover_ShootRepositionEQS}}
                     }
                    }
                    Stand Up {
                     class=FortBTTask_Crouch
                     children []
                     decorators []
                     properties {bIgnoreRestartSelf=false, nodeName=Stand Up, shouldCrouch {defaultValue=false, key=None}}
                     services []
                    }
                    Bot Move To {
                     class=FortAthenaBTTask_BotMoveTo
                     decorators []
                     properties {
                      movementResultKeyName=None
                      nodeName=Bot Move To
                      acceptableRadius {defaultValue=5}
                      allowRandomWobble {defaultValue=true}
                      bAllowPartialPath {defaultValue=false}
                      bAllowStrafe {defaultValue=true}
                      bReachTestIncludesAgentRadius {defaultValue=false}
                      bStartFromPreviousPath {defaultValue=true}
                      blackboardKey {selectedKeyName=Cover_ShootReposition}
                      filterClass {
                       defaultValue=&NavFilter_Phoebe_NoSmashableNoWater_C
                       key=Ranged_NavigationFilter
                      }
                      observedBlackboardValueTolerance {defaultValue=4.75}
                     }
                    }
                   ]
                   decorators [Force Success {class=BTDecorator_ForceSuccess}]
                   properties {children [
                     {childTask=&self:BTTask_RunEQSQuery_0}
                     {childTask=&self:FortBTTask_Crouch_5}
                     {childTask=&self:FortAthenaBTTask_BotMoveTo_5}
                    ]}
                  }
                  Sequence {
                   class=BTComposite_Sequence
                   children [
                    Wait {class=BTTask_Wait, properties {waitTime {defaultValue=0.5}}}
                    Selector {
                     class=BTComposite_Selector
                     children [
                      Always Fail {
                       class=FortBTTask_AlwaysFail
                       children []
                       decorators [Has Not Acquired LOF {
                         class=BTDecorator_Blackboard
                         properties {
                          arithmeticOperation=Equal
                          basicOperation=NotSet
                          flowAbortMode=None
                          intValue=0
                          nodeName=Has Not Acquired LOF
                          stringValue=""
                          blackboardKey {selectedKeyName=Cover_HasAcquiredLOF}
                         }
                        }]
                       properties {bIgnoreRestartSelf=false, nodeName=Always Fail}
                       services []
                      }
                      Always Succeed {
                       class=FortBTTask_AlwaysSucceed
                       children []
                       decorators [Has Acquired LOF {
                         class=BTDecorator_Blackboard
                         properties {
                          arithmeticOperation=Equal
                          flowAbortMode=None
                          intValue=0
                          nodeName=Has Acquired LOF
                          stringValue=""
                          blackboardKey {selectedKeyName=Cover_HasAcquiredLOF}
                         }
                        }]
                       properties {bIgnoreRestartSelf=false, nodeName=Always Succeed}
                       services []
                      }
                     ]
                     decorators []
                     properties {
                      children [
                       {
                        childComposite=null
                        childTask=&self:FortBTTask_AlwaysFail_0
                        decorators [&self:BTDecorator_Blackboard_20]
                       }
                       {
                        childComposite=null
                        childTask=&self:FortBTTask_AlwaysSucceed_3
                        decorators [&self:BTDecorator_Blackboard_19]
                       }
                      ]
                      services []
                     }
                     services []
                    }
                   ]
                   decorators []
                   properties {children [
                     {childTask=&self:BTTask_Wait_2}
                     {childComposite=&self:BTComposite_Selector_20, childTask=null}
                    ]}
                  }
                 ]
                 decorators []
                 properties {
                  children [
                   {
                    childComposite=&self:BTComposite_Sequence_7
                    childTask=null
                    decorators [&self:BTDecorator_ForceSuccess_2]
                   }
                   {childComposite=&self:BTComposite_Sequence_1, childTask=null}
                  ]
                  services [&self:FortAthenaBTService_SetInTacticalAction_14]
                 }
                 services [FortAthenaBTService_SetInTacticalAction_14 {
                   class=FortAthenaBTService_SetInTacticalAction
                   properties {shouldSetInTacticalAction {key=None}, tacticalActionIdentifierTag {defaultValue {_structType=&GameplayTag}}}
                  }]
                }
               ]
               decorators []
               properties {
                children [
                 {childComposite=&self:BTComposite_Selector_21, childTask=null, decorators []}
                 {childComposite=&self:BTComposite_Sequence_12, childTask=null, decorators []}
                ]
                services [&self:FortAthenaBTService_AIEvaluator_0]
               }
               services [Focus on target {
                 class=FortAthenaBTService_AIEvaluator
                 properties {interval=0.5, nodeName=Focus on target, aIEvaluatorClass {defaultValue=&FortAthenaAIBotEvaluator_HandleFocusing}}
                }]
              }
             ]
             decorators []
             properties {
              bApplyDecoratorScope=true
              children [
               {
                childComposite=null
                childTask=&self:FortBTTask_AlwaysSucceed_1
                decorators [
                 &self:BTDecorator_Blackboard_10
                 &self:BTDecorator_Blackboard_15
                 &self:BTDecorator_Blackboard_11
                ]
               }
               {childComposite=&self:BTComposite_Selector_17, childTask=null, decorators []}
              ]
              services []
             }
             services []
            }
           ]
           decorators []
           properties {children [
             {childTask=&self:BTTask_SetKeyValueBool_0}
             {childTask=&self:BTTask_SetKeyValueBool_2}
             {childComposite=&self:BTComposite_Selector_12, childTask=null}
            ]}
          }
          Sequence {
           class=BTComposite_Sequence
           children [
            Selector {
             class=BTComposite_Selector
             children [
              Sequence {
               class=BTComposite_Sequence
               children [
                Selector {
                 class=BTComposite_Selector
                 children [
                  Always crouch {
                   class=FortBTTask_Crouch
                   children []
                   decorators [Blackboard Based Condition {
                     class=BTDecorator_Blackboard
                     properties {
                      arithmeticOperation=Equal
                      flowAbortMode=None
                      intValue=0
                      nodeName=Blackboard Based Condition
                      stringValue=""
                      blackboardKey {selectedKeyName=Cover_ForceCrouchAtCover}
                     }
                    }]
                   properties {bIgnoreRestartSelf=false, nodeName=Always crouch, shouldCrouch {defaultValue=true, key=None}}
                   services []
                  }
                  Crouch if Crouch Cover {
                   class=FortBTTask_Crouch
                   children []
                   decorators [Blackboard Based Condition {
                     class=BTDecorator_Blackboard
                     properties {
                      arithmeticOperation=Equal
                      basicOperation=NotSet
                      flowAbortMode=None
                      intValue=0
                      nodeName=Blackboard Based Condition
                      stringValue=""
                      blackboardKey {selectedKeyName=Cover_ForceCrouchAtCover}
                     }
                    }]
                   properties {bIgnoreRestartSelf=false, nodeName=Crouch if Crouch Cover, shouldCrouch {defaultValue=true, key=Cover_DefendSpotShouldCrouch}}
                   services []
                  }
                 ]
                 decorators []
                 properties {
                  children [
                   {
                    childComposite=null
                    childTask=&self:FortBTTask_Crouch_7
                    decorators [&self:BTDecorator_Blackboard_5]
                   }
                   {
                    childComposite=null
                    childTask=&self:FortBTTask_Crouch_6
                    decorators [&self:BTDecorator_Blackboard_4]
                   }
                  ]
                  services []
                 }
                 services []
                }
                Bot Move To {
                 class=FortAthenaBTTask_BotMoveTo
                 decorators []
                 properties {
                  movementResultKeyName=None
                  nodeName=Bot Move To
                  acceptableRadius {defaultValue=5}
                  allowRandomWobble {defaultValue=true}
                  bAllowPartialPath {defaultValue=false}
                  bAllowStrafe {defaultValue=true}
                  bReachTestIncludesAgentRadius {defaultValue=false}
                  bStartFromPreviousPath {defaultValue=true}
                  blackboardKey {selectedKeyName=AIEvaluator_Positioning_Destination}
                  filterClass {
                   defaultValue=&NavFilter_Phoebe_NoSmashableNoWater_C
                   key=Ranged_NavigationFilter
                  }
                  observedBlackboardValueTolerance {defaultValue=4.75}
                 }
                }
               ]
               decorators [Hiding back into the same cover {
                 class=BTDecorator_CompareBBEntries
                 properties {
                  bInverseCondition=false
                  flowAbortMode=Self
                  nodeName=Hiding back into the same cover
                  operator=Equal
                  blackboardKeyA {bNoneIsAllowedValue=false, selectedKeyName=AIEvaluator_Positioning_Destination}
                  blackboardKeyB {bNoneIsAllowedValue=false, selectedKeyName=Cover_PreviousDefendSpotPosition}
                 }
                }]
               properties {
                children [
                 {childComposite=&self:BTComposite_Selector_8, childTask=null}
                 {childTask=&self:FortAthenaBTTask_BotMoveTo_4}
                ]
                services [&self:FortAthenaBTService_AIEvaluator_2]
               }
               services [Focus on target {
                 class=FortAthenaBTService_AIEvaluator
                 properties {interval=0.5, nodeName=Focus on target, aIEvaluatorClass {defaultValue=&FortAthenaAIBotEvaluator_HandleFocusing}}
                }]
              }
              Sequence {
               class=BTComposite_Sequence
               children [
                FortBTTask_Bark_1 {
                 class=FortBTTask_Bark
                 children []
                 decorators [Force Success {class=BTDecorator_ForceSuccess}]
                 properties {
                  bIgnoreRestartSelf=false
                  nodeName=""
                  barkTag {bCanEditDefaultValueType=false, key=Ranged_RepositioningBarkTag, defaultValue {_structType=&GameplayTag}}
                 }
                 services []
                }
                Stand Up {
                 class=FortBTTask_Crouch
                 children []
                 decorators []
                 properties {bIgnoreRestartSelf=false, nodeName=Stand Up, shouldCrouch {defaultValue=false, key=None}}
                 services []
                }
                Bot Move To {
                 class=FortAthenaBTTask_BotMoveTo
                 decorators []
                 properties {
                  movementResultKeyName=None
                  nodeName=Bot Move To
                  acceptableRadius {defaultValue=5}
                  allowRandomWobble {defaultValue=true}
                  bAllowPartialPath {defaultValue=false}
                  bAllowStrafe {defaultValue=true}
                  bReachTestIncludesAgentRadius {defaultValue=false}
                  bStartFromPreviousPath {defaultValue=true}
                  blackboardKey {selectedKeyName=AIEvaluator_Positioning_Destination}
                  filterClass {
                   defaultValue=&NavFilter_Phoebe_NoSmashableNoWater_C
                   key=Ranged_NavigationFilter
                  }
                  observedBlackboardValueTolerance {defaultValue=4.75}
                 }
                }
               ]
               decorators [Hiding back into another cover {
                 class=BTDecorator_CompareBBEntries
                 properties {
                  bInverseCondition=false
                  flowAbortMode=None
                  nodeName=Hiding back into another cover
                  operator=NotEqual
                  blackboardKeyA {bNoneIsAllowedValue=false, selectedKeyName=AIEvaluator_Positioning_Destination}
                  blackboardKeyB {bNoneIsAllowedValue=false, selectedKeyName=Cover_PreviousDefendSpotPosition}
                 }
                }]
               properties {
                children [
                 {
                  childTask=&self:FortBTTask_Bark_1
                  decorators [&self:BTDecorator_ForceSuccess_1]
                 }
                 {childTask=&self:FortBTTask_Crouch_2}
                 {childTask=&self:FortAthenaBTTask_BotMoveTo_1}
                ]
                services [&self:FortAthenaBTService_AIEvaluator_1]
               }
               services [Focus on target {
                 class=FortAthenaBTService_AIEvaluator
                 properties {
                  interval=0.5
                  nodeName=Focus on target
                  aIEvaluatorClass {defaultValue=&FortAthenaAIBotEvaluator_HandleFocusing, key=Ranged_RepositionFocusEvaluator}
                 }
                }]
              }
             ]
             decorators []
             properties {
              children [
               {
                childComposite=&self:BTComposite_Sequence_11
                childTask=null
                decorators [&self:BTDecorator_CompareBBEntries_2]
               }
               {
                childComposite=&self:BTComposite_Sequence_4
                childTask=null
                decorators [&self:BTDecorator_CompareBBEntries_3]
               }
              ]
              services []
             }
             services []
            }
            Set Previous Destination {
             class=BTTask_SetKeyValueVector
             children []
             decorators []
             properties {
              bIgnoreRestartSelf=false
              nodeName=Set Previous Destination
              blackboardKey {bNoneIsAllowedValue=false, selectedKeyName=Cover_PreviousDefendSpotPosition}
              value {key=AIEvaluator_Positioning_Destination, defaultValue=<0 0 0>}
             }
             services []
            }
           ]
           decorators []
           properties {children [
             {childComposite=&self:BTComposite_Selector_4, childTask=null}
             {childTask=&self:BTTask_SetKeyValueVector_1}
            ]}
          }
         ]
         decorators [Loop Until Failure {
           class=BTDecorator_LoopUntil
           properties {
            bInverseCondition=false
            flowAbortMode=None
            nodeName=Loop Until Failure
            requiredResult {defaultValue=1, enumType=/Script/AIModule.EBTNodeResult, key=None, nativeEnumTypeName=EBTNodeResult}
           }
          }]
         properties {
          children [
           {childComposite=&self:BTComposite_Sequence_5, childTask=null}
           {childComposite=&self:BTComposite_Selector_9, childTask=null}
           {childComposite=&self:BTComposite_Sequence_9, childTask=null}
           {childComposite=&self:BTComposite_Sequence_8, childTask=null}
          ]
          services [
           &self:FortAthenaBTService_ManageWeapon_2
           &self:FortBTService_Crouch_1
          ]
         }
         services [
          Executing...Weapon Actions {class=FortAthenaBTService_ManageWeapon, properties {interval=0.5, nodeName=Executing...Weapon Actions, randomDeviation=0.1}}
          FortBTService_Crouch_1 {
           class=FortBTService_Crouch
           properties {
            bCallTickOnSearchStart=false
            bRestartTimerOnEachActivation=false
            interval=0.5
            nodeName=""
            randomDeviation=0.1
            enterShouldCrouch {defaultValue=true, key=Cover_DefendSpotShouldCrouch}
            exitShouldCrouch {defaultValue=false, key=None}
           }
          }
         ]
        }
       ]
       decorators [Is Destination Valid {
         class=BTDecorator_Blackboard
         properties {
          arithmeticOperation=Equal
          flowAbortMode=None
          intValue=0
          nodeName=Is Destination Valid
          notifyObserver=ValueChange
          stringValue=""
          blackboardKey {selectedKeyName=AIEvaluator_Positioning_Destination}
         }
        }]
       properties {children [
         {
          childTask=&self:FortBTTask_Bark_0
          decorators [&self:BTDecorator_ForceSuccess_3]
         }
         {childTask=&self:FortAthenaBTTask_BotMoveTo_0}
         {childTask=&self:BTTask_SetKeyValueVector_0}
         {
          childComposite=&self:BTComposite_Sequence_6
          childTask=null
          decorators [&self:BTDecorator_LoopUntil_1]
         }
        ]}
      }]
     decorators []
     properties {
      children [{
        childComposite=&self:BTComposite_Sequence_14
        childTask=null
        decorators [&self:BTDecorator_Blackboard_8]
       }]
      services [
       &self:FortAthenaBTService_ApplySkillSet_0
       &self:FortAthenaBTService_AIEvaluator_5
       &self:FortAthenaBTService_AIEvaluator_6
       &self:FortAthenaBTService_MovementStyle_1
       &self:FortAthenaBTService_ActorAttentionTracking_0
      ]
     }
     services [
      Apply Skill Set {
       class=FortAthenaBTService_ApplySkillSet
       properties {
        bCallTickOnSearchStart=false
        bRestartTimerOnEachActivation=false
        interval=0.5
        nodeName=Apply Skill Set
        randomDeviation=0.1
        skillSetClass {baseClass=/Script/FortniteAI.FortAthenaAIBotSkillSet, defaultValue=null, key=Ranged_SkillsetOverride}
       }
      }
      Evaluating...Range Attack {
       class=FortAthenaBTService_AIEvaluator
       properties {interval=0.5, nodeName=Evaluating...Range Attack, aIEvaluatorClass {defaultValue=null, key=Ranged_RangeAttackEvaluator}}
      }
      Evaluating...Throwable {
       class=FortAthenaBTService_AIEvaluator
       properties {interval=0.5, nodeName=Evaluating...Throwable, aIEvaluatorClass {defaultValue=null, key=Ranged_ThrowableAttackEvaluator}}
      }
      Modulate Movement Style {
       class=FortAthenaBTService_MovementStyle
       properties {
        bCallTickOnSearchStart=false
        bRestartTimerOnEachActivation=false
        interval=0.5
        nodeName=Modulate Movement Style
        randomDeviation=0.1
        blackboardCondition {bNoneIsAllowedValue=true, selectedKeyName=Ranged_ModulateMovementStyle}
        invertBlackboardCondition {defaultValue=false, key=None}
        newMovementStyle {defaultValue=0, enumType=/Script/FortniteGame.EFortMovementStyle, key=Ranged_MovementStyle, nativeEnumTypeName=EFortMovementStyle}
        tagCondition {autoDescription="", tokenStreamVersion=0, userDescription="", queryTokenStream [], tagDictionary []}
       }
      }
      Actor Attention Tracking {
       class=FortAthenaBTService_ActorAttentionTracking
       properties {
        bCallTickOnSearchStart=false
        bRestartTimerOnEachActivation=false
        interval=0.25
        nodeName=Actor Attention Tracking
        randomDeviation=0.05
        actorToCheck {baseClass=&Actor, defaultValue=null, key=AIEvaluator_Global_TargetActor}
        attentionDurationThreshold {defaultValue=0.5, key=Cover_AimedAt_AimThresholdDurationSeconds}
        disableService {defaultValue=false, key=Cover_AimedAt_DisableService}
        isOwnerAimedAt {bNoneIsAllowedValue=false, selectedKeyName=Cover_IsBeingAimedAt}
        requireSightOfPerceivedHostileActor {defaultValue=true, key=Cover_AimedAt_RequireSightOfHostile}
        useAllPerceivedHostileActors {defaultValue=false, key=Cover_AimedAt_UseAllPerceivedHostileActors}
       }
      }
     ]
    }
   ]
   decorators [
    Is Location Reserved {
     class=BTDecorator_Blackboard
     properties {
      arithmeticOperation=Equal
      flowAbortMode=Self
      intValue=0
      nodeName=Is Location Reserved
      stringValue=""
      blackboardKey {selectedKeyName=AIEValuator_LocationSynchronizer_HasReservedLocation}
     }
    }
    Has Token {
     class=FortBTDecorator_HasToken
     properties {
      bInverseCondition=false
      flowAbortMode=Self
      nodeName=Has Token
      activationTagQuery {
       bCanEditDefaultValueType=false
       key=Ranged_TokenActivationTagQuery
       defaultValue {
        _structType=/Script/GameplayTags.GameplayTagQuery
        autoDescription=""
        tokenStreamVersion=0
        userDescription=""
        queryTokenStream []
        tagDictionary []
       }
      }
      blackboardKey {bNoneIsAllowedValue=false, selectedKeyName=AIEvaluator_Global_TargetActor}
      tokenQuery {
       bCanEditDefaultValueType=false
       key=Ranged_TokenTagQuery
       defaultValue {
        _structType=&GameplayTagQuery
        autoDescription=""
        tokenStreamVersion=0
        userDescription=""
        queryTokenStream []
        tagDictionary []
       }
      }
     }
    }
   ]
   properties {children [
     {
      childTask=&self:FortAthenaBTTask_EquipItem_0
      decorators [&self:BTDecorator_ForceSuccess_0]
     }
     {childTask=&self:FortAthenaBTTask_GetDefendSpotShootSpot_1}
     {childTask=&self:FortAthenaBTTask_GetDefendSpotShouldCrouch_1}
     {childComposite=&self:BTComposite_Selector_0, childTask=null}
    ]}
  }]
 decorators []
 properties {
  children [{
    childComposite=&self:BTComposite_Sequence_2
    childTask=null
    decorators [
     &self:BTDecorator_Blackboard_1
     &self:FortBTDecorator_HasToken_0
    ]
   }]
  services [
   &self:FortAthenaBTService_LocationSynchronizerReservation_0
   &self:FortAthenaBTService_TokenQuery_0
  ]
 }
 services [
  Executing...Location Synchronizer Reservation {
   class=FortAthenaBTService_LocationSynchronizerReservation
   properties {
    bCallTickOnSearchStart=false
    bRestartTimerOnEachActivation=false
    interval=0.5
    nodeName=Executing...Location Synchronizer Reservation
    randomDeviation=0.1
    hasLocationReservedKey {bNoneIsAllowedValue=true, selectedKeyName=AIEValuator_LocationSynchronizer_HasReservedLocation}
    locationHardReservationBoundsRatio {
     bCanEditDefaultValueType=false
     key=Ranged_HardReservationScalingRatio
     defaultValue {_structType=/Script/GameplayAbilities.ScalableFloat, registryType=None, value=1, curve {curveTable=null, rowName=None}}
    }
    locationSoftReservationBoundsRatio {
     bCanEditDefaultValueType=false
     key=Ranged_SoftReservationScalingRatio
     defaultValue {_structType=&ScalableFloat, registryType=None, value=1, curve {curveTable=null, rowName=None}}
    }
    locationToReserveKey {key=AIEvaluator_Positioning_Destination, defaultValue=<0 0 0>}
   }
  }
  Executing...Token Query {
   class=FortAthenaBTService_TokenQuery
   properties {
    bCallTickOnSearchStart=false
    bRestartTimerOnEachActivation=false
    interval=0.5
    nodeName=Executing...Token Query
    randomDeviation=0.1
    targetActorName=AIEvaluator_Global_TargetActor
    weaponKeyName=AIEvaluator_Global_DesiredWeapon
    activationTagQuery {
     bCanEditDefaultValueType=false
     key=Ranged_TokenActivationTagQuery
     defaultValue {
      _structType=&GameplayTagQuery
      autoDescription=""
      tokenStreamVersion=0
      userDescription=""
      queryTokenStream []
      tagDictionary []
     }
    }
    becomeRelevantTokenAction {
     defaultValue=1
     enumType=/Script/FortniteAIServer.EFortAthenaBTServiceTokenQueryAction
     key=None
     nativeEnumTypeName=EFortAthenaBTServiceTokenQueryAction
    }
    ceaseRelevantTokenAction {
     defaultValue=2
     enumType=&EFortAthenaBTServiceTokenQueryAction
     key=None
     nativeEnumTypeName=EFortAthenaBTServiceTokenQueryAction
    }
    defaultTokenSystemTagQuery {
     bCanEditDefaultValueType=false
     key=Ranged_TokenTagQuery
     defaultValue {
      _structType=&GameplayTagQuery
      autoDescription=""
      tokenStreamVersion=0
      userDescription=""
      queryTokenStream []
      tagDictionary []
     }
    }
    equippedItemTagAssociationDatas []
    targetChangeTokenAction {
     defaultValue=1
     enumType=&EFortAthenaBTServiceTokenQueryAction
     key=None
     nativeEnumTypeName=EFortAthenaBTServiceTokenQueryAction
    }
    tokenReleaseDelay {defaultValue=0, key=Ranged_TokenReleaseDelay}
    weaponChangeTokenAction {
     defaultValue=1
     enumType=&EFortAthenaBTServiceTokenQueryAction
     key=None
     nativeEnumTypeName=EFortAthenaBTServiceTokenQueryAction
    }
   }
  }
 ]
}
