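# Behavior tree for humanoid AI in the DBNO (down but not out) state, handling a bark and downed movement.
# A root sequence fires the bark under Force Success, then enters a looping selector.
# A frequently ticking DBNO guard evaluator service runs on that selector, whose Bot Move To heads for the blackboard destination while execution status is at least pending.
# The move self-aborts when that check fails, with a short wait as fallback.
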BehaviorTree /AICore/Humanoids/BehaviorTrees/BehaviorTrees/BT_DBNO
blackboard=/AICore/Humanoids/BehaviorTrees/Blackboards/BB_DBNO
root {
 name=Sequence
 class=BTComposite_Sequence
 children [
  FortBTTask_Bark_2 {
   class=FortBTTask_Bark
   children []
   decorators [Force Success {class=BTDecorator_ForceSuccess}]
   properties {
    bIgnoreRestartSelf=false
    nodeName=""
    barkTag {bCanEditDefaultValueType=false, key=DBNO_BarkTag, defaultValue {_structType=&GameplayTag}}
   }
   services []
  }
  Selector {
   class=BTComposite_Selector
   children [
    Bot Move To {
     class=FortAthenaBTTask_BotMoveTo
     decorators [Can Do DBNO Move {
       class=BTDecorator_Blackboard
       properties {
        arithmeticOperation=GreaterOrEqual
        flowAbortMode=Self
        intValue=3
        nodeName=Can Do DBNO Move
        stringValue=EExecutionStatus::ExecutionPending
        blackboardKey {selectedKeyName=AIEvaluator_DBNO_ExecutionStatus}
       }
      }]
     properties {
      movementResultKeyName=AIEvaluator_DBNO_MovementState
      nodeName=Bot Move To
      acceptableRadius {defaultValue=150}
      allowRandomWobble {defaultValue=true}
      bAllowStrafe {defaultValue=false}
      blackboardKey {selectedKeyName=AIEvaluator_DBNO_Destination}
      filterClass {defaultValue=&NavFilter_Phoebe_NoSmashable_C}
      observedBlackboardValueTolerance {defaultValue=4.75}
     }
    }
    Wait {class=BTTask_Wait, properties {waitTime {defaultValue=1}}}
   ]
   decorators [Loop {class=BTDecorator_Loop}]
   properties {
    children [
     {
      childComposite=null
      childTask=&self:FortAthenaBTTask_BotMoveTo_2
      decorators [&self:BTDecorator_Blackboard_10]
     }
     {childComposite=null, childTask=&self:BTTask_Wait_4, decorators []}
    ]
    services [&self:FortAthenaBTService_AIEvaluator_22]
   }
   services [Evaluating...DBNO Move {
     class=FortAthenaBTService_AIEvaluator
     properties {
      bCallTickOnSearchStart=true
      interval=0.1
      nodeName=Evaluating...DBNO Move
      randomDeviation=0
      aIEvaluatorClass {defaultValue=/Game/Athena/AI/NPCs/Base/Evaluators/Evaluator_DBNO_Guard.Evaluator_DBNO_Guard_C}
     }
    }]
  }
 ]
 decorators []
 properties {children [
   {
    childTask=&self:FortBTTask_Bark_2
    decorators [&self:BTDecorator_ForceSuccess_1]
   }
   {
    childComposite=&self:BTComposite_Selector_6
    childTask=null
    decorators [&self:BTDecorator_Loop_4]
   }
  ]}
}
