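# Behavior tree that sends a humanoid AI bot to a selected vehicle and has it enter.
# A Selector first runs a Sequence, gated on the injected SelectedVehicle blackboard key being valid, that moves within an acceptable radius of the vehicle and then runs the Fortnite enter-vehicle task.
# If that branch fails, it falls back to a short Wait.
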BehaviorTree /AICore/Humanoids/BehaviorTrees/BehaviorTrees/BT_EnterVehicle
blackboard=/AICore/Humanoids/BehaviorTrees/Blackboards/BB_EnterVehicle
root {
 name=Selector
 class=BTComposite_Selector
 children [
  Sequence {
   class=BTComposite_Sequence
   children [
    MoveTo Vehicle {
     class=FortAthenaBTTask_BotMoveTo
     decorators []
     properties {
      movementResultKeyName=None
      nodeName=MoveTo Vehicle
      acceptableRadius {defaultValue=250}
      allowRandomWobble {defaultValue=false}
      bAllowStrafe {defaultValue=false}
      blackboardKey {selectedKeyName=AIEvaluator_Injected_SelectedVehicle}
      observedBlackboardValueTolerance {defaultValue=3}
     }
    }
    FortAthenaBTTask_EnterVehicle_0 {
     class=FortAthenaBTTask_EnterVehicle
     children []
     decorators []
     properties {bIgnoreRestartSelf=false, nodeName="", selectedVehicleKeyName=AIEvaluator_Injected_SelectedVehicle}
     services []
    }
   ]
   decorators [Vehicle Object Valid {
     class=BTDecorator_Blackboard
     properties {
      arithmeticOperation=GreaterOrEqual
      intValue=4
      nodeName=Vehicle Object Valid
      stringValue=""
      blackboardKey {selectedKeyName=AIEvaluator_Injected_SelectedVehicle}
     }
    }]
   properties {children [
     {childTask=&self:FortAthenaBTTask_BotMoveTo_2}
     {childTask=&self:FortAthenaBTTask_EnterVehicle_0}
    ]}
  }
  Wait {class=BTTask_Wait, properties {waitTime {defaultValue=1}}}
 ]
 decorators []
 properties {
  children [
   {
    childComposite=&self:BTComposite_Sequence_12
    childTask=null
    decorators [&self:BTDecorator_Blackboard_29]
   }
   {childComposite=null, childTask=&self:BTTask_Wait_0, decorators []}
  ]
  services []
 }
 services []
}
