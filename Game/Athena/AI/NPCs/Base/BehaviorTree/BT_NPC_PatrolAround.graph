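# NPC patrol movement behavior tree on BB_Phoebe that moves a bot to its patrol destination and pauses there.
# A root selector tries the Phoebe unstuck subtree, then a movement selector, then a brief wait so the selector still returns success.
# A zipline evaluator service feeds the movement selector, which runs the zipline subtree if either move or usage status allows it.
# Otherwise it moves using an obstacle-avoiding nav filter, then waits a randomized time.
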BehaviorTree /Game/Athena/AI/NPCs/Base/BehaviorTree/BT_NPC_PatrolAround
blackboard=&BB_Phoebe
root {
 name=Patrol Move Move Selector
 class=BTComposite_Selector
 children [
  Run Unstuck Behavior {
   class=BTTask_RunBehavior
   decorators []
   properties {behaviorAsset=&BT_Phoebe_UnstuckNew, nodeName=Run Unstuck Behavior}
  }
  Selector {
   class=BTComposite_Selector
   children [
    Run Zipline Behavior {
     class=BTTask_RunBehavior
     decorators [
      Blackboard Based Condition {
       class=BTDecorator_Blackboard
       properties {
        arithmeticOperation=GreaterOrEqual
        intValue=4
        nodeName=Blackboard Based Condition
        stringValue=EExecutionStatus::ExecutionAllowed
        blackboardKey {selectedKeyName=AIEvaluator_Zipline_MoveExecutionStatus}
       }
      }
      Blackboard Based Condition {
       class=BTDecorator_Blackboard
       properties {
        arithmeticOperation=GreaterOrEqual
        intValue=4
        nodeName=Blackboard Based Condition
        stringValue=EExecutionStatus::ExecutionAllowed
        blackboardKey {selectedKeyName=AIEvaluator_Zipline_UsageExecutionStatus}
       }
      }
     ]
     properties {behaviorAsset=&BT_Phoebe_Zipline, nodeName=Run Zipline Behavior}
    }
    Sequence {
     class=BTComposite_Sequence
     children [
      Patrol Move To {
       class=FortAthenaBTTask_BotMoveTo
       decorators []
       properties {
        movementResultKeyName=AIEvaluator_Patrol_MovementState
        nodeName=Patrol Move To
        acceptableRadius {defaultValue=5}
        allowRandomWobble {defaultValue=false}
        bAllowStrafe {defaultValue=true}
        bTrackMovingGoal {defaultValue=false}
        blackboardKey {selectedKeyName=AIEvaluator_Patrol_Destination}
        filterClass {defaultValue=/Game/AI/NavFilters/NavFilter_AvoidObstacles.NavFilter_AvoidObstacles_C}
        observedBlackboardValueTolerance {defaultValue=200}
       }
      }
      Wait {class=BTTask_Wait, properties {randomDeviation {defaultValue=2.5}, waitTime {defaultValue=5}}}
     ]
     decorators []
     properties {children [
       {childTask=&self:FortAthenaBTTask_BotMoveTo_1}
       {childTask=&self:BTTask_Wait_1}
      ]}
    }
   ]
   decorators []
   properties {
    children [
     {
      childComposite=null
      childTask=&self:BTTask_RunBehavior_23
      decoratorOps [{number=2, operation=Or}, {number=0}, {number=1}]
      decorators [
       &self:BTDecorator_Blackboard_2
       &self:BTDecorator_Blackboard_4
      ]
     }
     {childComposite=&self:BTComposite_Sequence_0, childTask=null, decorators []}
    ]
    services [&self:FortAthenaBTService_AIEvaluator_72]
   }
   services [Evaluate...Zipline {
     class=FortAthenaBTService_AIEvaluator
     properties {
      bCallTickOnSearchStart=true
      interval=0.5
      nodeName=Evaluate...Zipline
      aIEvaluatorClass {defaultValue=/Game/Athena/AI/Phoebe/Evaluators/Phoebe_Evaluator_Zipline_PatrolAround.Phoebe_Evaluator_Zipline_PatrolAround_C}
     }
    }]
  }
  Wait so above selector return success {class=BTTask_Wait, properties {nodeName=Wait so above selector return success, waitTime {defaultValue=0.1}}}
 ]
 decorators []
 properties {
  nodeName=Patrol Move Move Selector
  children [
   {childComposite=null, childTask=&self:BTTask_RunBehavior_5, decorators []}
   {childComposite=&self:BTComposite_Selector_0, childTask=null, decorators []}
   {childComposite=null, childTask=&self:BTTask_Wait_0, decorators []}
  ]
  services []
 }
 services []
}
