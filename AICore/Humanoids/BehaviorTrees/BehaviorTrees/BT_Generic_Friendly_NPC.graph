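# Behavior tree for generic friendly humanoid NPCs, choosing among injected behaviors, conversation, smart objects and patrolling.
# A root selector prioritizes high injection, conversation, smart object, medium injection, dynamic blueprint, path patrol, patrol-around fallback, then waiting.
# Evaluator and injector services write blackboard keys that gate branches, alongside behavior-control switches on conversation and both patrols.
# Both injection points share the HighPrio evaluator key, while their action-object keys differ.
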BehaviorTree /AICore/Humanoids/BehaviorTrees/BehaviorTrees/BT_Generic_Friendly_NPC
blackboard=/AICore/Humanoids/BehaviorTrees/Blackboards/BB_Generic_Friendly_NPC
root {
 name=Selector
 class=BTComposite_Selector
 children [
  Injection Point Generic.FriendlyNPC.HighPrio {
   class=FortBTTask_InjectionPoint
   children []
   decorators [Can Do Injected Behavior {
     class=FortBTDecorator_CanDoInjectedBehavior
     properties {
      arithmeticOperation=GreaterOrEqual
      bInverseCondition=false
      basicOperation=Set
      floatValue=0
      flowAbortMode=Both
      intValue=3
      nodeName=Can Do Injected Behavior
      notifyObserver=ResultChange
      stringValue=ExecutionPending
      textOperation=Equal
      blackboardKey {bNoneIsAllowedValue=false, selectedKeyName=AIEvaluator_Injection_Generic_FriendlyNPC_HighPrio}
     }
    }]
   properties {
    bIgnoreRestartSelf=false
    nodeName=Injection Point Generic.FriendlyNPC.HighPrio
    injectedAction {baseClass=&Object, defaultValue=null, key=Injection_Generic_FriendlyNPC_HighPrio_ActionObject}
    injectionTag {tagName=Behavior.Injection.Generic.FriendlyNPC.HighPrio}
   }
   services []
  }
  Conversation {
   class=FortAthenaBTTask_Conversation
   children []
   decorators [
    Conversation? {
     class=BTDecorator_Blackboard
     properties {
      arithmeticOperation=GreaterOrEqual
      intValue=3
      nodeName=Conversation?
      stringValue=EExecutionStatus::ExecutionPending
      blackboardKey {selectedKeyName=AIEvaluator_Conversation_ExecutionStatus}
     }
    }
    "Behavior Controls = CanConverse" {
     class=FortAthenaBTDecorator_BehaviorControls
     properties {bInverseCondition=false, behaviorTreeBranch=CanConverse, flowAbortMode=Both, nodeName="Behavior Controls = CanConverse"}
    }
   ]
   properties {
    bIgnoreRestartSelf=false
    bResetEvaluatorStatusKeyOnFinish=true
    conversationStatusKeyName=AIEvaluator_Conversation_ExecutionStatus
    nodeName=Conversation
   }
   services []
  }
  UseSmartObject {
   class=BTTask_RunBehavior
   decorators []
   properties {behaviorAsset=/Game/Athena/AI/Phoebe/BehaviorTrees/BT_Phoebe_UseSmartObject, nodeName=UseSmartObject}
  }
  Injection Point Generic.FriendlyNPC.MedPrio {
   class=FortBTTask_InjectionPoint
   children []
   decorators [Can Do Injected Behavior {
     class=FortBTDecorator_CanDoInjectedBehavior
     properties {
      arithmeticOperation=GreaterOrEqual
      bInverseCondition=false
      basicOperation=Set
      floatValue=0
      flowAbortMode=Both
      intValue=3
      nodeName=Can Do Injected Behavior
      notifyObserver=ResultChange
      stringValue=ExecutionPending
      textOperation=Equal
      blackboardKey {bNoneIsAllowedValue=false, selectedKeyName=AIEvaluator_Injection_Generic_FriendlyNPC_HighPrio}
     }
    }]
   properties {
    bIgnoreRestartSelf=false
    nodeName=Injection Point Generic.FriendlyNPC.MedPrio
    injectedAction {baseClass=&Object, defaultValue=null, key=Injection_Generic_FriendlyNPC_MedPrio_ActionObject}
    injectionTag {tagName=Behavior.Injection.Generic.FriendlyNPC.MedPrio}
   }
   services []
  }
  Run DynamicBlueprint {
   class=FortAthenaBTTask_DynamicBlueprint
   children []
   decorators [Can Run LocalGameplayBehavior? {
     class=BTDecorator_Blackboard
     properties {
      arithmeticOperation=GreaterOrEqual
      intValue=3
      nodeName=Can Run LocalGameplayBehavior?
      stringValue=EExecutionStatus::ExecutionPending
      blackboardKey {selectedKeyName=AIEvaluator_DynamicBlueprint_ExecutionStatus}
     }
    }]
   properties {
    bIgnoreRestartSelf=false
    dynamicBlueprintActorKeyName=AIEvaluator_DynamicBlueprint_Actor
    dynamicBlueprintStatusKeyName=AIEvaluator_DynamicBlueprint_ExecutionStatus
    nodeName=Run DynamicBlueprint
   }
   services []
  }
  Can Patrolling? {
   class=BTComposite_Selector
   children [
    Selector {
     class=BTComposite_Selector
     children [
      Run Unstuck Behavior {
       class=BTTask_RunBehavior
       decorators []
       properties {behaviorAsset=&BT_Phoebe_UnstuckNew, nodeName=Run Unstuck Behavior}
      }
      Selector {
       class=BTComposite_Selector
       children [
        MoveTo PatrolPoint {
         class=FortAthenaBTTask_BotMoveTo
         decorators [Is Stressed? {
           class=BTDecorator_Blackboard
           properties {
            arithmeticOperation=GreaterOrEqual
            intValue=3
            nodeName=Is Stressed?
            stringValue=""
            blackboardKey {selectedKeyName="AIEvaluator_IsStressed "}
           }
          }]
         properties {
          movementResultKeyName=AIEvaluator_Patrolling_MovementState
          nodeName=MoveTo PatrolPoint
          acceptableRadius {defaultValue=50}
          allowRandomWobble {defaultValue=true}
          bAllowStrafe {defaultValue=true}
          bTrackMovingGoal {defaultValue=false}
          blackboardKey {selectedKeyName=AIEvaluator_Patrolling_Destination}
          observedBlackboardValueTolerance {defaultValue=50}
         }
        }
        MoveTo PatrolPoint {
         class=FortAthenaBTTask_BotMoveTo
         decorators []
         properties {
          movementResultKeyName=AIEvaluator_Patrolling_MovementState
          nodeName=MoveTo PatrolPoint
          acceptableRadius {defaultValue=50}
          allowRandomWobble {defaultValue=true}
          bAllowStrafe {defaultValue=true}
          bTrackMovingGoal {defaultValue=false}
          blackboardKey {selectedKeyName=AIEvaluator_Patrolling_Destination}
          observedBlackboardValueTolerance {defaultValue=50}
         }
        }
       ]
       decorators [Loop {class=BTDecorator_Loop}]
       properties {
        children [
         {
          childComposite=null
          childTask=&self:FortAthenaBTTask_BotMoveTo_10
          decorators [&self:BTDecorator_Blackboard_26]
         }
         {
          childComposite=null
          childTask=&self:FortAthenaBTTask_BotMoveTo_9
          decorators []
         }
        ]
        services [&self:FortAthenaBTService_Patrolling_1]
       }
       services [Executing...Patrolling {
         class=FortAthenaBTService_Patrolling
         properties {
          acceptableRadius=5
          bAllowPartialPath=true
          bCallTickOnSearchStart=false
          bProjectGoalLocation=true
          bRestartTimerOnEachActivation=false
          filterClass=null
          interval=0.5
          nodeName=Executing...Patrolling
          patrollingAppendDestinationKeyName=AIEvaluator_Patrolling_AppendDestination
          randomDeviation=0.1
         }
        }]
      }
     ]
     decorators [Should move to patrol point? {
       class=BTDecorator_Blackboard
       properties {
        arithmeticOperation=Equal
        intValue=0
        nodeName=Should move to patrol point?
        stringValue=""
        blackboardKey {selectedKeyName=AIEvaluator_Patrolling_ShouldMove}
       }
      }]
     properties {
      children [
       {childComposite=null, childTask=&self:BTTask_RunBehavior_2, decorators []}
       {
        childComposite=&self:BTComposite_Selector_20
        childTask=null
        decorators [&self:BTDecorator_Loop_0]
       }
      ]
      services []
     }
     services []
    }
    Wait On PatrolPoint {class=BTTask_Wait, properties {nodeName=Wait On PatrolPoint, waitTime {defaultValue=5}}}
   ]
   decorators [
    Can Patrolling? {
     class=BTDecorator_Blackboard
     properties {
      arithmeticOperation=GreaterOrEqual
      intValue=3
      nodeName=Can Patrolling?
      stringValue=EExecutionStatus::ExecutionPending
      blackboardKey {selectedKeyName=AIEvaluator_Patrolling_ExecutionStatus}
     }
    }
    "Behavior Controls = CanPatrolOnPath" {
     class=FortAthenaBTDecorator_BehaviorControls
     properties {bInverseCondition=false, behaviorTreeBranch=CanPatrolOnPath, flowAbortMode=Both, nodeName="Behavior Controls = CanPatrolOnPath"}
    }
   ]
   properties {
    nodeName=Can Patrolling?
    children [
     {
      childComposite=&self:BTComposite_Selector_13
      childTask=null
      decorators [&self:BTDecorator_Blackboard_30]
     }
     {childComposite=null, childTask=&self:BTTask_Wait_14, decorators []}
    ]
    services []
   }
   services []
  }
  Fallback Behavior {
   class=BTComposite_Selector
   children [
    Run PatrolAround NPC Behavior {
     class=BTTask_RunBehavior
     decorators [Patrol Around {
       class=BTDecorator_Blackboard
       properties {
        arithmeticOperation=GreaterOrEqual
        intValue=2
        nodeName=Patrol Around
        stringValue=EExecutionStatus::ExecutionSuccess
        blackboardKey {selectedKeyName=AIEvaluator_Patrol_ExecutionStatus}
       }
      }]
     properties {behaviorAsset=/Game/Athena/AI/NPCs/Base/BehaviorTree/BT_NPC_PatrolAround, nodeName=Run PatrolAround NPC Behavior}
    }
    Wait {class=BTTask_Wait, properties {waitTime {defaultValue=2}}}
   ]
   decorators [
    Can Patrol Around? {
     class=BTDecorator_Blackboard
     properties {
      arithmeticOperation=Equal
      intValue=0
      nodeName=Can Patrol Around?
      stringValue=""
      blackboardKey {selectedKeyName=AIEvaluator_CanUseFallbackPatrolAround}
     }
    }
    "Behavior Controls = CanPatrolAround" {
     class=FortAthenaBTDecorator_BehaviorControls
     properties {bInverseCondition=false, behaviorTreeBranch=CanPatrolAround, flowAbortMode=Both, nodeName="Behavior Controls = CanPatrolAround"}
    }
   ]
   properties {
    nodeName=Fallback Behavior
    children [
     {
      childComposite=null
      childTask=&self:BTTask_RunBehavior_3
      decorators [&self:BTDecorator_Blackboard_32]
     }
     {childComposite=null, childTask=&self:BTTask_Wait_15, decorators []}
    ]
    services [
     &self:FortAthenaBTService_AIEvaluator_8
     &self:FortAthenaBTService_AIEvaluator_9
    ]
   }
   services [
    Evaluating... Patrol Around {
     class=FortAthenaBTService_AIEvaluator
     properties {
      bCallTickOnSearchStart=true
      interval=2.5
      nodeName=Evaluating... Patrol Around
      randomDeviation=0.5
      aIEvaluatorClass {defaultValue=/Game/Athena/AI/NPCs/Base/Evaluators/Evaluator_PatrolAround_NPC.Evaluator_PatrolAround_NPC_C}
     }
    }
    Focus on target {
     class=FortAthenaBTService_AIEvaluator
     properties {interval=0.5, nodeName=Focus on target, aIEvaluatorClass {defaultValue=&FortAthenaAIBotEvaluator_HandleFocusing}}
    }
   ]
  }
  Wait {class=BTTask_Wait, properties {waitTime {defaultValue=2}}}
 ]
 decorators []
 properties {
  children [
   {
    childComposite=null
    childTask=&self:FortBTTask_InjectionPoint_0
    decorators [&self:FortBTDecorator_CanDoInjectedBehavior_1]
   }
   {
    childComposite=null
    childTask=&self:FortAthenaBTTask_Conversation_4
    decorators [
     &self:BTDecorator_Blackboard_59
     &self:FortAthenaBTDecorator_BehaviorControls_22
    ]
   }
   {childComposite=null, childTask=&self:BTTask_RunBehavior_14, decorators []}
   {
    childComposite=null
    childTask=&self:FortBTTask_InjectionPoint_1
    decorators [&self:FortBTDecorator_CanDoInjectedBehavior_0]
   }
   {
    childComposite=null
    childTask=&self:FortAthenaBTTask_DynamicBlueprint_0
    decorators [&self:BTDecorator_Blackboard_2]
   }
   {
    childComposite=&self:BTComposite_Selector_5
    childTask=null
    decorators [
     &self:BTDecorator_Blackboard_1
     &self:FortAthenaBTDecorator_BehaviorControls_1
    ]
   }
   {
    childComposite=&self:BTComposite_Selector_4
    childTask=null
    decorators [
     &self:BTDecorator_Blackboard_0
     &self:FortAthenaBTDecorator_BehaviorControls_0
    ]
   }
   {childComposite=null, childTask=&self:BTTask_Wait_0, decorators []}
  ]
  services [
   &self:FortAthenaBTService_UpdateTarget_1
   &self:FortAthenaBTService_AIEvaluator_6
   &self:FortAthenaBTService_AIEvaluator_5
   &self:FortAthenaBTService_AIEvaluator_7
   &self:FortBTService_AIEvaluatorsInjector_2
  ]
 }
 services [
  Bot Controller PeriodicUpdate {class=FortAthenaBTService_UpdateTarget, properties {interval=0.3, nodeName=Bot Controller PeriodicUpdate}}
  Evaluating...SmartObject {
   class=FortAthenaBTService_AIEvaluator
   properties {
    interval=2
    nodeName=Evaluating...SmartObject
    aIEvaluatorClass {defaultValue=/Script/FortniteAIServer.FortAthenaAIBotEvaluator_SmartObjects}
   }
  }
  Evaluating...Conversation {
   class=FortAthenaBTService_AIEvaluator
   properties {
    interval=1
    nodeName=Evaluating...Conversation
    randomDeviation=0.2
    aIEvaluatorClass {defaultValue=/Script/FortniteAIServer.FortAthenaAIBotEvaluator_Conversation}
   }
  }
  Evaluating...Patrolling {
   class=FortAthenaBTService_AIEvaluator
   properties {
    interval=0.5
    nodeName=Evaluating...Patrolling
    aIEvaluatorClass {defaultValue=/Game/Athena/AI/NPCs/Base/Evaluators/Evaluator_Patrolling_NPC.Evaluator_Patrolling_NPC_C}
   }
  }
  Behavior Injector {
   class=FortBTService_AIEvaluatorsInjector
   properties {
    bCallTickOnSearchStart=false
    bRestartTimerOnEachActivation=false
    interval=0.5
    nodeName=Behavior Injector
    randomDeviation=0.1
    injectionTagsKeys [
     {
      actionInterfaceObjectKeyName=Injection_Generic_FriendlyNPC_HighPrio_ActionObject
      injectionKeyName=AIEvaluator_Injection_Generic_FriendlyNPC_HighPrio
      injectionTag {tagName=Behavior.Injection.Generic.FriendlyNPC.HighPrio}
     }
     {
      actionInterfaceObjectKeyName=Injection_Generic_FriendlyNPC_MedPrio_ActionObject
      injectionKeyName=AIEvaluator_Injection_Generic_FriendlyNPC_HighPrio
      injectionTag {tagName=Behavior.Injection.Generic.FriendlyNPC.MedPrio}
     }
    ]
   }
  }
 ]
}
