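# Behavior tree for humanoid AI in vehicles, including a driver branch that chooses between combat driving and patrol.
# Tagged injection points let external behaviors take priority, while evaluator services feed blackboard keys that gate each looping branch.
# Threatened drivers pursue aggressively and last-known-position drivers approach attack destinations, both exiting as fallback.
# Otherwise drivers follow a group leader or patrol, modulating speed, waiting for passengers, escalating, honking or exiting.
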BehaviorTree /AICore/Humanoids/BehaviorTrees/BehaviorTrees/BT_InVehicle
blackboard=/AICore/Humanoids/BehaviorTrees/Blackboards/BB_InVehicle
root {
 name=Selector
 class=BTComposite_Selector
 children [
  Injection Point AICore.InVehicle.HighPriority {
   class=FortBTTask_InjectionPoint
   children []
   decorators [Can Do Injected Behavior {
     class=FortBTDecorator_CanDoInjectedBehavior
     properties {
      arithmeticOperation=GreaterOrEqual
      bInverseCondition=false
      basicOperation=Set
      floatValue=0
      flowAbortMode=Both
      intValue=3
      nodeName=Can Do Injected Behavior
      notifyObserver=ResultChange
      stringValue=ExecutionPending
      textOperation=Equal
      blackboardKey {bNoneIsAllowedValue=false, selectedKeyName=AIEvaluator_Injection_AICore_InVehicle_HighPriority}
     }
    }]
   properties {
    bIgnoreRestartSelf=false
    nodeName=Injection Point AICore.InVehicle.HighPriority
    injectedAction {baseClass=&Object, defaultValue=null, key=Injection_AICore_InVehicle_HighPriority_ActionObject}
    injectionTag {tagName=Behavior.Injection.AICore.InVehicle.HighPriority}
   }
   services []
  }
  Sequence {
   class=BTComposite_Sequence
   children [
    FortAthenaBTTask_ModulateVehicleSpeed_2 {
     class=FortAthenaBTTask_ModulateVehicleSpeed
     children []
     decorators []
     properties {bIgnoreRestartSelf=false, bResetControlPIds=true, nodeName="", drivingSpeed {defaultValue=-1, key=None}}
     services []
    }
    Selector {
     class=BTComposite_Selector
     children [
      Injection Point AICore.InVehicle.Driver.HighPriority {
       class=FortBTTask_InjectionPoint
       children []
       decorators [Can Do Injected Behavior {
         class=FortBTDecorator_CanDoInjectedBehavior
         properties {
          arithmeticOperation=GreaterOrEqual
          bInverseCondition=false
          basicOperation=Set
          floatValue=0
          flowAbortMode=Both
          intValue=3
          nodeName=Can Do Injected Behavior
          notifyObserver=ResultChange
          stringValue=ExecutionPending
          textOperation=Equal
          blackboardKey {bNoneIsAllowedValue=false, selectedKeyName=AIEvaluator_Injection_AICore_InVehicle_Driver_HighPriority}
         }
        }]
       properties {
        bIgnoreRestartSelf=false
        nodeName=Injection Point AICore.InVehicle.Driver.HighPriority
        injectedAction {baseClass=&Object, defaultValue=null, key=Injection_AICore_InVehicle_Driver_HighPriority_ActionObject}
        injectionTag {tagName=Behavior.Injection.AICore.InVehicle.Driver.HighPriority}
       }
       services []
      }
      Selector {
       class=BTComposite_Selector
       children [
        Sequence {
         class=BTComposite_Sequence
         children [
          FortAthenaBTTask_SetAggressiveDriving_0 {
           class=FortAthenaBTTask_SetAggressiveDriving
           children []
           decorators []
           properties {bAggressiveDriving=true, bIgnoreRestartSelf=false, nodeName=""}
           services []
          }
          "Speed = 30" {
           class=FortAthenaBTTask_ModulateVehicleSpeed
           children []
           decorators []
           properties {bIgnoreRestartSelf=false, bResetControlPIds=true, nodeName="Speed = 30", drivingSpeed {defaultValue=30, key=None}}
           services []
          }
          Loop Threaten {
           class=BTComposite_Selector
           children [
            LeaveVehicle {
             class=FortAthenaBTTask_LeaveVehicle
             children []
             decorators [Should leave vehicle {
               class=BTDecorator_Blackboard
               properties {
                arithmeticOperation=GreaterOrEqual
                intValue=4
                nodeName=Should leave vehicle
                stringValue=EExecutionStatus::ExecutionAllowed
                blackboardKey {selectedKeyName=AIEvaluator_LeaveSeatStatus}
               }
              }]
             properties {bIgnoreRestartSelf=false, bWaitVehicleStop=false, nodeName=LeaveVehicle}
             services []
            }
            Sequence {
             class=BTComposite_Sequence
             children [
              "Speed = 60" {
               class=FortAthenaBTTask_ModulateVehicleSpeed
               children []
               decorators []
               properties {bIgnoreRestartSelf=false, bResetControlPIds=true, nodeName="Speed = 60", drivingSpeed {defaultValue=60, key=None}}
               services []
              }
              Range Move To {
               class=FortAthenaBTTask_BotMoveTo
               decorators [Loop {class=BTDecorator_Loop}]
               properties {
                movementResultKeyName=AIEvaluator_RangeAttack_MovementState
                nodeName=Range Move To
                acceptableRadius {defaultValue=50}
                allowRandomWobble {defaultValue=false}
                bAllowStrafe {defaultValue=false}
                blackboardKey {selectedKeyName=AIEvaluator_Global_TargetActor}
                observedBlackboardValueTolerance {defaultValue=200}
               }
              }
             ]
             decorators [Target Actor Is Valid {
               class=BTDecorator_Blackboard
               properties {
                arithmeticOperation=Equal
                flowAbortMode=LowerPriority
                intValue=0
                nodeName=Target Actor Is Valid
                notifyObserver=ValueChange
                stringValue=""
                blackboardKey {selectedKeyName=AIEvaluator_Global_TargetActor}
               }
              }]
             properties {children [
               {childTask=&self:FortAthenaBTTask_ModulateVehicleSpeed_0}
               {
                childTask=&self:FortAthenaBTTask_BotMoveTo_1
                decorators [&self:BTDecorator_Loop_4]
               }
              ]}
            }
            Sequence {
             class=BTComposite_Sequence
             children [
              Simple Parallel {
               class=BTComposite_SimpleParallel
               children [
                Range Move To {
                 class=FortAthenaBTTask_BotMoveTo
                 decorators []
                 properties {
                  movementResultKeyName=AIEvaluator_RangeAttack_MovementState
                  nodeName=Range Move To
                  acceptableRadius {defaultValue=50}
                  allowRandomWobble {defaultValue=false}
                  bAllowStrafe {defaultValue=false}
                  blackboardKey {selectedKeyName=AIEvaluator_Attack_Destination}
                  observedBlackboardValueTolerance {defaultValue=200}
                 }
                }
                Sequence {
                 class=BTComposite_Sequence
                 children [
                  FortAthenaBTTask_PauseVehicle_0 {class=FortAthenaBTTask_PauseVehicle, children [], decorators [], properties {bIgnoreRestartSelf=false, nodeName=""}, services []}
                  LeaveVehicle {
                   class=FortAthenaBTTask_LeaveVehicle
                   children []
                   decorators []
                   properties {bIgnoreRestartSelf=false, bWaitVehicleStop=false, nodeName=LeaveVehicle}
                   services []
                  }
                 ]
                 decorators [Distance Between {
                   class=FortBTDecorator_DistanceBetween
                   properties {
                    flowAbortMode=None
                    nodeName=Distance Between
                    operator=LessOrEqual
                    blackboardKeyB {selectedKeyName=AIEvaluator_Global_LastKnownPosition}
                    specifiedDistance {defaultValue=2000}
                   }
                  }]
                 properties {children [
                   {childTask=&self:FortAthenaBTTask_PauseVehicle_0}
                   {childTask=&self:FortAthenaBTTask_LeaveVehicle_1}
                  ]}
                }
               ]
               properties {
                children [
                 {childTask=&self:FortAthenaBTTask_BotMoveTo_0}
                 {
                  childComposite=&self:BTComposite_Sequence_1
                  childTask=null
                  decorators [&self:FortBTDecorator_DistanceBetween_0]
                 }
                ]
                services []
               }
               services []
              }
              LeaveVehicle {
               class=FortAthenaBTTask_LeaveVehicle
               children []
               decorators [Distance Between {
                 class=FortBTDecorator_DistanceBetween
                 properties {
                  flowAbortMode=None
                  nodeName=Distance Between
                  operator=LessOrEqual
                  blackboardKeyB {selectedKeyName=AIEvaluator_Global_LastKnownPosition}
                  specifiedDistance {defaultValue=2000}
                 }
                }]
               properties {bIgnoreRestartSelf=false, bWaitVehicleStop=false, nodeName=LeaveVehicle}
               services []
              }
             ]
             decorators [
              Blackboard Based Condition {
               class=BTDecorator_Blackboard
               properties {
                arithmeticOperation=GreaterOrEqual
                flowAbortMode=None
                intValue=3
                nodeName=Blackboard Based Condition
                stringValue=EExecutionStatus::ExecutionPending
                blackboardKey {selectedKeyName=AIEvaluator_RangeAttack_ExecutionStatus}
               }
              }
              Blackboard Based Condition {
               class=BTDecorator_Blackboard
               properties {
                arithmeticOperation=GreaterOrEqual
                flowAbortMode=None
                intValue=3
                nodeName=Blackboard Based Condition
                stringValue=EExecutionStatus::ExecutionPending
                blackboardKey {selectedKeyName=AIEvaluator_ThrowableAttack_ExecutionStatus}
               }
              }
              Should move toward Target {
               class=BTDecorator_Blackboard
               properties {
                arithmeticOperation=Equal
                flowAbortMode=LowerPriority
                intValue=0
                nodeName=Should move toward Target
                notifyObserver=ValueChange
                stringValue=""
                blackboardKey {selectedKeyName=AIEvaluator_Attack_MoveToDestination}
               }
              }
              Attack Destination Is Valid {
               class=BTDecorator_Blackboard
               properties {
                arithmeticOperation=Equal
                flowAbortMode=LowerPriority
                intValue=0
                nodeName=Attack Destination Is Valid
                notifyObserver=ValueChange
                stringValue=""
                blackboardKey {selectedKeyName=AIEvaluator_Attack_Destination}
               }
              }
             ]
             properties {children [
               {childComposite=&self:BTComposite_SimpleParallel_0, childTask=null}
               {
                childTask=&self:FortAthenaBTTask_LeaveVehicle_2
                decorators [&self:FortBTDecorator_DistanceBetween_1]
               }
              ]}
            }
            Sequence {
             class=BTComposite_Sequence
             children [
              Wait {class=BTTask_Wait, properties {waitTime {defaultValue=2}}}
              LeaveVehicle {
               class=FortAthenaBTTask_LeaveVehicle
               children []
               decorators []
               properties {bIgnoreRestartSelf=false, bWaitVehicleStop=false, nodeName=LeaveVehicle}
               services []
              }
             ]
             decorators []
             properties {children [
               {childTask=&self:BTTask_Wait_5}
               {childTask=&self:FortAthenaBTTask_LeaveVehicle_3}
              ]}
            }
           ]
           decorators [Loop {class=BTDecorator_Loop}]
           properties {
            nodeName=Loop Threaten
            children [
             {
              childComposite=null
              childTask=&self:FortAthenaBTTask_LeaveVehicle_0
              decorators [&self:BTDecorator_Blackboard_4]
             }
             {
              childComposite=&self:BTComposite_Sequence_10
              childTask=null
              decorators [&self:BTDecorator_Blackboard_26]
             }
             {
              childComposite=&self:BTComposite_Sequence_2
              childTask=null
              decoratorOps [{number=3, operation=And}, {number=2, operation=Or}, {number=0}, {number=1}, {number=2}, {number=3}]
              decorators [
               &self:BTDecorator_Blackboard_71
               &self:BTDecorator_Blackboard_70
               &self:BTDecorator_Blackboard_21
               &self:BTDecorator_Blackboard_22
              ]
             }
             {childComposite=&self:BTComposite_Sequence_4, childTask=null, decorators []}
            ]
            services [
             &self:FortAthenaBTService_AIEvaluator_1
             &self:FortAthenaBTService_AIEvaluator_2
            ]
           }
           services [
            Focus on target {
             class=FortAthenaBTService_AIEvaluator
             properties {interval=0.5, nodeName=Focus on target, aIEvaluatorClass {defaultValue=&FortAthenaAIBotEvaluator_HandleFocusing}}
            }
            Evaluating...Leave Vehicle {
             class=FortAthenaBTService_AIEvaluator
             properties {
              interval=0.5
              nodeName=Evaluating...Leave Vehicle
              aIEvaluatorClass {defaultValue=/Game/Athena/AI/Vehicle/Evaluators/VehicleLeaveSeatEvaluator.VehicleLeaveSeatEvaluator_C}
             }
            }
           ]
          }
         ]
         decorators [IsThreatened {
           class=BTDecorator_Blackboard
           properties {
            arithmeticOperation=Equal
            intValue=3
            nodeName=IsThreatened
            stringValue=EAlertLevel::Threatened
            blackboardKey {selectedKeyName=AIEvaluator_Global_AlertLevel}
           }
          }]
         properties {children [
           {childTask=&self:FortAthenaBTTask_SetAggressiveDriving_0}
           {childTask=&self:FortAthenaBTTask_ModulateVehicleSpeed_1}
           {
            childComposite=&self:BTComposite_Selector_4
            childTask=null
            decorators [&self:BTDecorator_Loop_0]
           }
          ]}
        }
        Sequence {
         class=BTComposite_Sequence
         children [
          FortAthenaBTTask_ActivateVehicleBoost_0 {
           class=FortAthenaBTTask_ActivateVehicleBoost
           children []
           decorators []
           properties {bActivateBoost=false, bIgnoreMinimumDistanceLeft=false, bIgnoreRestartSelf=false, boostLength=-1, nodeName=""}
           services []
          }
          "Speed = 30" {
           class=FortAthenaBTTask_ModulateVehicleSpeed
           children []
           decorators []
           properties {bIgnoreRestartSelf=false, bResetControlPIds=true, nodeName="Speed = 30", drivingSpeed {defaultValue=30, key=None}}
           services []
          }
          Selector {
           class=BTComposite_Selector
           children [
            Range Move To {
             class=FortAthenaBTTask_BotMoveTo
             decorators [
              Should move toward Target {
               class=BTDecorator_Blackboard
               properties {
                arithmeticOperation=Equal
                flowAbortMode=LowerPriority
                intValue=0
                nodeName=Should move toward Target
                stringValue=""
                blackboardKey {selectedKeyName=AIEvaluator_Attack_MoveToDestination}
               }
              }
              Attack Destination Is Valid {
               class=BTDecorator_Blackboard
               properties {
                arithmeticOperation=Equal
                flowAbortMode=LowerPriority
                intValue=0
                nodeName=Attack Destination Is Valid
                stringValue=""
                blackboardKey {selectedKeyName=AIEvaluator_Attack_Destination}
               }
              }
             ]
             properties {
              movementResultKeyName=AIEvaluator_RangeAttack_MovementState
              nodeName=Range Move To
              acceptableRadius {defaultValue=50}
              allowRandomWobble {defaultValue=false}
              bAllowStrafe {defaultValue=false}
              blackboardKey {selectedKeyName=AIEvaluator_Attack_Destination}
              observedBlackboardValueTolerance {defaultValue=200}
             }
            }
            Sequence {
             class=BTComposite_Sequence
             children [
              Wait {class=BTTask_Wait, properties {randomDeviation {defaultValue=1}, waitTime {defaultValue=5}}}
              LeaveVehicle {
               class=FortAthenaBTTask_LeaveVehicle
               children []
               decorators []
               properties {bIgnoreRestartSelf=false, bWaitVehicleStop=false, nodeName=LeaveVehicle}
               services []
              }
             ]
             decorators []
             properties {children [
               {childTask=&self:BTTask_Wait_2}
               {childTask=&self:FortAthenaBTTask_LeaveVehicle_13}
              ]}
            }
           ]
           decorators [Loop {class=BTDecorator_Loop}]
           properties {
            children [
             {
              childComposite=null
              childTask=&self:FortAthenaBTTask_BotMoveTo_4
              decorators [
               &self:BTDecorator_Blackboard_15
               &self:BTDecorator_Blackboard_16
              ]
             }
             {childComposite=&self:BTComposite_Sequence_14, childTask=null, decorators []}
            ]
            services [&self:FortAthenaBTService_AIEvaluator_27]
           }
           services [Focus on target {
             class=FortAthenaBTService_AIEvaluator
             properties {interval=0.5, nodeName=Focus on target, aIEvaluatorClass {defaultValue=&FortAthenaAIBotEvaluator_HandleFocusing}}
            }]
          }
         ]
         decorators [IsLKP {
           class=BTDecorator_Blackboard
           properties {
            arithmeticOperation=Equal
            intValue=2
            nodeName=IsLKP
            stringValue=EAlertLevel::LKP
            blackboardKey {selectedKeyName=AIEvaluator_Global_AlertLevel}
           }
          }]
         properties {children [
           {childTask=&self:FortAthenaBTTask_ActivateVehicleBoost_0}
           {childTask=&self:FortAthenaBTTask_ModulateVehicleSpeed_3}
           {
            childComposite=&self:BTComposite_Selector_15
            childTask=null
            decorators [&self:BTDecorator_Loop_11]
           }
          ]}
        }
       ]
       decorators []
       properties {
        children [
         {
          childComposite=&self:BTComposite_Sequence_5
          childTask=null
          decorators [&self:BTDecorator_Blackboard_19]
         }
         {
          childComposite=&self:BTComposite_Sequence_0
          childTask=null
          decorators [&self:BTDecorator_Blackboard_35]
         }
        ]
        services [&self:FortAthenaBTService_AIEvaluator_29]
       }
       services [Evaluating...Range Attack {
         class=FortAthenaBTService_AIEvaluator
         properties {
          bCallTickOnSearchStart=true
          interval=0.2
          nodeName=Evaluating...Range Attack
          randomDeviation=0.05
          aIEvaluatorClass {defaultValue=/Game/Athena/AI/Vehicle/Evaluators/VehicleRangeAttackEvaluator.VehicleRangeAttackEvaluator_C}
         }
        }]
      }
      Sequence {
       class=BTComposite_Sequence
       children [
        FortAthenaBTTask_SetAggressiveDriving_3 {
         class=FortAthenaBTTask_SetAggressiveDriving
         children []
         decorators []
         properties {bAggressiveDriving=false, bIgnoreRestartSelf=false, nodeName=""}
         services []
        }
        "Speed = 21" {
         class=FortAthenaBTTask_ModulateVehicleSpeed
         children []
         decorators []
         properties {bIgnoreRestartSelf=false, bResetControlPIds=true, nodeName="Speed = 21", drivingSpeed {defaultValue=21, key=VehicleDriver_PatrolTargetSpeed}}
         services []
        }
        Selector {
         class=BTComposite_Selector
         children [
          Selector {
           class=BTComposite_Selector
           children [
            Selector {
             class=BTComposite_Selector
             children [Simple Parallel {
               class=BTComposite_SimpleParallel
               children [
                Move To Group Leader {
                 class=FortAthenaBTTask_BotMoveTo
                 decorators []
                 properties {
                  movementResultKeyName=AIEvaluator_FollowGroupLeader_MovementState
                  nodeName=Move To Group Leader
                  acceptableRadius {defaultValue=10}
                  allowRandomWobble {defaultValue=true}
                  bAllowStrafe {defaultValue=false}
                  bReachTestIncludesAgentRadius {defaultValue=false}
                  bReachTestIncludesGoalRadius {defaultValue=false}
                  blackboardKey {selectedKeyName=AIEvaluator_FollowGroupLeader_Destination}
                  filterClass {key=InVehicle_Patrol_MoveTo_FilterClass}
                  observedBlackboardValueTolerance {defaultValue=4.75}
                 }
                }
                FortAthenaBTTask_Escalate_0 {
                 class=FortAthenaBTTask_Escalate
                 children []
                 decorators [Escalate? {
                   class=BTDecorator_Blackboard
                   properties {
                    arithmeticOperation=GreaterOrEqual
                    flowAbortMode=None
                    intValue=3
                    nodeName=Escalate?
                    stringValue=ExecutionPending
                    blackboardKey {selectedKeyName=AIEvaluator_Escalate_ExecutionStatus}
                   }
                  }]
                 properties {bIgnoreRestartSelf=false, nodeName=""}
                 services []
                }
               ]
               properties {
                children [
                 {childTask=&self:FortAthenaBTTask_BotMoveTo_2}
                 {
                  childTask=&self:FortAthenaBTTask_Escalate_0
                  decorators [&self:BTDecorator_Blackboard_5]
                 }
                ]
                services []
               }
               services []
              }]
             decorators [Following Leader {
               class=BTDecorator_Blackboard
               properties {
                arithmeticOperation=GreaterOrEqual
                intValue=3
                nodeName=Following Leader
                stringValue=EExecutionStatus::ExecutionPending
                blackboardKey {selectedKeyName=AIEvaluator_FollowGroupLeader_ExecutionStatus}
               }
              }]
             properties {
              children [{childComposite=&self:BTComposite_SimpleParallel_2, childTask=null, decorators []}]
              services [&self:FortAthenaBTService_AIEvaluator_59]
             }
             services [Focus on target {
               class=FortAthenaBTService_AIEvaluator
               properties {interval=0.5, nodeName=Focus on target, aIEvaluatorClass {defaultValue=&FortAthenaAIBotEvaluator_HandleFocusing}}
              }]
            }
            Sequence {
             class=BTComposite_Sequence
             children [
              Wait {class=BTTask_Wait, properties {waitTime {defaultValue=5}}}
              LeaveVehicle {
               class=FortAthenaBTTask_LeaveVehicle
               children []
               decorators []
               properties {bIgnoreRestartSelf=false, bWaitVehicleStop=false, nodeName=LeaveVehicle}
               services []
              }
             ]
             decorators []
             properties {children [
               {childTask=&self:BTTask_Wait_6}
               {childTask=&self:FortAthenaBTTask_LeaveVehicle_10}
              ]}
            }
           ]
           decorators [Has a leader to follow {
             class=BTDecorator_Blackboard
             properties {
              arithmeticOperation=Equal
              intValue=0
              nodeName=Has a leader to follow
              stringValue=""
              blackboardKey {selectedKeyName=AIEvaluator_FollowGroupLeader_HasLeaderToFollow}
             }
            }]
           properties {
            children [
             {
              childComposite=&self:BTComposite_Selector_0
              childTask=null
              decorators [&self:BTDecorator_Blackboard_39]
             }
             {childComposite=&self:BTComposite_Sequence_12, childTask=null, decorators []}
            ]
            services [&self:FortAthenaBTService_AIEvaluator_23]
           }
           services [Focus on target {
             class=FortAthenaBTService_AIEvaluator
             properties {interval=0.5, nodeName=Focus on target, aIEvaluatorClass {defaultValue=&FortAthenaAIBotEvaluator_HandleFocusing}}
            }]
          }
          Sequence {
           class=BTComposite_Sequence
           children [
            Simple Parallel {
             class=BTComposite_SimpleParallel
             children [
              MoveTo PatrolPoint {
               class=FortAthenaBTTask_BotMoveTo
               decorators []
               properties {
                movementResultKeyName=AIEvaluator_Patrolling_MovementState
                nodeName=MoveTo PatrolPoint
                acceptableRadius {defaultValue=10}
                allowRandomWobble {defaultValue=false}
                bAllowStrafe {defaultValue=true}
                bStartFromPreviousPath {defaultValue=true}
                bTrackMovingGoal {defaultValue=false}
                blackboardKey {selectedKeyName=AIEvaluator_Patrolling_Destination}
                filterClass {
                 defaultValue=&NavFilter_Phoebe_NoSmashable_C
                 key=InVehicle_Patrol_MoveTo_FilterClass
                }
                observedBlackboardValueTolerance {defaultValue=50}
               }
              }
              Selector {
               class=BTComposite_Selector
               children [
                Sequence {
                 class=BTComposite_Sequence
                 children [
                  FortAthenaBTTask_PauseVehicle_3 {class=FortAthenaBTTask_PauseVehicle, children [], decorators [], properties {bIgnoreRestartSelf=false, nodeName=""}, services []}
                  LeaveVehicle {
                   class=FortAthenaBTTask_LeaveVehicle
                   children []
                   decorators []
                   properties {bIgnoreRestartSelf=false, bWaitVehicleStop=false, nodeName=LeaveVehicle}
                   services []
                  }
                 ]
                 decorators [Should execute smart object behavior? {
                   class=BTDecorator_Blackboard
                   properties {
                    arithmeticOperation=GreaterOrEqual
                    flowAbortMode=None
                    intValue=4
                    nodeName=Should execute smart object behavior?
                    stringValue=EExecutionStatus::ExecutionAllowed
                    blackboardKey {selectedKeyName=AIEvaluator_SmartObject_ExecutionStatus}
                   }
                  }]
                 properties {children [
                   {childTask=&self:FortAthenaBTTask_PauseVehicle_3}
                   {childTask=&self:FortAthenaBTTask_LeaveVehicle_5}
                  ]}
                }
                Simple Parallel {
                 class=BTComposite_SimpleParallel
                 children [
                  FortAthenaBTTask_Escalate_1 {
                   class=FortAthenaBTTask_Escalate
                   children []
                   decorators [Escalate? {
                     class=BTDecorator_Blackboard
                     properties {
                      arithmeticOperation=GreaterOrEqual
                      flowAbortMode=None
                      intValue=3
                      nodeName=Escalate?
                      stringValue=ExecutionPending
                      blackboardKey {selectedKeyName=AIEvaluator_Escalate_ExecutionStatus}
                     }
                    }]
                   properties {bIgnoreRestartSelf=false, nodeName=""}
                   services []
                  }
                  Honk {
                   class=FortAthenaBTTask_VehicleHonk
                   children []
                   decorators []
                   properties {bIgnoreRestartSelf=false, maxFlickerTime=0.5, maxHonkTime=4, minFlickerTime=0.1, minHonkTime=1, nodeName=Honk}
                   services []
                  }
                 ]
                 properties {
                  children [
                   {
                    childTask=&self:FortAthenaBTTask_Escalate_1
                    decorators [&self:BTDecorator_Blackboard_10]
                   }
                   {childTask=&self:FortAthenaBTTask_VehicleHonk_0}
                  ]
                  services []
                 }
                 services []
                }
               ]
               decorators []
               properties {
                children [
                 {
                  childComposite=&self:BTComposite_Sequence_8
                  childTask=null
                  decorators [&self:BTDecorator_Blackboard_12]
                 }
                 {childComposite=&self:BTComposite_SimpleParallel_3, childTask=null, decorators []}
                ]
                services []
               }
               services []
              }
             ]
             properties {
              children [
               {childTask=&self:FortAthenaBTTask_BotMoveTo_5}
               {childComposite=&self:BTComposite_Selector_11, childTask=null}
              ]
              services []
             }
             services []
            }
            LeaveVehicle {
             class=FortAthenaBTTask_LeaveVehicle
             children []
             decorators [Can Run LocalGameplayBehavior? {
               class=BTDecorator_Blackboard
               properties {
                arithmeticOperation=GreaterOrEqual
                flowAbortMode=Self
                intValue=3
                nodeName=Can Run LocalGameplayBehavior?
                stringValue=EExecutionStatus::ExecutionPending
                blackboardKey {selectedKeyName=AIEvaluator_DynamicBlueprint_ExecutionStatus}
               }
              }]
             properties {bIgnoreRestartSelf=false, bWaitVehicleStop=false, nodeName=LeaveVehicle}
             services []
            }
           ]
           decorators [Should Move? {
             class=BTDecorator_Blackboard
             properties {
              arithmeticOperation=GreaterOrEqual
              intValue=3
              nodeName=Should Move?
              stringValue=""
              blackboardKey {selectedKeyName=AIEvaluator_Patrolling_ShouldMove}
             }
            }]
           properties {children [
             {childComposite=&self:BTComposite_SimpleParallel_1, childTask=null}
             {
              childTask=&self:FortAthenaBTTask_LeaveVehicle_7
              decorators [&self:BTDecorator_Blackboard_18]
             }
            ]}
          }
          LeaveVehicle {
           class=FortAthenaBTTask_LeaveVehicle
           children []
           decorators []
           properties {bIgnoreRestartSelf=false, bWaitVehicleStop=false, nodeName=LeaveVehicle}
           services []
          }
         ]
         decorators [Loop {class=BTDecorator_Loop}]
         properties {
          children [
           {
            childComposite=&self:BTComposite_Selector_13
            childTask=null
            decorators [&self:BTDecorator_Blackboard_6]
           }
           {
            childComposite=&self:BTComposite_Sequence_9
            childTask=null
            decorators [&self:BTDecorator_Blackboard_0]
           }
           {childComposite=null, childTask=&self:FortAthenaBTTask_LeaveVehicle_11, decorators []}
          ]
          services [
           &self:FortAthenaBTService_AIEvaluator_10
           &self:FortAthenaBTService_Patrolling_1
           &self:FortAthenaBTService_AIEvaluator_11
           &self:FortAthenaBTService_AIEvaluator_13
           &self:FortAthenaBTService_ModulateSpeedOnDistance_0
           &self:FortAthenaBTService_AIEvaluator_12
          ]
         }
         services [
          Evaluating...Patrolling {
           class=FortAthenaBTService_AIEvaluator
           properties {
            interval=0.25
            nodeName=Evaluating...Patrolling
            randomDeviation=0
            aIEvaluatorClass {defaultValue=/Script/FortniteAIServer.FortAthenaNpcEvaluator_Patrolling}
           }
          }
          Executing...Patrolling {
           class=FortAthenaBTService_Patrolling
           properties {
            acceptableRadius=5
            bAllowPartialPath=true
            bCallTickOnSearchStart=false
            bProjectGoalLocation=true
            bRestartTimerOnEachActivation=false
            filterClass=null
            interval=0.5
            nodeName=Executing...Patrolling
            patrollingAppendDestinationKeyName=AIEvaluator_Patrolling_AppendDestination
            randomDeviation=0.1
           }
          }
          Evaluating...Smart Object {
           class=FortAthenaBTService_AIEvaluator
           properties {
            interval=0.5
            nodeName=Evaluating...Smart Object
            aIEvaluatorClass {defaultValue=/Script/FortniteAIServer.FortAthenaAIBotEvaluator_SmartObjects}
           }
          }
          Evaluating... FollowGroupLeader {
           class=FortAthenaBTService_AIEvaluator
           properties {
            interval=0.5
            nodeName=Evaluating... FollowGroupLeader
            aIEvaluatorClass {defaultValue=/AICore/Humanoids/BehaviorTrees/Evaluators/AIEvaluator_VehicleFollowGroupLeader.AIEvaluator_VehicleFollowGroupLeader_C}
           }
          }
          Executing... Modulate Vehicle Speed On Distance {
           class=FortAthenaBTService_ModulateSpeedOnDistance
           properties {
            bCallTickOnSearchStart=false
            bRestartTimerOnEachActivation=false
            followGroupLeaderDestinationKeyName=AIEvaluator_FollowGroupLeader_Destination
            interval=0.5
            nodeName=Executing... Modulate Vehicle Speed On Distance
            randomDeviation=0.1
            tooFarFromLeaderKeyName=AIEvaluator_FollowGroupLeader_TooFarFromLeader
            maxSpeed {defaultValue=30, key=VehicleDriver_PatrolMaxSpeed}
            targetSpeed {defaultValue=21, key=VehicleDriver_PatrolTargetSpeed}
           }
          }
          Evaluating... Escalate {
           class=FortAthenaBTService_AIEvaluator
           properties {
            interval=0.5
            nodeName=Evaluating... Escalate
            aIEvaluatorClass {defaultValue=/Script/FortniteAIServer.FortAthenaAIBotEvaluator_VehicleEscalate}
           }
          }
         ]
        }
       ]
       decorators []
       properties {
        children [
         {childTask=&self:FortAthenaBTTask_SetAggressiveDriving_3}
         {childTask=&self:FortAthenaBTTask_ModulateVehicleSpeed_4}
         {
          childComposite=&self:BTComposite_Selector_6
          childTask=null
          decorators [&self:BTDecorator_Loop_1]
         }
        ]
        services [
         &self:FortAthenaBTService_AIEvaluator_15
         &self:FortAthenaBTService_AIEvaluator_16
         &self:FortAthenaBTService_PropagatePatrolProgressToPassengers_2
         &self:FortAthenaBTService_WaitForPassengers_2
         &self:FortAthenaBTService_AIEvaluator_17
         &self:FortAthenaBTService_AIEvaluator_18
        ]
       }
       services [
        Evaluating... Holster Weapon {
         class=FortAthenaBTService_AIEvaluator
         properties {
          interval=1
          nodeName=Evaluating... Holster Weapon
          randomDeviation=0.2
          aIEvaluatorClass {defaultValue=/Script/FortniteAIServer.FortAthenaAIBotEvaluator_HolsterWeapon}
         }
        }
        Evaluating...Wait For Passengers {
         class=FortAthenaBTService_AIEvaluator
         properties {
          interval=0.5
          nodeName=Evaluating...Wait For Passengers
          aIEvaluatorClass {defaultValue=/Script/FortniteAIServer.FortAthenaAIBotEvaluator_WaitForPassengers}
         }
        }
        Executing...PropagatePatrolProgressToPassengers {
         class=FortAthenaBTService_PropagatePatrolProgressToPassengers
         properties {
          bCallTickOnSearchStart=false
          bRestartTimerOnEachActivation=false
          interval=0.5
          nodeName=Executing...PropagatePatrolProgressToPassengers
          patrollingAppendDestinationKeyName=AIEvaluator_Patrolling_AppendDestination
          randomDeviation=0.1
         }
        }
        Executing...WaitForPassengers {
         class=FortAthenaBTService_WaitForPassengers
         properties {
          bCallTickOnSearchStart=false
          bRestartTimerOnEachActivation=false
          interval=0.5
          nodeName=Executing...WaitForPassengers
          randomDeviation=0.1
          waitForPassengersStatusKeyName=AIEvaluator_WaitForPassengerStatus
         }
        }
        Focus on target {
         class=FortAthenaBTService_AIEvaluator
         properties {interval=0.5, nodeName=Focus on target, aIEvaluatorClass {defaultValue=&FortAthenaAIBotEvaluator_HandleFocusing}}
        }
        Evaluating...Range Attack {
         class=FortAthenaBTService_AIEvaluator
         properties {
          interval=1
          nodeName=Evaluating...Range Attack
          aIEvaluatorClass {defaultValue=&FortAthenaAIBotEvaluator_RangeAttack}
         }
        }
       ]
      }
     ]
     decorators [Loop {class=BTDecorator_Loop}]
     properties {
      children [
       {
        childComposite=null
        childTask=&self:FortBTTask_InjectionPoint_1
        decorators [&self:FortBTDecorator_CanDoInjectedBehavior_0]
       }
       {childComposite=&self:BTComposite_Selector_2, childTask=null, decorators []}
       {childComposite=&self:BTComposite_Sequence_11, childTask=null, decorators []}
      ]
      services [&self:FortBTService_AIEvaluatorsInjector_0]
     }
     services [Behavior Injector {
       class=FortBTService_AIEvaluatorsInjector
       properties {
        bCallTickOnSearchStart=false
        bRestartTimerOnEachActivation=false
        interval=0.5
        nodeName=Behavior Injector
        randomDeviation=0.1
        injectionTagsKeys [{
          actionInterfaceObjectKeyName=Injection_AICore_InVehicle_Driver_HighPriority_ActionObject
          injectionKeyName=AIEvaluator_Injection_AICore_InVehicle_Driver_HighPriority
          injectionTag {tagName=Behavior.Injection.AICore.InVehicle.Driver.HighPriority}
         }]
       }
      }]
    }
   ]
   decorators [Is Driving Vehicle {
     class=BTDecorator_Blackboard
     properties {
      arithmeticOperation=Equal
      intValue=0
      nodeName=Is Driving Vehicle
      notifyObserver=ValueChange
      stringValue=""
      blackboardKey {selectedKeyName=AIEvaluator_IsDrivingVehicle}
     }
    }]
   properties {children [
     {childTask=&self:FortAthenaBTTask_ModulateVehicleSpeed_2}
     {
      childComposite=&self:BTComposite_Selector_19
      childTask=null
      decorators [&self:BTDecorator_Loop_9]
     }
    ]}
  }
  Selector {
   class=BTComposite_Selector
   children [
    LeaveVehicle {
     class=FortAthenaBTTask_LeaveVehicle
     children []
     decorators [Should leave vehicle {
       class=BTDecorator_Blackboard
       properties {
        arithmeticOperation=GreaterOrEqual
        intValue=4
        nodeName=Should leave vehicle
        stringValue=EExecutionStatus::ExecutionAllowed
        blackboardKey {selectedKeyName=AIEvaluator_LeaveSeatStatus}
       }
      }]
     properties {bIgnoreRestartSelf=false, bWaitVehicleStop=false, nodeName=LeaveVehicle}
     services []
    }
    Selector {
     class=BTComposite_Selector
     children [
      Selector {
       class=BTComposite_Selector
       children [
        Sequence {
         class=BTComposite_Sequence
         children [
          Wait {class=BTTask_Wait, properties {randomDeviation {defaultValue=0.5}, waitTime {defaultValue=2}}}
          LeaveVehicle {
           class=FortAthenaBTTask_LeaveVehicle
           children []
           decorators []
           properties {bIgnoreRestartSelf=false, bWaitVehicleStop=false, nodeName=LeaveVehicle}
           services []
          }
         ]
         decorators [
          Can't fire weapon {
           class=BTDecorator_Blackboard
           properties {
            arithmeticOperation=LessOrEqual
            intValue=1
            nodeName=Can't fire weapon
            notifyObserver=ValueChange
            stringValue=EExecutionStatus::ExecutionDenied
            blackboardKey {selectedKeyName=AIEvaluator_WeaponFire_ExecutionStatus}
           }
          }
          Not Reloading {
           class=BTDecorator_Blackboard
           properties {
            arithmeticOperation=LessOrEqual
            intValue=1
            nodeName=Not Reloading
            stringValue=EExecutionStatus::ExecutionDenied
            blackboardKey {selectedKeyName=AIEvaluator_WeaponReload_ExecutionStatus}
           }
          }
          Should Switch Seat {
           class=BTDecorator_Blackboard
           properties {
            arithmeticOperation=GreaterOrEqual
            intValue=4
            nodeName=Should Switch Seat
            stringValue=EExecutionStatus::ExecutionAllowed
            blackboardKey {selectedKeyName=AIEvaluator_SwitchSeatStatus}
           }
          }
         ]
         properties {children [
           {childTask=&self:BTTask_Wait_3}
           {childTask=&self:FortAthenaBTTask_LeaveVehicle_6}
          ]}
        }
        Wait {class=BTTask_Wait, decorators [Loop {class=BTDecorator_Loop}], properties {waitTime {defaultValue=10}}}
       ]
       decorators [IsThreatened {
         class=BTDecorator_Blackboard
         properties {
          arithmeticOperation=Equal
          intValue=3
          nodeName=IsThreatened
          stringValue=EAlertLevel::Threatened
          blackboardKey {selectedKeyName=AIEvaluator_Global_AlertLevel}
         }
        }]
       properties {
        children [
         {
          childComposite=&self:BTComposite_Sequence_3
          childTask=null
          decorators [
           &self:BTDecorator_Blackboard_28
           &self:BTDecorator_Blackboard_29
           &self:BTDecorator_Blackboard_30
          ]
         }
         {
          childComposite=null
          childTask=&self:BTTask_Wait_0
          decorators [&self:BTDecorator_Loop_3]
         }
        ]
        services [
         &self:FortAthenaBTService_AIEvaluator_6
         &self:FortAthenaBTService_ManageWeapon_2
         &self:FortAthenaBTService_AIEvaluator_7
        ]
       }
       services [
        Evaluating...Vehicle Switch Seat {
         class=FortAthenaBTService_AIEvaluator
         properties {
          interval=0.5
          nodeName=Evaluating...Vehicle Switch Seat
          aIEvaluatorClass {defaultValue=/AICore/Humanoids/BehaviorTrees/Evaluators/VehicleSwitchSeatEvaluator.VehicleSwitchSeatEvaluator_C}
         }
        }
        Executing...Weapons Actions {class=FortAthenaBTService_ManageWeapon, properties {interval=1, randomDeviation=0.3}}
        Focus on target {
         class=FortAthenaBTService_AIEvaluator
         properties {interval=0.5, nodeName=Focus on target, aIEvaluatorClass {defaultValue=&FortAthenaAIBotEvaluator_HandleFocusing}}
        }
       ]
      }
      LKP {
       class=BTComposite_Selector
       children [Sequence {
         class=BTComposite_Sequence
         children [
          Wait {class=BTTask_Wait, properties {randomDeviation {defaultValue=1}, waitTime {defaultValue=5}}}
          LeaveVehicle {
           class=FortAthenaBTTask_LeaveVehicle
           children []
           decorators [Should Switch Seat {
             class=BTDecorator_Blackboard
             properties {
              arithmeticOperation=GreaterOrEqual
              flowAbortMode=Self
              intValue=4
              nodeName=Should Switch Seat
              stringValue=EExecutionStatus::ExecutionAllowed
              blackboardKey {selectedKeyName=AIEvaluator_SwitchSeatStatus}
             }
            }]
           properties {bIgnoreRestartSelf=false, bWaitVehicleStop=false, nodeName=LeaveVehicle}
           services []
          }
         ]
         decorators []
         properties {children [
           {childTask=&self:BTTask_Wait_9}
           {
            childTask=&self:FortAthenaBTTask_LeaveVehicle_4
            decorators [&self:BTDecorator_Blackboard_11]
           }
          ]}
        }]
       decorators [IsLKP {
         class=BTDecorator_Blackboard
         properties {
          arithmeticOperation=Equal
          intValue=2
          nodeName=IsLKP
          stringValue=EAlertLevel::LKP
          blackboardKey {selectedKeyName=AIEvaluator_Global_AlertLevel}
         }
        }]
       properties {
        nodeName=LKP
        children [{childComposite=&self:BTComposite_Sequence_7, childTask=null, decorators []}]
        services [
         &self:FortAthenaBTService_ManageWeapon_0
         &self:FortAthenaBTService_AIEvaluator_4
        ]
       }
       services [
        Executing...Weapons Actions {class=FortAthenaBTService_ManageWeapon, properties {interval=0.2, randomDeviation=0.1}}
        Evaluating...Vehicle Switch Seat {
         class=FortAthenaBTService_AIEvaluator
         properties {
          interval=0.5
          nodeName=Evaluating...Vehicle Switch Seat
          aIEvaluatorClass {defaultValue=&VehicleSwitchSeatEvaluator_C}
         }
        }
       ]
      }
     ]
     decorators []
     properties {
      children [
       {
        childComposite=&self:BTComposite_Selector_5
        childTask=null
        decorators [&self:BTDecorator_Blackboard_7]
       }
       {
        childComposite=&self:BTComposite_Selector_17
        childTask=null
        decorators [&self:BTDecorator_Blackboard_3]
       }
      ]
      services [
       &self:FortAthenaBTService_AIEvaluator_38
       &self:FortAthenaBTService_AIEvaluator_30
      ]
     }
     services [
      Evaluating...Select Weapon {
       class=FortAthenaBTService_AIEvaluator
       properties {
        interval=0.2
        nodeName=Evaluating...Select Weapon
        randomDeviation=0.05
        aIEvaluatorClass {defaultValue=&FortAthenaAIBotEvaluator_WeaponSelection}
       }
      }
      Evaluating...Range Attack {
       class=FortAthenaBTService_AIEvaluator
       properties {
        bCallTickOnSearchStart=true
        interval=0.2
        nodeName=Evaluating...Range Attack
        randomDeviation=0.05
        aIEvaluatorClass {defaultValue=&VehicleRangeAttackEvaluator_C}
       }
      }
     ]
    }
    ALERTED {
     class=BTComposite_Selector
     children [Loop Alerted {
       class=BTComposite_Selector
       children [
        Wait {
         class=BTTask_Wait
         decorators [Should Observe {
           class=BTDecorator_Blackboard
           properties {
            arithmeticOperation=GreaterOrEqual
            intValue=3
            nodeName=Should Observe
            stringValue=EExecutionStatus::ExecutionPending
            blackboardKey {selectedKeyName=AIEvaluator_Observe_ExecutionStatus}
           }
          }]
         properties {randomDeviation {defaultValue=0.25}, waitTime {defaultValue=1}}
        }
        FallBack {class=BTTask_Wait, properties {nodeName=FallBack, waitTime {defaultValue=1}}}
       ]
       decorators [Loop {class=BTDecorator_Loop}]
       properties {
        nodeName=Loop Alerted
        children [
         {
          childComposite=null
          childTask=&self:BTTask_Wait_16
          decorators [&self:BTDecorator_Blackboard_9]
         }
         {childComposite=null, childTask=&self:BTTask_Wait_8, decorators []}
        ]
        services []
       }
       services []
      }]
     decorators [IsAlerted {
       class=BTDecorator_Blackboard
       properties {
        arithmeticOperation=Equal
        intValue=1
        nodeName=IsAlerted
        stringValue=EAlertLevel::Alerted
        blackboardKey {selectedKeyName=AIEvaluator_Global_AlertLevel}
       }
      }]
     properties {
      nodeName=ALERTED
      children [{
        childComposite=&self:BTComposite_Selector_1
        childTask=null
        decorators [&self:BTDecorator_Loop_8]
       }]
      services [&self:FortAthenaBTService_AIEvaluator_8]
     }
     services [Evaluating...Observe Threat {
       class=FortAthenaBTService_AIEvaluator
       properties {
        bCallTickOnSearchStart=true
        interval=0.5
        nodeName=Evaluating...Observe Threat
        aIEvaluatorClass {defaultValue=/Script/FortniteAIServer.FortAthenaAIBotEvaluator_Observe}
       }
      }]
    }
    UNAWARE {
     class=BTComposite_Selector
     children [
      Selector {
       class=BTComposite_Selector
       children [
        LeaveVehicle {
         class=FortAthenaBTTask_LeaveVehicle
         children []
         decorators [Should execute smart object behavior? {
           class=BTDecorator_Blackboard
           properties {
            arithmeticOperation=GreaterOrEqual
            flowAbortMode=None
            intValue=4
            nodeName=Should execute smart object behavior?
            stringValue=EExecutionStatus::ExecutionAllowed
            blackboardKey {selectedKeyName=AIEvaluator_SmartObject_ExecutionStatus}
           }
          }]
         properties {bIgnoreRestartSelf=false, bWaitVehicleStop=false, nodeName=LeaveVehicle}
         services []
        }
        Sequence {
         class=BTComposite_Sequence
         children [
          Wait {class=BTTask_Wait, properties {waitTime {defaultValue=5}}}
          FortAthenaBTTask_VehicleSwitchSeat_0 {
           class=FortAthenaBTTask_VehicleSwitchSeat
           children []
           decorators []
           properties {bIgnoreRestartSelf=false, nodeName="", switchSeatType=ToDriver}
           services []
          }
         ]
         decorators []
         properties {children [
           {childTask=&self:BTTask_Wait_4}
           {childTask=&self:FortAthenaBTTask_VehicleSwitchSeat_0}
          ]}
        }
       ]
       decorators [Should Switch Seat {
         class=BTDecorator_Blackboard
         properties {
          arithmeticOperation=GreaterOrEqual
          intValue=4
          nodeName=Should Switch Seat
          stringValue=EExecutionStatus::ExecutionAllowed
          blackboardKey {selectedKeyName=AIEvaluator_SwitchSeatStatus}
         }
        }]
       properties {
        children [
         {
          childComposite=null
          childTask=&self:FortAthenaBTTask_LeaveVehicle_8
          decorators [&self:BTDecorator_Blackboard_20]
         }
         {childComposite=&self:BTComposite_Sequence_6, childTask=null, decorators []}
        ]
        services [&self:FortAthenaBTService_AIEvaluator_31]
       }
       services [Evaluating...Smart Object {
         class=FortAthenaBTService_AIEvaluator
         properties {
          interval=0.5
          nodeName=Evaluating...Smart Object
          aIEvaluatorClass {defaultValue=&FortAthenaAIBotEvaluator_SmartObjects}
         }
        }]
      }
      Wait {class=BTTask_Wait, properties {waitTime {defaultValue=5}}}
     ]
     decorators []
     properties {
      nodeName=UNAWARE
      children [
       {
        childComposite=&self:BTComposite_Selector_8
        childTask=null
        decorators [&self:BTDecorator_Blackboard_1]
       }
       {childComposite=null, childTask=&self:BTTask_Wait_1, decorators []}
      ]
      services [
       &self:FortAthenaBTService_AIEvaluator_0
       &self:FortAthenaBTService_AIEvaluator_20
       &self:FortAthenaBTService_AIEvaluator_21
       &self:FortAthenaBTService_ManageWeapon_5
       &self:FortAthenaBTService_AIEvaluator_22
      ]
     }
     services [
      Evaluating...Select Weapon {
       class=FortAthenaBTService_AIEvaluator
       properties {
        interval=0.2
        nodeName=Evaluating...Select Weapon
        randomDeviation=0.05
        aIEvaluatorClass {defaultValue=&FortAthenaAIBotEvaluator_WeaponSelection}
       }
      }
      Evaluating...Range Attack {
       class=FortAthenaBTService_AIEvaluator
       properties {
        interval=1
        nodeName=Evaluating...Range Attack
        aIEvaluatorClass {defaultValue=&FortAthenaAIBotEvaluator_RangeAttack}
       }
      }
      Evaluating... Holster Weapon {
       class=FortAthenaBTService_AIEvaluator
       properties {
        interval=1
        nodeName=Evaluating... Holster Weapon
        randomDeviation=0.2
        aIEvaluatorClass {defaultValue=&FortAthenaAIBotEvaluator_HolsterWeapon}
       }
      }
      Executing...Weapons Actions {class=FortAthenaBTService_ManageWeapon, properties {interval=0.2, randomDeviation=0.1}}
      Evaluating...Vehicle Switch Seat {
       class=FortAthenaBTService_AIEvaluator
       properties {
        interval=0.5
        nodeName=Evaluating...Vehicle Switch Seat
        aIEvaluatorClass {defaultValue=&VehicleSwitchSeatEvaluator_C}
       }
      }
     ]
    }
   ]
   decorators [
    Is Driving Vehicle {
     class=BTDecorator_Blackboard
     properties {
      arithmeticOperation=Equal
      basicOperation=NotSet
      intValue=0
      nodeName=Is Driving Vehicle
      notifyObserver=ValueChange
      stringValue=""
      blackboardKey {selectedKeyName=AIEvaluator_IsDrivingVehicle}
     }
    }
    Loop {class=BTDecorator_Loop, properties {numLoops {defaultValue=1}}}
   ]
   properties {
    children [
     {
      childComposite=null
      childTask=&self:FortAthenaBTTask_LeaveVehicle_9
      decorators [&self:BTDecorator_Blackboard_33]
     }
     {childComposite=&self:BTComposite_Selector_12, childTask=null, decorators []}
     {
      childComposite=&self:BTComposite_Selector_3
      childTask=null
      decorators [&self:BTDecorator_Blackboard_8]
     }
     {childComposite=&self:BTComposite_Selector_9, childTask=null, decorators []}
    ]
    services [&self:FortAthenaBTService_AIEvaluator_3]
   }
   services [Evaluating...Leave Vehicle {
     class=FortAthenaBTService_AIEvaluator
     properties {
      interval=0.5
      nodeName=Evaluating...Leave Vehicle
      aIEvaluatorClass {defaultValue=&VehicleLeaveSeatEvaluator_C}
     }
    }]
  }
 ]
 decorators []
 properties {
  children [
   {
    childComposite=null
    childTask=&self:FortBTTask_InjectionPoint_0
    decorators [&self:FortBTDecorator_CanDoInjectedBehavior_1]
   }
   {
    childComposite=&self:BTComposite_Sequence_15
    childTask=null
    decorators [&self:BTDecorator_Blackboard_14]
   }
   {
    childComposite=&self:BTComposite_Selector_22
    childTask=null
    decorators [
     &self:BTDecorator_Blackboard_2
     &self:BTDecorator_Loop_2
    ]
   }
  ]
  services [&self:FortBTService_AIEvaluatorsInjector_1]
 }
 services [Behavior Injector {
   class=FortBTService_AIEvaluatorsInjector
   properties {
    bCallTickOnSearchStart=false
    bRestartTimerOnEachActivation=false
    interval=0.5
    nodeName=Behavior Injector
    randomDeviation=0.1
    injectionTagsKeys [{
      actionInterfaceObjectKeyName=Injection_AICore_InVehicle_HighPriority_ActionObject
      injectionKeyName=AIEvaluator_Injection_AICore_InVehicle_HighPriority
      injectionTag {tagName=Behavior.Injection.AICore.InVehicle.HighPriority}
     }]
   }
  }]
}
