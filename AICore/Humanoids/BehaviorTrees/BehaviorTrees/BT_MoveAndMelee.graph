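# Humanoid AI behavior tree that equips a weapon, then moves to a token position and evaluates melee attacks, flagged by a tactical-action service.
# Services run focus, token position, movement style and weapon actions, with evaluator classes, movement style and navigation filter supplied through blackboard keys.
# A nested selector runs the melee attack evaluator, strafing to the token destination when its execution status is at least pending, otherwise waiting in place.
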BehaviorTree /AICore/Humanoids/BehaviorTrees/BehaviorTrees/BT_MoveAndMelee
blackboard=/AICore/Humanoids/BehaviorTrees/Blackboards/BB_MoveAndMelee
root {
 name=Sequence
 class=BTComposite_Sequence
 children [
  Equip Item {
   class=FortAthenaBTTask_EquipItem
   children []
   decorators []
   properties {
    bIgnoreRestartSelf=false
    equipItemKeyName=AIEvaluator_Injection_Item
    nodeName=Equip Item
    weaponKeyName=AIEvaluator_Global_DesiredWeapon
    disableEquipAnimation {defaultValue=false, key=None}
    itemTagToEquip {key=None, defaultValue {gameplayTags []}}
    waitForEquipToFinish {defaultValue=false, key=None}
   }
   services []
  }
  Selector {
   class=BTComposite_Selector
   children [Selector {
     class=BTComposite_Selector
     children [
      Reposition {
       class=FortAthenaBTTask_BotMoveTo
       decorators [Token Destination {
         class=BTDecorator_Blackboard
         properties {
          arithmeticOperation=GreaterOrEqual
          intValue=3
          nodeName=Token Destination
          stringValue=ExecutionPending
          blackboardKey {selectedKeyName=AIEvaluator_TokenPosition_ExecutionStatus}
         }
        }]
       properties {
        executionStatusKeyName=AIEvaluator_TokenPosition_ExecutionStatus
        movementResultKeyName=AIEvaluator_TokenPosition_MovementState
        nodeName=Reposition
        acceptableRadius {defaultValue=1}
        allowRandomWobble {defaultValue=true}
        bAllowPartialPath {defaultValue=false}
        bAllowStrafe {defaultValue=true}
        bReachTestIncludesAgentRadius {defaultValue=false}
        bReachTestIncludesGoalRadius {defaultValue=false}
        blackboardKey {selectedKeyName=AIEvaluator_TokenPosition_Destination}
        filterClass {key=Melee_NavigationFilter}
        observedBlackboardValueTolerance {defaultValue=2.75}
       }
      }
      Stay in Position {class=BTTask_Wait, properties {nodeName=Stay in Position, waitTime {defaultValue=30}}}
     ]
     decorators []
     properties {
      children [
       {
        childComposite=null
        childTask=&self:FortAthenaBTTask_BotMoveTo_0
        decorators [&self:BTDecorator_Blackboard_0]
       }
       {childComposite=null, childTask=&self:BTTask_Wait_4, decorators []}
      ]
      services [&self:FortAthenaBTService_AIEvaluator_0]
     }
     services [Evaluating...Melee Attack {
       class=FortAthenaBTService_AIEvaluator
       properties {interval=0.1, nodeName=Evaluating...Melee Attack, randomDeviation=0, aIEvaluatorClass {defaultValue=null, key=Melee_MeleeAttackEvaluator}}
      }]
    }]
   decorators []
   properties {
    children [{childComposite=&self:BTComposite_Selector_3, childTask=null, decorators []}]
    services [
     &self:FortAthenaBTService_AIEvaluator_6
     &self:FortAthenaBTService_MovementStyle_2
     &self:FortAthenaBTService_AIEvaluator_5
     &self:FortAthenaBTService_ManageWeapon_0
    ]
   }
   services [
    Focus on target {
     class=FortAthenaBTService_AIEvaluator
     properties {interval=0.5, nodeName=Focus on target, aIEvaluatorClass {defaultValue=null, key=Melee_FocusingEvaluator}}
    }
    FortPawn Movement Style {
     class=FortAthenaBTService_MovementStyle
     properties {
      bCallTickOnSearchStart=false
      bRestartTimerOnEachActivation=false
      interval=0.5
      nodeName=FortPawn Movement Style
      randomDeviation=0.1
      blackboardCondition {bNoneIsAllowedValue=true, selectedKeyName=None}
      invertBlackboardCondition {defaultValue=false, key=None}
      newMovementStyle {defaultValue=0, enumType=/Script/FortniteGame.EFortMovementStyle, key=Melee_MovementStyle, nativeEnumTypeName=EFortMovementStyle}
      tagCondition {autoDescription="", tokenStreamVersion=0, userDescription="", queryTokenStream [], tagDictionary []}
     }
    }
    Evaluating...Has Token Position {
     class=FortAthenaBTService_AIEvaluator
     properties {
      interval=0.1
      nodeName=Evaluating...Has Token Position
      randomDeviation=0
      aIEvaluatorClass {defaultValue=null, key=Melee_TokenPositionEvaluator}
     }
    }
    Executing...Weapon Actions {class=FortAthenaBTService_ManageWeapon, properties {interval=0.5, nodeName=Executing...Weapon Actions, randomDeviation=0.1}}
   ]
  }
 ]
 decorators []
 properties {
  children [
   {childTask=&self:FortAthenaBTTask_EquipItem_0}
   {childComposite=&self:BTComposite_Selector_1, childTask=null}
  ]
  services [&self:FortAthenaBTService_SetInTacticalAction_0]
 }
 services [FortAthenaBTService_SetInTacticalAction_0 {
   class=FortAthenaBTService_SetInTacticalAction
   properties {shouldSetInTacticalAction {key=None}, tacticalActionIdentifierTag {defaultValue {_structType=&GameplayTag}}}
  }]
}
