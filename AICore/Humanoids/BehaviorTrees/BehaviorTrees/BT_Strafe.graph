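# Behavior tree for humanoid AI local strafing, periodically moving the bot to a nearby repositioning point when Strafe_CanLocalStrafe is set.
# A looped sequence waits a blackboard-driven cooldown, runs the EQS query named by the blackboard, then moves the bot to the resulting destination with wobble allowed.
# When strafing is not allowed, the selector falls back to a very long idle wait; neither blackboard check aborts on change.
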BehaviorTree /AICore/Humanoids/BehaviorTrees/BehaviorTrees/BT_Strafe
blackboard=/AICore/Humanoids/BehaviorTrees/Blackboards/BB_Strafe
root {
 name=Selector
 class=BTComposite_Selector
 children [
  Sequence {
   class=BTComposite_Sequence
   children [
    Wait {
     class=BTTask_Wait
     properties {randomDeviation {key=Strafe_LocalStrafingCooldownDeviation}, waitTime {defaultValue=5, key=Strafe_LocalStrafingCooldown}}
    }
    Run EQS Query {
     class=BTTask_RunEQSQuery
     properties {
      nodeName=Run EQS Query
      blackboardKey {selectedKeyName=Strafe_LocalStrafingDestination}
      eQSRequest {bUseBBKeyForQueryTemplate=true, queryTemplate=null, runMode=SingleResult, eQSQueryBlackboardKey {selectedKeyName=Strafe_LocalStrafingEQS}}
     }
    }
    Bot Move To {
     class=FortAthenaBTTask_BotMoveTo
     decorators []
     properties {
      movementResultKeyName=None
      nodeName=Bot Move To
      acceptableRadius {defaultValue=5}
      allowRandomWobble {defaultValue=true}
      bAllowStrafe {defaultValue=false}
      blackboardKey {selectedKeyName=Strafe_LocalStrafingDestination}
      observedBlackboardValueTolerance {defaultValue=4.75}
     }
    }
   ]
   decorators [
    Allow local strafing {
     class=BTDecorator_Blackboard
     properties {
      arithmeticOperation=Equal
      flowAbortMode=None
      intValue=0
      nodeName=Allow local strafing
      stringValue=""
      blackboardKey {selectedKeyName=Strafe_CanLocalStrafe}
     }
    }
    Loop {class=BTDecorator_Loop}
   ]
   properties {children [
     {childTask=&self:BTTask_Wait_0}
     {childTask=&self:BTTask_RunEQSQuery_1}
     {childTask=&self:FortAthenaBTTask_BotMoveTo_0}
    ]}
  }
  Wait {
   class=BTTask_Wait
   decorators [Allow local strafing {
     class=BTDecorator_Blackboard
     properties {
      arithmeticOperation=Equal
      basicOperation=NotSet
      flowAbortMode=None
      intValue=0
      nodeName=Allow local strafing
      stringValue=""
      blackboardKey {selectedKeyName=Strafe_CanLocalStrafe}
     }
    }]
   properties {waitTime {defaultValue=1000}}
  }
 ]
 decorators []
 properties {
  children [
   {
    childComposite=&self:BTComposite_Sequence_1
    childTask=null
    decorators [
     &self:BTDecorator_Blackboard_6
     &self:BTDecorator_Loop_0
    ]
   }
   {
    childComposite=null
    childTask=&self:BTTask_Wait_2
    decorators [&self:BTDecorator_Blackboard_7]
   }
  ]
  services []
 }
 services []
}
