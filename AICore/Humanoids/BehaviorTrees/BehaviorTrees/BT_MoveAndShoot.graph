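# Humanoid AI behavior tree that repositions a bot to an evaluated destination while running ranged, throwable and ability attacks.
# Root services reserve the destination and manage combat tokens, gating the self-aborting branch that equips the desired weapon.
# Inside, services drive attack evaluators, weapon actions, skill set, movement style and abilities while the bot barks, moves with strafing, then runs BT_Strafe.
# BB_RangedAndAbility keys supply most tuning, making the tree reusable.
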BehaviorTree /AICore/Humanoids/BehaviorTrees/BehaviorTrees/BT_MoveAndShoot
blackboard=/AICore/Humanoids/BehaviorTrees/Blackboards/BB_RangedAndAbility
root {
 name=Sequence
 class=BTComposite_Sequence
 children [Sequence {
   class=BTComposite_Sequence
   children [
    Equip Item {
     class=FortAthenaBTTask_EquipItem
     children []
     decorators [Force Success {class=BTDecorator_ForceSuccess}]
     properties {
      bIgnoreRestartSelf=false
      equipItemKeyName=AIEvaluator_Injection_Item
      nodeName=Equip Item
      weaponKeyName=AIEvaluator_Global_DesiredWeapon
      disableEquipAnimation {defaultValue=false, key=None}
      itemTagToEquip {key=None, defaultValue {gameplayTags []}}
      waitForEquipToFinish {defaultValue=false, key=None}
     }
     services []
    }
    Selector {
     class=BTComposite_Selector
     children [Sequence {
       class=BTComposite_Sequence
       children [
        FortBTTask_Bark_1 {
         class=FortBTTask_Bark
         children []
         decorators [Force Success {class=BTDecorator_ForceSuccess}]
         properties {
          bIgnoreRestartSelf=false
          nodeName=""
          barkTag {bCanEditDefaultValueType=false, key=Ranged_RepositioningBarkTag, defaultValue {_structType=&GameplayTag}}
         }
         services []
        }
        Reposition {
         class=FortAthenaBTTask_BotMoveTo
         decorators []
         properties {
          executionStatusKeyName=AIEvaluator_TokenPosition_ExecutionStatus
          movementResultKeyName=AIEvaluator_TokenPosition_MovementState
          nodeName=Reposition
          acceptableRadius {defaultValue=5, key=Ranged_Reposition_AcceptableRadius}
          allowRandomWobble {defaultValue=true}
          bAllowPartialPath {defaultValue=false}
          bAllowStrafe {defaultValue=true}
          bReachTestIncludesAgentRadius {defaultValue=false}
          bReachTestIncludesGoalRadius {defaultValue=false}
          blackboardKey {selectedKeyName=AIEvaluator_Positioning_Destination}
          filterClass {key=Ranged_NavigationFilter}
          observedBlackboardValueTolerance {defaultValue=2.75}
         }
        }
        Run Behavior {
         class=BTTask_RunBehavior
         decorators []
         properties {behaviorAsset=&/BT_Strafe, nodeName=Run Behavior}
        }
       ]
       decorators [Is Destination Valid {
         class=BTDecorator_Blackboard
         properties {
          arithmeticOperation=Equal
          flowAbortMode=Self
          intValue=0
          nodeName=Is Destination Valid
          notifyObserver=ValueChange
          stringValue=""
          blackboardKey {selectedKeyName=AIEvaluator_Positioning_Destination}
         }
        }]
       properties {children [
         {
          childTask=&self:FortBTTask_Bark_1
          decorators [&self:BTDecorator_ForceSuccess_1]
         }
         {childTask=&self:FortAthenaBTTask_BotMoveTo_3}
         {childTask=&self:BTTask_RunBehavior_0}
        ]}
      }]
     decorators []
     properties {
      children [{
        childComposite=&self:BTComposite_Sequence_4
        childTask=null
        decorators [&self:BTDecorator_Blackboard_5]
       }]
      services [
       &self:FortAthenaBTService_ApplySkillSet_0
       &self:FortAthenaBTService_AIEvaluator_2
       &self:FortAthenaBTService_AIEvaluator_3
       &self:FortAthenaBTService_ManageWeapon_0
       &self:FortAthenaBTService_MovementStyle_1
       &self:FortBTService_ActivateAbility_4
       &self:FortAthenaBTService_WeaponPrimaryInputTacticalAction_0
      ]
     }
     services [
      Apply Skill Set {
       class=FortAthenaBTService_ApplySkillSet
       properties {
        bCallTickOnSearchStart=false
        bRestartTimerOnEachActivation=false
        interval=0.5
        nodeName=Apply Skill Set
        randomDeviation=0.1
        skillSetClass {baseClass=/Script/FortniteAI.FortAthenaAIBotSkillSet, defaultValue=null, key=Ranged_SkillsetOverride}
       }
      }
      Evaluating...Range Attack {
       class=FortAthenaBTService_AIEvaluator
       properties {interval=0.5, nodeName=Evaluating...Range Attack, aIEvaluatorClass {defaultValue=null, key=Ranged_RangeAttackEvaluator}}
      }
      Evaluating...Throwable {
       class=FortAthenaBTService_AIEvaluator
       properties {interval=0.5, nodeName=Evaluating...Throwable, aIEvaluatorClass {defaultValue=null, key=Ranged_ThrowableAttackEvaluator}}
      }
      Executing...Weapon Actions {class=FortAthenaBTService_ManageWeapon, properties {interval=0.5, nodeName=Executing...Weapon Actions, randomDeviation=0.1}}
      Modulate Movement Style {
       class=FortAthenaBTService_MovementStyle
       properties {
        bCallTickOnSearchStart=false
        bRestartTimerOnEachActivation=false
        interval=0.5
        nodeName=Modulate Movement Style
        randomDeviation=0.1
        blackboardCondition {bNoneIsAllowedValue=true, selectedKeyName=Ranged_ModulateMovementStyle}
        invertBlackboardCondition {defaultValue=false, key=None}
        newMovementStyle {defaultValue=0, enumType=/Script/FortniteGame.EFortMovementStyle, key=Ranged_MovementStyle, nativeEnumTypeName=EFortMovementStyle}
        tagCondition {autoDescription="", tokenStreamVersion=0, userDescription="", queryTokenStream [], tagDictionary []}
       }
      }
      Activate Ability {
       class=FortBTService_ActivateAbility
       properties {
        bCallTickOnSearchStart=false
        bPawnTargetsOnly=false
        bRequireCanHitTargetWithAbility=false
        bRestartTimerOnEachActivation=false
        bShowEventDetails=false
        bShowPropertyDetails=false
        customDescription=""
        interval=0.5
        nodeName=Activate Ability
        randomDeviation=0
        abilityTags {key=RangedAndAbility_AbilityTag, defaultValue {gameplayTags []}}
        activateOnBecomeRelevant {defaultValue=false, key=RangedAndAbility_ActivateAbilityOnBecomeRelevant}
        bCanActivateWhenMoving {defaultValue=true, key=None}
        deactivateOnCeaseRelevant {defaultValue=false, key=RangedAndAbility_DeactivateAbilityOnCeaseRelevant}
        distanceChecks []
        initialDelayRandomDeviation {defaultValue=0, key=None}
        initialDelayTime {defaultValue=0, key=None}
        prohibitedTargetTags {gameplayTags []}
        targetActorKey {bNoneIsAllowedValue=false, selectedKeyName=AIEvaluator_Global_TargetActor}
        wantsReceiveActivatedAbilityEnded {defaultValue=false, key=None}
       }
      }
      AI Bot Weapon Primary Input Tactical Action {
       class=FortAthenaBTService_WeaponPrimaryInputTacticalAction
       properties {
        bCallTickOnSearchStart=false
        bRestartTimerOnEachActivation=false
        interval=0.5
        nodeName=AI Bot Weapon Primary Input Tactical Action
        randomDeviation=0.1
        enabled {defaultValue=false, key=RangedAndAbility_WeaponPrimaryInputSetInTacticalAction}
       }
      }
     ]
    }
   ]
   decorators [
    Is Location Reserved {
     class=BTDecorator_Blackboard
     properties {
      arithmeticOperation=Equal
      flowAbortMode=Self
      intValue=0
      nodeName=Is Location Reserved
      stringValue=""
      blackboardKey {selectedKeyName=AIEValuator_LocationSynchronizer_HasReservedLocation}
     }
    }
    Has Token {
     class=FortBTDecorator_HasToken
     properties {
      bInverseCondition=false
      flowAbortMode=Self
      nodeName=Has Token
      activationTagQuery {
       bCanEditDefaultValueType=false
       key=Ranged_TokenActivationTagQuery
       defaultValue {
        _structType=/Script/GameplayTags.GameplayTagQuery
        autoDescription=""
        tokenStreamVersion=0
        userDescription=""
        queryTokenStream []
        tagDictionary []
       }
      }
      blackboardKey {bNoneIsAllowedValue=false, selectedKeyName=AIEvaluator_Global_TargetActor}
      tokenQuery {
       bCanEditDefaultValueType=false
       key=Ranged_TokenTagQuery
       defaultValue {
        _structType=&GameplayTagQuery
        autoDescription=""
        tokenStreamVersion=0
        userDescription=""
        queryTokenStream []
        tagDictionary []
       }
      }
     }
    }
   ]
   properties {children [
     {
      childTask=&self:FortAthenaBTTask_EquipItem_0
      decorators [&self:BTDecorator_ForceSuccess_0]
     }
     {childComposite=&self:BTComposite_Selector_2, childTask=null}
    ]}
  }]
 decorators []
 properties {
  children [{
    childComposite=&self:BTComposite_Sequence_2
    childTask=null
    decorators [
     &self:BTDecorator_Blackboard_1
     &self:FortBTDecorator_HasToken_0
    ]
   }]
  services [
   &self:FortAthenaBTService_LocationSynchronizerReservation_0
   &self:FortAthenaBTService_TokenQuery_0
   &self:FortAthenaBTService_SetInTacticalAction_0
  ]
 }
 services [
  Executing...Location Synchronizer Reservation {
   class=FortAthenaBTService_LocationSynchronizerReservation
   properties {
    bCallTickOnSearchStart=false
    bRestartTimerOnEachActivation=false
    interval=0.5
    nodeName=Executing...Location Synchronizer Reservation
    randomDeviation=0.1
    hasLocationReservedKey {bNoneIsAllowedValue=true, selectedKeyName=AIEValuator_LocationSynchronizer_HasReservedLocation}
    locationHardReservationBoundsRatio {
     bCanEditDefaultValueType=false
     key=Ranged_HardReservationScalingRatio
     defaultValue {_structType=/Script/GameplayAbilities.ScalableFloat, registryType=None, value=1, curve {curveTable=null, rowName=None}}
    }
    locationSoftReservationBoundsRatio {
     bCanEditDefaultValueType=false
     key=Ranged_SoftReservationScalingRatio
     defaultValue {_structType=&ScalableFloat, registryType=None, value=1, curve {curveTable=null, rowName=None}}
    }
    locationToReserveKey {key=AIEvaluator_Positioning_Destination, defaultValue=<0 0 0>}
   }
  }
  Executing...Token Query {
   class=FortAthenaBTService_TokenQuery
   properties {
    bCallTickOnSearchStart=false
    bRestartTimerOnEachActivation=false
    interval=0.5
    nodeName=Executing...Token Query
    randomDeviation=0.1
    targetActorName=AIEvaluator_Global_TargetActor
    weaponKeyName=AIEvaluator_Global_DesiredWeapon
    activationTagQuery {
     bCanEditDefaultValueType=false
     key=Ranged_TokenActivationTagQuery
     defaultValue {
      _structType=&GameplayTagQuery
      autoDescription=""
      tokenStreamVersion=0
      userDescription=""
      queryTokenStream []
      tagDictionary []
     }
    }
    becomeRelevantTokenAction {
     defaultValue=1
     enumType=/Script/FortniteAIServer.EFortAthenaBTServiceTokenQueryAction
     key=None
     nativeEnumTypeName=EFortAthenaBTServiceTokenQueryAction
    }
    ceaseRelevantTokenAction {
     defaultValue=2
     enumType=&EFortAthenaBTServiceTokenQueryAction
     key=None
     nativeEnumTypeName=EFortAthenaBTServiceTokenQueryAction
    }
    defaultTokenSystemTagQuery {
     bCanEditDefaultValueType=false
     key=Ranged_TokenTagQuery
     defaultValue {
      _structType=&GameplayTagQuery
      autoDescription=""
      tokenStreamVersion=0
      userDescription=""
      queryTokenStream []
      tagDictionary []
     }
    }
    equippedItemTagAssociationDatas []
    targetChangeTokenAction {
     defaultValue=1
     enumType=&EFortAthenaBTServiceTokenQueryAction
     key=None
     nativeEnumTypeName=EFortAthenaBTServiceTokenQueryAction
    }
    tokenReleaseDelay {defaultValue=0, key=Ranged_TokenReleaseDelay}
    weaponChangeTokenAction {
     defaultValue=1
     enumType=&EFortAthenaBTServiceTokenQueryAction
     key=None
     nativeEnumTypeName=EFortAthenaBTServiceTokenQueryAction
    }
   }
  }
  FortAthenaBTService_SetInTacticalAction_0 {
   class=FortAthenaBTService_SetInTacticalAction
   properties {
    shouldSetInTacticalAction {key=RangedAndAbility_EntireBehaviorSetInTacticalAction}
    tacticalActionIdentifierTag {defaultValue {_structType=&GameplayTag, tagName=AI.TacticalAction.MoveAndShoot.EntireBehavior}}
   }
  }
 ]
}
