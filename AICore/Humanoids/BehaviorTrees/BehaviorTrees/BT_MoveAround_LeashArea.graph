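# Behavior tree that makes a humanoid AI wander within its leash area.
# An EQS query, bounded by blackboard min/max distances from self, writes a random top-5% point to the patrol destination key, with Force Success so movement follows even if it fails.
# A selector then picks one of two bot MoveTo tasks to that key based on ShouldFocusOnTarget; both show identical settings.
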BehaviorTree /AICore/Humanoids/BehaviorTrees/BehaviorTrees/BT_MoveAround_LeashArea
blackboard=/AICore/Humanoids/BehaviorTrees/Blackboards/BB_MoveAround_LeashArea
root {
 name=Sequence
 class=BTComposite_Sequence
 children [
  Selector {
   class=BTComposite_Selector
   children [Find Wander Location {
     class=BTTask_RunEQSQuery
     properties {
      nodeName=Find Wander Location
      blackboardKey {selectedKeyName=AIEvaluator_Patrolling_Destination}
      eQSRequest {
       queryTemplate=/AICore/Humanoids/EQS/EQS_MoveAround_LeashArea
       runMode=RandomBest5Pct
       queryConfig [
        {
         bAllowBBKey=true
         paramName=MinDistanceSelf
         paramType=Float
         value=500
         bBKey {bNoneIsAllowedValue=true, selectedKeyName=MinDistanceFromSelf}
        }
        {
         bAllowBBKey=true
         paramName=MaxDistanceSelf
         paramType=Float
         value=2000
         bBKey {bNoneIsAllowedValue=true, selectedKeyName=MaxDistanceFromSelf}
        }
       ]
      }
     }
    }]
   decorators [Force Success {class=BTDecorator_ForceSuccess}]
   properties {
    children [{childComposite=null, childTask=&self:BTTask_RunEQSQuery_0, decorators []}]
    services []
   }
   services []
  }
  Selector {
   class=BTComposite_Selector
   children [
    "Move  To Location" {
     class=FortAthenaBTTask_BotMoveTo
     decorators [Should NOT Focus On Target {
       class=BTDecorator_Blackboard
       properties {
        arithmeticOperation=Equal
        basicOperation=NotSet
        flowAbortMode=None
        intValue=0
        nodeName=Should NOT Focus On Target
        stringValue=""
        blackboardKey {selectedKeyName=ShouldFocusOnTarget}
       }
      }]
     properties {
      movementResultKeyName=None
      nodeName="Move  To Location"
      acceptableRadius {defaultValue=50}
      allowRandomWobble {defaultValue=true}
      bAllowStrafe {defaultValue=false}
      blackboardKey {selectedKeyName=AIEvaluator_Patrolling_Destination}
      observedBlackboardValueTolerance {defaultValue=200}
     }
    }
    "Move  To Location" {
     class=FortAthenaBTTask_BotMoveTo
     decorators [Should Focus On Target {
       class=BTDecorator_Blackboard
       properties {
        arithmeticOperation=Equal
        flowAbortMode=None
        intValue=0
        nodeName=Should Focus On Target
        stringValue=""
        blackboardKey {selectedKeyName=ShouldFocusOnTarget}
       }
      }]
     properties {
      movementResultKeyName=None
      nodeName="Move  To Location"
      acceptableRadius {defaultValue=50}
      allowRandomWobble {defaultValue=true}
      bAllowStrafe {defaultValue=false}
      blackboardKey {selectedKeyName=AIEvaluator_Patrolling_Destination}
      observedBlackboardValueTolerance {defaultValue=200}
     }
    }
   ]
   decorators []
   properties {
    children [
     {
      childComposite=null
      childTask=&self:FortAthenaBTTask_BotMoveTo_0
      decorators [&self:BTDecorator_Blackboard_0]
     }
     {
      childComposite=null
      childTask=&self:FortAthenaBTTask_BotMoveTo_1
      decorators [&self:BTDecorator_Blackboard_4]
     }
    ]
    services []
   }
   services []
  }
 ]
 decorators []
 properties {children [
   {
    childComposite=&self:BTComposite_Selector_0
    childTask=null
    decorators [&self:BTDecorator_ForceSuccess_0]
   }
   {childComposite=&self:BTComposite_Selector_1, childTask=null}
  ]}
}
