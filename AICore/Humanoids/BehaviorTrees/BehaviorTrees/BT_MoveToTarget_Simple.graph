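# Simple humanoid AI behavior tree on BB_MoveAndMelee that moves a bot toward its target while evaluating melee attacks.
# Root services update target focus, movement style and weapon actions, while the child selector runs the melee attack evaluator.
# The selector repositions to the target actor with strafing, wobble, partial paths and a melee navigation filter, reporting through TokenPosition keys.
# When repositioning is unavailable or fails, the bot waits in place.
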BehaviorTree /AICore/Humanoids/BehaviorTrees/BehaviorTrees/BT_MoveToTarget_Simple
blackboard=/AICore/Humanoids/BehaviorTrees/Blackboards/BB_MoveAndMelee
root {
 name=Sequence
 class=BTComposite_Sequence
 children [Selector {
   class=BTComposite_Selector
   children [
    Reposition {
     class=FortAthenaBTTask_BotMoveTo
     decorators [Has Target {
       class=BTDecorator_Blackboard
       properties {
        arithmeticOperation=GreaterOrEqual
        intValue=3
        nodeName=Has Target
        stringValue=ExecutionPending
        blackboardKey {selectedKeyName=AIEvaluator_Global_TargetActor}
       }
      }]
     properties {
      bAllowPartialPathToUncompleteNavigationArea=true
      executionStatusKeyName=AIEvaluator_TokenPosition_ExecutionStatus
      movementResultKeyName=AIEvaluator_TokenPosition_MovementState
      nodeName=Reposition
      acceptableRadius {defaultValue=1}
      allowRandomWobble {defaultValue=true}
      bAllowStrafe {defaultValue=true}
      bProjectGoalLocation {defaultValue=false}
      bReachTestIncludesAgentRadius {defaultValue=false}
      bReachTestIncludesGoalRadius {defaultValue=false}
      bRequireNavigableEndLocation {defaultValue=false}
      blackboardKey {selectedKeyName=AIEvaluator_Global_TargetActor}
      filterClass {key=Melee_NavigationFilter}
      observedBlackboardValueTolerance {defaultValue=2.75}
     }
    }
    Stay in Position {class=BTTask_Wait, properties {nodeName=Stay in Position, waitTime {defaultValue=30}}}
   ]
   decorators []
   properties {
    children [
     {
      childComposite=null
      childTask=&self:FortAthenaBTTask_BotMoveTo_0
      decorators [&self:BTDecorator_Blackboard_0]
     }
     {childComposite=null, childTask=&self:BTTask_Wait_4, decorators []}
    ]
    services [&self:FortAthenaBTService_AIEvaluator_0]
   }
   services [Evaluating...Melee Attack {
     class=FortAthenaBTService_AIEvaluator
     properties {interval=0.1, nodeName=Evaluating...Melee Attack, randomDeviation=0, aIEvaluatorClass {defaultValue=null, key=Melee_MeleeAttackEvaluator}}
    }]
  }]
 decorators []
 properties {
  children [{childComposite=&self:BTComposite_Selector_3, childTask=null}]
  services [
   &self:FortAthenaBTService_AIEvaluator_2
   &self:FortAthenaBTService_MovementStyle_0
   &self:FortAthenaBTService_ManageWeapon_1
  ]
 }
 services [
  Focus on target {
   class=FortAthenaBTService_AIEvaluator
   properties {interval=0.5, nodeName=Focus on target, aIEvaluatorClass {defaultValue=null, key=Melee_FocusingEvaluator}}
  }
  FortPawn Movement Style {
   class=FortAthenaBTService_MovementStyle
   properties {
    bCallTickOnSearchStart=false
    bRestartTimerOnEachActivation=false
    interval=0.5
    nodeName=FortPawn Movement Style
    randomDeviation=0.1
    blackboardCondition {bNoneIsAllowedValue=true, selectedKeyName=None}
    invertBlackboardCondition {defaultValue=false, key=None}
    newMovementStyle {defaultValue=0, enumType=/Script/FortniteGame.EFortMovementStyle, key=Melee_MovementStyle, nativeEnumTypeName=EFortMovementStyle}
    tagCondition {autoDescription="", tokenStreamVersion=0, userDescription="", queryTokenStream [], tagDictionary []}
   }
  }
  Executing...Weapon Actions {class=FortAthenaBTService_ManageWeapon, properties {interval=0.5, nodeName=Executing...Weapon Actions, randomDeviation=0.1}}
 ]
}
