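# Behavior tree for humanoid AI unsynchronized movement, repositioning the bot to a token position or holding in place.
# Root services run blackboard-supplied token-position and focus evaluators and set movement style.
# When token status reaches at least ExecutionPending, a sequence runs a force-success bark, then a strafing, wobble-enabled move to the token destination without partial paths.
# Otherwise, or if that fails, the selector falls back to a blackboard-keyed wait.
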BehaviorTree /AICore/Humanoids/BehaviorTrees/BehaviorTrees/BT_MoveUnsynchronized
blackboard=/AICore/Humanoids/BehaviorTrees/Blackboards/BB_MoveUnsynchronized
root {
 name=Selector
 class=BTComposite_Selector
 children [
  Sequence {
   class=BTComposite_Sequence
   children [
    FortBTTask_Bark_1 {
     class=FortBTTask_Bark
     children []
     decorators [Force Success {class=BTDecorator_ForceSuccess}]
     properties {
      bIgnoreRestartSelf=false
      nodeName=""
      barkTag {bCanEditDefaultValueType=false, key=None, defaultValue {_structType=&GameplayTag}}
     }
     services []
    }
    Reposition {
     class=FortAthenaBTTask_BotMoveTo
     decorators []
     properties {
      executionStatusKeyName=AIEvaluator_TokenPosition_ExecutionStatus
      movementResultKeyName=AIEvaluator_TokenPosition_MovementState
      nodeName=Reposition
      acceptableRadius {defaultValue=1}
      allowRandomWobble {defaultValue=true}
      bAllowPartialPath {defaultValue=false}
      bAllowStrafe {defaultValue=true}
      bReachTestIncludesAgentRadius {defaultValue=false}
      bReachTestIncludesGoalRadius {defaultValue=false}
      blackboardKey {selectedKeyName=AIEvaluator_TokenPosition_Destination}
      filterClass {key=Move_NavigationFilter}
      observedBlackboardValueTolerance {defaultValue=2.75}
     }
    }
   ]
   decorators [Token Destination {
     class=BTDecorator_Blackboard
     properties {
      arithmeticOperation=GreaterOrEqual
      intValue=3
      nodeName=Token Destination
      stringValue=ExecutionPending
      blackboardKey {selectedKeyName=AIEvaluator_TokenPosition_ExecutionStatus}
     }
    }]
   properties {children [
     {
      childTask=&self:FortBTTask_Bark_1
      decorators [&self:BTDecorator_ForceSuccess_1]
     }
     {childTask=&self:FortAthenaBTTask_BotMoveTo_1}
    ]}
  }
  Stay in Position {class=BTTask_Wait, properties {nodeName=Stay in Position, waitTime {defaultValue=30, key=Move_FallbackWaitTime}}}
 ]
 decorators []
 properties {
  children [
   {
    childComposite=&self:BTComposite_Sequence_0
    childTask=null
    decorators [&self:BTDecorator_Blackboard_1]
   }
   {childComposite=null, childTask=&self:BTTask_Wait_0, decorators []}
  ]
  services [
   &self:FortAthenaBTService_AIEvaluator_1
   &self:FortAthenaBTService_MovementStyle_0
   &self:FortAthenaBTService_AIEvaluator_2
  ]
 }
 services [
  Focus on target {
   class=FortAthenaBTService_AIEvaluator
   properties {interval=0.5, nodeName=Focus on target, aIEvaluatorClass {defaultValue=null, key=Move_FocusingEvaluator}}
  }
  FortPawn Movement Style {
   class=FortAthenaBTService_MovementStyle
   properties {
    bCallTickOnSearchStart=false
    bRestartTimerOnEachActivation=false
    interval=0.5
    nodeName=FortPawn Movement Style
    randomDeviation=0.1
    blackboardCondition {bNoneIsAllowedValue=true, selectedKeyName=None}
    invertBlackboardCondition {defaultValue=false, key=None}
    newMovementStyle {defaultValue=0, enumType=/Script/FortniteGame.EFortMovementStyle, key=Move_MovementStyle, nativeEnumTypeName=EFortMovementStyle}
    tagCondition {autoDescription="", tokenStreamVersion=0, userDescription="", queryTokenStream [], tagDictionary []}
   }
  }
  Evaluating...Has Token Position {
   class=FortAthenaBTService_AIEvaluator
   properties {
    interval=0.1
    nodeName=Evaluating...Has Token Position
    randomDeviation=0
    aIEvaluatorClass {defaultValue=null, key=Move_TokenPositionEvaluator}
   }
  }
 ]
}
