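# Phoebe bot behavior tree that runs injected loot behavior, with a root service that periodically drives loot pickup through blackboard keys.
# While a loot destination is valid, a looped selector tries the unstuck subtree, then interaction when its status allows, then movement to the loot.
# The move runs in a Simple Parallel with defensive building, gated by its execution status, as the background task.
# A short Wait is the fallback.
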BehaviorTree /AICore/Humanoids/BehaviorTrees/BehaviorTrees/BT_Phoebe_Loot_Injection
blackboard=/AICore/Humanoids/BehaviorTrees/Blackboards/BB_Phoebe_Loot_Injection
root {
 name=Selector
 class=BTComposite_Selector
 children [
  Loot Selector {
   class=BTComposite_Selector
   children [
    "Run Unstuck Behavior " {
     class=BTTask_RunBehavior
     decorators []
     properties {behaviorAsset=&BT_Phoebe_UnstuckNew, nodeName="Run Unstuck Behavior "}
    }
    Loot Interaction {
     class=FortAthenaBTTask_Interact
     children []
     decorators [Can Interact {
       class=BTDecorator_Blackboard
       properties {
        arithmeticOperation=GreaterOrEqual
        intValue=3
        nodeName=Can Interact
        stringValue=ExecutionPending
        blackboardKey {selectedKeyName=AIEvaluator_Injection_Loot_InteractionExecutionStatus}
       }
      }]
     properties {
      attemptInterval=0.25
      bIgnoreRestartSelf=false
      bShouldFocusOnInteraction=true
      focalPointName=AIEvaluator_Global_FocalPoint
      interactActorName=AIEvaluator_Global_InteractActor
      jumpExecutionStatusName=AIEvaluator_Jump_ExecutionStatus
      maxInteractAttempts=5
      nodeName=Loot Interaction
      weaponTriggerMeleeName=AIEvaluator_WeaponTriggerMelee_ExecutionStatus
      interactContextInfoKeySelector {bNoneIsAllowedValue=false, selectedKeyName=AIEvaluator_Injection_Loot_InteractionContextInfo}
      interactExecutionStatusKeySelector {bNoneIsAllowedValue=false, selectedKeyName=AIEvaluator_Injection_Loot_InteractionExecutionStatus}
     }
     services []
    }
    Simple Parallel {
     class=BTComposite_SimpleParallel
     children [
      Loot Move To {
       class=FortAthenaBTTask_BotMoveTo
       decorators []
       properties {
        movementResultKeyName=AIEvaluator_Injection_Loot_MovementState
        noSmashMoveGoalActorKeyName=AIEvaluator_Injection_Loot_Actor
        nodeName=Loot Move To
        acceptableRadius {defaultValue=50}
        allowRandomWobble {defaultValue=false}
        bAllowStrafe {defaultValue=false}
        blackboardKey {selectedKeyName=AIEvaluator_Injection_Loot_Destination}
        observedBlackboardValueTolerance {defaultValue=4.75}
       }
      }
      Build Construction {
       class=FortAthenaBTTask_Build
       children []
       decorators [Blackboard Based Condition {
         class=BTDecorator_Blackboard
         properties {
          arithmeticOperation=GreaterOrEqual
          flowAbortMode=None
          intValue=4
          nodeName=Blackboard Based Condition
          stringValue=EExecutionStatus::ExecutionAllowed
          blackboardKey {selectedKeyName=AIEvaluator_DefensiveBuilding_ExecutionStatus}
         }
        }]
       properties {
        bIgnoreRestartSelf=false
        executionStatusKeyName=AIEvaluator_DefensiveBuilding_ExecutionStatus
        focalPointName=AIEvaluator_Global_FocalPoint
        nodeName=Build Construction
       }
       services []
      }
     ]
     properties {
      children [
       {childTask=&self:FortAthenaBTTask_BotMoveTo_9}
       {
        childTask=&self:FortAthenaBTTask_Build_0
        decorators [&self:BTDecorator_Blackboard_1]
       }
      ]
      services []
     }
     services []
    }
   ]
   decorators [
    Has Valid Loot Dest {
     class=BTDecorator_Blackboard
     properties {
      arithmeticOperation=Equal
      intValue=0
      nodeName=Has Valid Loot Dest
      stringValue=""
      blackboardKey {selectedKeyName=AIEvaluator_Injection_Loot_Destination}
     }
    }
    Loop when blocked {class=BTDecorator_Loop, properties {nodeName=Loop when blocked}}
   ]
   properties {
    nodeName=Loot Selector
    children [
     {childComposite=null, childTask=&self:BTTask_RunBehavior_0, decorators []}
     {
      childComposite=null
      childTask=&self:FortAthenaBTTask_Interact_0
      decorators [&self:BTDecorator_Blackboard_26]
     }
     {
      childComposite=&self:BTComposite_SimpleParallel_0
      childTask=null
      decorators []
     }
    ]
    services []
   }
   services []
  }
  Wait {class=BTTask_Wait, properties {waitTime {defaultValue=0.7}}}
 ]
 decorators []
 properties {
  children [
   {
    childComposite=&self:BTComposite_Selector_23
    childTask=null
    decorators [
     &self:BTDecorator_Blackboard_0
     &self:BTDecorator_Loop_0
    ]
   }
   {childComposite=null, childTask=&self:BTTask_Wait_0, decorators []}
  ]
  services [&self:FortAthenaBTService_PickUpLoot_1]
 }
 services [Executing...Pick Up Loot {
   class=FortAthenaBTService_PickUpLoot
   properties {
    bCallTickOnSearchStart=false
    bRestartTimerOnEachActivation=false
    executionStatusName=AIEvaluator_Injection_Loot_ExecutionStatus
    interactionContextInfoName=AIEvaluator_Injection_Loot_InteractionContextInfo
    interactionExecutionStatusName=AIEvaluator_Injection_Loot_InteractionExecutionStatus
    interval=0.5
    lootObjectKeyName=AIEvaluator_Injection_Loot_Actor
    movementStateKeyName=AIEvaluator_Injection_Loot_MovementState
    nodeName=Executing...Pick Up Loot
    randomDeviation=0.1
   }
  }]
}
