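# Reusable AICore humanoid behavior tree that sends a gameplay event, then optionally pauses.
# A Sequence runs the send-event task with a tag from the GameplayEvent_EventTag blackboard key, then a Wait if that task succeeds.
# The Wait takes its duration and random deviation from blackboard keys and runs only when the pause time is above zero.
# Force Success on the Wait keeps a skipped pause from failing the sequence.
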BehaviorTree /AICore/Humanoids/BehaviorTrees/BehaviorTrees/BT_SendGameplayEvent
blackboard=/AICore/Humanoids/BehaviorTrees/Blackboards/BB_SendGameplayEvent
root {
 name=Sequence
 class=BTComposite_Sequence
 children [
  BTT_SendGameplayEvent_C_0 {
   class=BTT_SendGameplayEvent_C
   children []
   decorators []
   properties {
    bIgnoreRestartSelf=false
    bShowPropertyDetails=true
    customDescription=""
    nodeName=""
    taskSucceeded=false
    gameplayEventTag {key=GameplayEvent_EventTag, defaultValue {gameplayTags []}}
    tickInterval {interval=-1}
   }
   services []
  }
  Wait {
   class=BTTask_Wait
   decorators [
    Force Success {class=BTDecorator_ForceSuccess}
    Should Pause After Success {
     class=BTDecorator_Blackboard
     properties {
      arithmeticOperation=Greater
      flowAbortMode=None
      intValue=0
      nodeName=Should Pause After Success
      stringValue=""
      blackboardKey {selectedKeyName=GameplayEvent_PauseTimeAfterSuccess}
     }
    }
   ]
   properties {randomDeviation {key=GameplayEvent_PauseTimeDeviationAfterSuccess}, waitTime {defaultValue=5, key=GameplayEvent_PauseTimeAfterSuccess}}
  }
 ]
 decorators []
 properties {children [
   {childTask=&self:BTT_SendGameplayEvent_C_0}
   {
    childTask=&self:BTTask_Wait_2
    decorators [
     &self:BTDecorator_ForceSuccess_9
     &self:BTDecorator_Blackboard_0
    ]
   }
  ]}
}
