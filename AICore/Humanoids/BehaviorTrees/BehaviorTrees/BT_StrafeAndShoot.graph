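# Behavior tree for humanoid AI that equips its desired weapon, then repositions by strafing while fighting its target at range.
# A root sequence runs a force-succeeded equip task, then a combat sequence that plays an optional repositioning bark and runs the BT_Strafe subtree.
# Services on that sequence apply a skill set, evaluate ranged and throwable attacks, manage weapon actions, movement style and focus, and activate abilities, mostly configured through BB_RangedAndAbility keys.
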BehaviorTree /AICore/Humanoids/BehaviorTrees/BehaviorTrees/BT_StrafeAndShoot
blackboard=/AICore/Humanoids/BehaviorTrees/Blackboards/BB_RangedAndAbility
root {
 name=Sequence
 class=BTComposite_Sequence
 children [
  Equip Item {
   class=FortAthenaBTTask_EquipItem
   children []
   decorators [Force Success {class=BTDecorator_ForceSuccess}]
   properties {
    bIgnoreRestartSelf=false
    equipItemKeyName=AIEvaluator_Injection_Item
    nodeName=Equip Item
    weaponKeyName=AIEvaluator_Global_DesiredWeapon
    disableEquipAnimation {defaultValue=false, key=None}
    itemTagToEquip {key=None, defaultValue {gameplayTags []}}
    waitForEquipToFinish {defaultValue=false, key=None}
   }
   services []
  }
  Sequence {
   class=BTComposite_Sequence
   children [
    FortBTTask_Bark_0 {
     class=FortBTTask_Bark
     children []
     decorators [Force Success {class=BTDecorator_ForceSuccess}]
     properties {
      bIgnoreRestartSelf=false
      nodeName=""
      barkTag {bCanEditDefaultValueType=false, key=Ranged_RepositioningBarkTag, defaultValue {_structType=&GameplayTag}}
     }
     services []
    }
    Run Behavior {
     class=BTTask_RunBehavior
     decorators []
     properties {behaviorAsset=&/BT_Strafe, nodeName=Run Behavior}
    }
   ]
   decorators []
   properties {
    children [
     {
      childTask=&self:FortBTTask_Bark_0
      decorators [&self:BTDecorator_ForceSuccess_0]
     }
     {childTask=&self:BTTask_RunBehavior_1}
    ]
    services [
     &self:FortAthenaBTService_ApplySkillSet_1
     &self:FortAthenaBTService_AIEvaluator_2
     &self:FortAthenaBTService_AIEvaluator_3
     &self:FortAthenaBTService_ManageWeapon_0
     &self:FortAthenaBTService_MovementStyle_1
     &self:FortAthenaBTService_AIEvaluator_4
     &self:FortBTService_ActivateAbility_4
    ]
   }
   services [
    Apply Skill Set {
     class=FortAthenaBTService_ApplySkillSet
     properties {
      bCallTickOnSearchStart=false
      bRestartTimerOnEachActivation=false
      interval=0.5
      nodeName=Apply Skill Set
      randomDeviation=0.1
      skillSetClass {baseClass=/Script/FortniteAI.FortAthenaAIBotSkillSet, defaultValue=null, key=Ranged_SkillsetOverride}
     }
    }
    Evaluating...Range Attack {
     class=FortAthenaBTService_AIEvaluator
     properties {interval=0.5, nodeName=Evaluating...Range Attack, aIEvaluatorClass {defaultValue=null, key=Ranged_RangeAttackEvaluator}}
    }
    Evaluating...Throwable {
     class=FortAthenaBTService_AIEvaluator
     properties {interval=0.5, nodeName=Evaluating...Throwable, aIEvaluatorClass {defaultValue=null, key=Ranged_ThrowableAttackEvaluator}}
    }
    Executing...Weapon Actions {class=FortAthenaBTService_ManageWeapon, properties {interval=0.5, nodeName=Executing...Weapon Actions, randomDeviation=0.1}}
    Modulate Movement Style {
     class=FortAthenaBTService_MovementStyle
     properties {
      bCallTickOnSearchStart=false
      bRestartTimerOnEachActivation=false
      interval=0.5
      nodeName=Modulate Movement Style
      randomDeviation=0.1
      blackboardCondition {bNoneIsAllowedValue=true, selectedKeyName=Ranged_ModulateMovementStyle}
      invertBlackboardCondition {defaultValue=false, key=None}
      newMovementStyle {defaultValue=0, enumType=/Script/FortniteGame.EFortMovementStyle, key=Ranged_MovementStyle, nativeEnumTypeName=EFortMovementStyle}
      tagCondition {autoDescription="", tokenStreamVersion=0, userDescription="", queryTokenStream [], tagDictionary []}
     }
    }
    Focus on target {
     class=FortAthenaBTService_AIEvaluator
     properties {interval=0.5, nodeName=Focus on target, aIEvaluatorClass {defaultValue=&FortAthenaAIBotEvaluator_HandleFocusing}}
    }
    Activate Ability {
     class=FortBTService_ActivateAbility
     properties {
      bCallTickOnSearchStart=false
      bPawnTargetsOnly=false
      bRequireCanHitTargetWithAbility=false
      bRestartTimerOnEachActivation=false
      bShowEventDetails=false
      bShowPropertyDetails=false
      customDescription=""
      interval=0.5
      nodeName=Activate Ability
      randomDeviation=0
      abilityTags {key=RangedAndAbility_AbilityTag, defaultValue {gameplayTags []}}
      activateOnBecomeRelevant {defaultValue=false, key=RangedAndAbility_ActivateAbilityOnBecomeRelevant}
      bCanActivateWhenMoving {defaultValue=true, key=None}
      deactivateOnCeaseRelevant {defaultValue=false, key=RangedAndAbility_DeactivateAbilityOnCeaseRelevant}
      distanceChecks []
      initialDelayRandomDeviation {defaultValue=0, key=None}
      initialDelayTime {defaultValue=0, key=None}
      prohibitedTargetTags {gameplayTags []}
      targetActorKey {bNoneIsAllowedValue=false, selectedKeyName=AIEvaluator_Global_TargetActor}
      wantsReceiveActivatedAbilityEnded {defaultValue=false, key=None}
     }
    }
   ]
  }
 ]
 decorators []
 properties {children [
   {
    childTask=&self:FortAthenaBTTask_EquipItem_0
    decorators [&self:BTDecorator_ForceSuccess_1]
   }
   {childComposite=&self:BTComposite_Sequence_1, childTask=null}
  ]}
}
