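# Shared behavior tree subtree that makes a humanoid AI use a gameplay ability, configured through BB_UseAbility blackboard keys.
# A sequence runs optional item equip, pre-use pause, ability trigger, post-success pause and second equip, under tactical-action and crouch services.
# The trigger step checks ability possession and activation, and loops by repeat count with optional pauses between.
# Force Success keeps optional steps from failing the sequence, but a failed ability step skips the rest.
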BehaviorTree /AICore/Humanoids/BehaviorTrees/BehaviorTrees/BT_UseAbility_Subtree_SharedImpl
blackboard=/AICore/Humanoids/BehaviorTrees/Blackboards/BB_UseAbility
root {
 name=Sequence
 class=BTComposite_Sequence
 children [
  Equip Item {
   class=FortAthenaBTTask_EquipItem
   children []
   decorators [
    Force Success {class=BTDecorator_ForceSuccess}
    Should Equip Item {
     class=BTDecorator_Blackboard
     properties {
      arithmeticOperation=Equal
      flowAbortMode=None
      intValue=0
      nodeName=Should Equip Item
      stringValue=""
      blackboardKey {selectedKeyName=Ability_ShouldEquipItem}
     }
    }
   ]
   properties {
    bIgnoreRestartSelf=false
    equipItemKeyName=AIEvaluator_Injection_Item
    nodeName=Equip Item
    weaponKeyName=AIEvaluator_Global_DesiredWeapon
    disableEquipAnimation {defaultValue=false, key=Ability_DisableItemEquipAnimationAtStart}
    itemTagToEquip {key=None, defaultValue {gameplayTags []}}
    waitForEquipToFinish {defaultValue=false, key=Ability_WaitForItemEquipAtStartToFinish}
   }
   services []
  }
  Wait {
   class=BTTask_Wait
   decorators [
    Force Success {class=BTDecorator_ForceSuccess}
    Should Pause Before Use Ability {
     class=BTDecorator_Blackboard
     properties {
      arithmeticOperation=Greater
      flowAbortMode=None
      intValue=0
      nodeName=Should Pause Before Use Ability
      stringValue=""
      blackboardKey {selectedKeyName=Ability_PauseTimeBeforeUseAbility}
     }
    }
   ]
   properties {waitTime {defaultValue=5, key=Ability_PauseTimeBeforeUseAbility}}
  }
  Sequence {
   class=BTComposite_Sequence
   children [
    TriggerAbility {
     class=FortBTTask_ExecuteGameplayAbility
     decorators []
     properties {nodeName=TriggerAbility, gameplayAbilityTagContainer {key=Ability_TagContainer, defaultValue {gameplayTags []}}}
    }
    Wait {
     class=BTTask_Wait
     decorators [
      Force Success {class=BTDecorator_ForceSuccess}
      Wants To Loop? {
       class=BTDecorator_Blackboard
       properties {
        arithmeticOperation=Greater
        flowAbortMode=None
        intValue=1
        nodeName=Wants To Loop?
        stringValue=""
        blackboardKey {selectedKeyName=Ability_NumberOfRepeats}
       }
      }
      Pause Before Repeat? {
       class=BTDecorator_Blackboard
       properties {
        arithmeticOperation=Greater
        flowAbortMode=None
        intValue=0
        nodeName=Pause Before Repeat?
        stringValue=""
        blackboardKey {selectedKeyName=Ability_PauseTimeBeforeRepeat}
       }
      }
     ]
     properties {waitTime {defaultValue=5, key=Ability_PauseTimeBeforeRepeat}}
    }
   ]
   decorators [
    Has Gameplay Ability {
     class=FortBTDecorator_GameplayAbility_HasGameplayAbility
     properties {
      gameplayAbilityTagContainer {key=Ability_TagContainer, defaultValue {gameplayTags []}}
      onlyTestActiveAbility {key=Ability_OnlyTestActiveAbility}
     }
    }
    Can Activate Gameplay Ability {
     class=FortBTDecorator_GameplayAbility_CanActivate
     properties {
      flowAbortMode=None
      activeAbilityTagsToSkipTestingValue {key=Ability_ActiveAbilityTagsToSkipTesting}
      gameplayAbilityTagContainer {key=Ability_TagContainer, defaultValue {gameplayTags []}}
      targetValue {key=SelfActor}
     }
    }
    Loop {class=BTDecorator_Loop, properties {bInfiniteLoop=false, numLoops {defaultValue=1, key=Ability_NumberOfRepeats}}}
   ]
   properties {children [
     {childTask=&self:FortBTTask_ExecuteGameplayAbility_1}
     {
      childTask=&self:BTTask_Wait_0
      decorators [
       &self:BTDecorator_ForceSuccess_2
       &self:BTDecorator_Blackboard_2
       &self:BTDecorator_Blackboard_3
      ]
     }
    ]}
  }
  Wait {
   class=BTTask_Wait
   decorators [
    Force Success {class=BTDecorator_ForceSuccess}
    Should Pause After Success {
     class=BTDecorator_Blackboard
     properties {
      arithmeticOperation=Greater
      flowAbortMode=None
      intValue=0
      nodeName=Should Pause After Success
      stringValue=""
      blackboardKey {selectedKeyName=Ability_PauseTimeAfterSuccess}
     }
    }
   ]
   properties {waitTime {defaultValue=5, key=Ability_PauseTimeAfterSuccess}}
  }
  Equip Item {
   class=FortAthenaBTTask_EquipItem
   children []
   decorators [
    Force Success {class=BTDecorator_ForceSuccess}
    Should Equip Item {
     class=BTDecorator_Blackboard
     properties {
      arithmeticOperation=Equal
      flowAbortMode=None
      intValue=0
      nodeName=Should Equip Item
      stringValue=""
      blackboardKey {selectedKeyName=Ability_ShouldEquipItem}
     }
    }
   ]
   properties {
    bIgnoreRestartSelf=false
    equipItemKeyName=AIEvaluator_Injection_Item
    nodeName=Equip Item
    weaponKeyName=AIEvaluator_Global_DesiredWeapon
    disableEquipAnimation {defaultValue=false, key=Ability_DisableItemEquipAnimationAtEnd}
    itemTagToEquip {key=None, defaultValue {gameplayTags []}}
    waitForEquipToFinish {defaultValue=true, key=Ability_WaitForItemEquipAtEndToFinish}
   }
   services []
  }
 ]
 decorators []
 properties {
  children [
   {
    childTask=&self:FortAthenaBTTask_EquipItem_1
    decorators [
     &self:BTDecorator_ForceSuccess_1
     &self:BTDecorator_Blackboard_1
    ]
   }
   {
    childTask=&self:BTTask_Wait_1
    decorators [
     &self:BTDecorator_ForceSuccess_9
     &self:BTDecorator_Blackboard_18
    ]
   }
   {
    childComposite=&self:BTComposite_Sequence_1
    childTask=null
    decorators [
     &self:FortBTDecorator_GameplayAbility_HasGameplayAbility_0
     &self:FortBTDecorator_GameplayAbility_CanActivate_0
     &self:BTDecorator_Loop_1
    ]
   }
   {
    childTask=&self:BTTask_Wait_6
    decorators [
     &self:BTDecorator_ForceSuccess_3
     &self:BTDecorator_Blackboard_4
    ]
   }
   {
    childTask=&self:FortAthenaBTTask_EquipItem_0
    decorators [
     &self:BTDecorator_ForceSuccess_0
     &self:BTDecorator_Blackboard_0
    ]
   }
  ]
  services [
   &self:FortAthenaBTService_SetInTacticalAction_0
   &self:FortBTService_Crouch_0
  ]
 }
 services [
  FortAthenaBTService_SetInTacticalAction_0 {
   class=FortAthenaBTService_SetInTacticalAction
   properties {
    shouldSetInTacticalAction {key=Ability_TacticalAction}
    tacticalActionIdentifierTag {defaultValue {_structType=&GameplayTag}}
   }
  }
  FortBTService_Crouch_0 {
   class=FortBTService_Crouch
   properties {
    bCallTickOnSearchStart=false
    bRestartTimerOnEachActivation=false
    interval=0.5
    nodeName=""
    randomDeviation=0.1
    enterShouldCrouch {defaultValue=false, key=Ability_ShouldCrouchOnEnter}
    exitShouldCrouch {defaultValue=false, key=Ability_ShouldCrouchOnExit}
   }
  }
 ]
}
