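# BT_UseAbility is a humanoid AI behavior tree that runs ability use through BT_UseAbility_Subtree_SharedImpl subtrees, using the BB_UseAbility blackboard.
# When Ability_ShouldFocusOnTarget passes, a nested selector picks a subtree run by whether SelfActor is a FortPlayerPawn or FortAIPawn.
# Otherwise, or if that branch fails, the root selector falls back to an unconditional subtree run.
# A root service suppresses goal updates, controlled by the Ability_SuppressGoalUpdates key.
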BehaviorTree /AICore/Humanoids/BehaviorTrees/BehaviorTrees/BT_UseAbility
blackboard=/AICore/Humanoids/BehaviorTrees/Blackboards/BB_UseAbility
root {
 name=Selector
 class=BTComposite_Selector
 children [
  Selector {
   class=BTComposite_Selector
   children [
    Run Behavior {
     class=BTTask_RunBehavior
     decorators [FortPlayerPawn? {
       class=BTDecorator_IsBBEntryOfClass
       properties {
        bInverseCondition=false
        flowAbortMode=None
        nodeName=FortPlayerPawn?
        blackboardKey {bNoneIsAllowedValue=false, selectedKeyName=SelfActor}
        testClass {baseClass=&Object, defaultValue=/Script/FortniteGame.FortPlayerPawn, key=None}
       }
      }]
     properties {behaviorAsset=&/BT_UseAbility_Subtree_SharedImpl, nodeName=Run Behavior}
    }
    Run Behavior {
     class=BTTask_RunBehavior
     decorators [FortAIPawn? {
       class=BTDecorator_IsBBEntryOfClass
       properties {
        bInverseCondition=false
        flowAbortMode=None
        nodeName=FortAIPawn?
        blackboardKey {bNoneIsAllowedValue=false, selectedKeyName=SelfActor}
        testClass {baseClass=&Object, defaultValue=/Script/FortniteGame.FortAIPawn, key=None}
       }
      }]
     properties {behaviorAsset=&BT_UseAbility_Subtree_SharedImpl, nodeName=Run Behavior}
    }
   ]
   decorators [Focus On Target? {
     class=BTDecorator_Blackboard
     properties {
      arithmeticOperation=Equal
      flowAbortMode=None
      intValue=0
      nodeName=Focus On Target?
      stringValue=""
      blackboardKey {selectedKeyName=Ability_ShouldFocusOnTarget}
     }
    }]
   properties {
    children [
     {
      childComposite=null
      childTask=&self:BTTask_RunBehavior_0
      decorators [&self:BTDecorator_IsBBEntryOfClass_4]
     }
     {
      childComposite=null
      childTask=&self:BTTask_RunBehavior_2
      decorators [&self:BTDecorator_IsBBEntryOfClass_5]
     }
    ]
    services []
   }
   services []
  }
  Run Behavior {
   class=BTTask_RunBehavior
   decorators []
   properties {behaviorAsset=&BT_UseAbility_Subtree_SharedImpl, nodeName=Run Behavior}
  }
 ]
 decorators []
 properties {
  children [
   {
    childComposite=&self:BTComposite_Selector_1
    childTask=null
    decorators [&self:BTDecorator_Blackboard_1]
   }
   {childComposite=null, childTask=&self:BTTask_RunBehavior_1, decorators []}
  ]
  services [&self:FortBTContext_SuppressGoalUpdate_0]
 }
 services [Suppress Goal Updates {
   class=FortBTContext_SuppressGoalUpdate
   properties {
    bCallTickOnSearchStart=false
    bRestartTimerOnEachActivation=false
    bUnregisterFromGoalManager=false
    interval=0.5
    nodeName=Suppress Goal Updates
    randomDeviation=0.1
    enabled {defaultValue=false, key=Ability_SuppressGoalUpdates}
   }
  }]
}
