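# Behavior tree that makes a humanoid AI bot use an item, configured through Item_* keys on the BB_UseItem blackboard.
# Root services handle focusing and query tokens on the target actor, and a HasToken decorator gates the sequence, aborting it if the condition fails.
# The sequence rolls a random use count, optionally equips, moves and plays a speech bubble via selectors with always-succeed fallbacks, then loops use and wait.
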BehaviorTree /AICore/Humanoids/BehaviorTrees/BehaviorTrees/BT_UseItem
blackboard=/AICore/Humanoids/BehaviorTrees/Blackboards/BB_UseItem
root {
 name=Selector
 class=BTComposite_Selector
 children [Sequence {
   class=BTComposite_Sequence
   children [
    Determine Use Count {
     class=FortBTTask_RandomInt
     children []
     decorators []
     properties {
      bIgnoreRestartSelf=false
      nodeName=Determine Use Count
      blackboardKey {bNoneIsAllowedValue=false, selectedKeyName=Item_UseCount}
      maximumValue {defaultValue=0, key=Item_MaxUseCount}
      minimumValue {defaultValue=0, key=Item_MinUseCount}
     }
     services []
    }
    ShouldEquipItem {
     class=BTComposite_Selector
     children [
      Equip Item {
       class=FortAthenaBTTask_EquipItem
       children []
       decorators [Blackboard Based Condition {
         class=BTDecorator_Blackboard
         properties {
          arithmeticOperation=Equal
          flowAbortMode=Self
          intValue=0
          nodeName=Blackboard Based Condition
          stringValue=""
          blackboardKey {selectedKeyName=Item_ShouldEquipBeforeMove}
         }
        }]
       properties {
        bIgnoreRestartSelf=false
        equipItemKeyName=AIEvaluator_Injection_Item
        nodeName=Equip Item
        weaponKeyName=AIEvaluator_Global_DesiredWeapon
        disableEquipAnimation {defaultValue=false, key=None}
        itemTagToEquip {key=None, defaultValue {gameplayTags []}}
        waitForEquipToFinish {defaultValue=false, key=None}
       }
       services []
      }
      BTTask_AlwaysSuccess {
       class=FortBTTask_AlwaysSucceed
       children []
       decorators [Blackboard Based Condition {
         class=BTDecorator_Blackboard
         properties {
          arithmeticOperation=Equal
          basicOperation=NotSet
          flowAbortMode=Self
          intValue=0
          nodeName=Blackboard Based Condition
          stringValue=""
          blackboardKey {selectedKeyName=Item_ShouldEquipBeforeMove}
         }
        }]
       properties {bIgnoreRestartSelf=false, nodeName=BTTask_AlwaysSuccess}
       services []
      }
     ]
     decorators []
     properties {
      nodeName=ShouldEquipItem
      children [
       {
        childComposite=null
        childTask=&self:FortAthenaBTTask_EquipItem_0
        decorators [&self:BTDecorator_Blackboard_2]
       }
       {
        childComposite=null
        childTask=&self:BTTask_AlwaysSuccess_C_0
        decorators [&self:BTDecorator_Blackboard_3]
       }
      ]
      services []
     }
     services []
    }
    ShouldMoveToTarget {
     class=BTComposite_Selector
     children [
      Move To {
       class=BTTask_MoveTo
       children []
       decorators [Blackboard Based Condition {
         class=BTDecorator_Blackboard
         properties {
          arithmeticOperation=Equal
          flowAbortMode=Self
          intValue=0
          nodeName=Blackboard Based Condition
          stringValue=""
          blackboardKey {selectedKeyName=Item_ShouldMoveToTarget}
         }
        }]
       properties {
        bIgnoreRestartSelf=false
        nodeName=Move To
        acceptableRadius {defaultValue=150, key=None}
        bAllowPartialPath {defaultValue=true, key=None}
        bAllowStrafe {defaultValue=false, key=None}
        bProjectGoalLocation {defaultValue=true, key=None}
        bReachTestIncludesAgentRadius {defaultValue=true, key=None}
        bReachTestIncludesGoalRadius {defaultValue=true, key=None}
        bRequireNavigableEndLocation {defaultValue=true, key=None}
        bStartFromPreviousPath {defaultValue=false, key=None}
        bTrackMovingGoal {defaultValue=true, key=None}
        blackboardKey {bNoneIsAllowedValue=false, selectedKeyName=AIEvaluator_Global_TargetActor}
        filterClass {baseClass=&NavigationQueryFilter, defaultValue=null, key=None}
        observedBlackboardValueTolerance {defaultValue=4.75, key=None}
       }
       services []
      }
      BTTask_AlwaysSuccess {
       class=FortBTTask_AlwaysSucceed
       children []
       decorators [Blackboard Based Condition {
         class=BTDecorator_Blackboard
         properties {
          arithmeticOperation=Equal
          basicOperation=NotSet
          flowAbortMode=Self
          intValue=0
          nodeName=Blackboard Based Condition
          stringValue=""
          blackboardKey {selectedKeyName=Item_ShouldMoveToTarget}
         }
        }]
       properties {bIgnoreRestartSelf=false, nodeName=BTTask_AlwaysSuccess}
       services []
      }
     ]
     decorators []
     properties {
      nodeName=ShouldMoveToTarget
      children [
       {
        childComposite=null
        childTask=&self:BTTask_MoveTo_0
        decorators [&self:BTDecorator_Blackboard_6]
       }
       {
        childComposite=null
        childTask=&self:BTTask_AlwaysSuccess_C_1
        decorators [&self:BTDecorator_Blackboard_1]
       }
      ]
      services []
     }
     services []
    }
    ShouldTriggerSpeechBubble {
     class=BTComposite_Selector
     children [
      BTT_PlaySpeechBubble_C_0 {
       class=BTT_PlaySpeechBubble_C
       children []
       decorators [Blackboard Based Condition {
         class=BTDecorator_Blackboard
         properties {
          arithmeticOperation=Equal
          flowAbortMode=Self
          intValue=0
          nodeName=Blackboard Based Condition
          stringValue=""
          blackboardKey {selectedKeyName=Item_ShouldTriggerSpeechBubble}
         }
        }]
       properties {
        bIgnoreRestartSelf=false
        bShowPropertyDetails=true
        customDescription=""
        nodeName=""
        speechBubbleTagKey {bNoneIsAllowedValue=false, selectedKeyName=Item_SpeechBubbleTag}
        tickInterval {interval=-1}
       }
       services []
      }
      BTTask_AlwaysSuccess {
       class=FortBTTask_AlwaysSucceed
       children []
       decorators [Blackboard Based Condition {
         class=BTDecorator_Blackboard
         properties {
          arithmeticOperation=Equal
          basicOperation=NotSet
          flowAbortMode=Self
          intValue=0
          nodeName=Blackboard Based Condition
          stringValue=""
          blackboardKey {selectedKeyName=Item_ShouldTriggerSpeechBubble}
         }
        }]
       properties {bIgnoreRestartSelf=false, nodeName=BTTask_AlwaysSuccess}
       services []
      }
     ]
     decorators []
     properties {
      nodeName=ShouldTriggerSpeechBubble
      children [
       {
        childComposite=null
        childTask=&self:BTT_PlaySpeechBubble_C_0
        decorators [&self:BTDecorator_Blackboard_8]
       }
       {
        childComposite=null
        childTask=&self:FortBTTask_AlwaysSucceed_0
        decorators [&self:BTDecorator_Blackboard_9]
       }
      ]
      services []
     }
     services []
    }
    Sequence {
     class=BTComposite_Sequence
     children [
      Use Item {
       class=FortAthenaBTTask_UseItem
       decorators []
       properties {
        actionObjectKeyName=AIEvaluator_Injection_Item
        executionStatusKeyName=None
        nodeName=Use Item
        finishAfterItemUse {defaultValue=true}
        inputHoldDuration {key=Item_InputHoldDuration}
        maximumWaitTimeBetweenUses {defaultValue=0}
        resetActionObjectKey {defaultValue=false}
        useAlternateMode {key=Item_UseAlternateFire}
        validateAbility {key=Item_ShouldValidateAbility}
        validateAbilityBeforeEquipping {defaultValue=false}
       }
      }
      Wait {class=BTTask_Wait, properties {randomDeviation {key=Item_WaitTimeAfterUseDeviation}, waitTime {defaultValue=5, key=Item_WaitTimeAfterUse}}}
     ]
     decorators [Loop {class=BTDecorator_Loop, properties {bInfiniteLoop=false, numLoops {key=Item_UseCount}}}]
     properties {children [
       {childTask=&self:FortAthenaBTTask_UseItem_0}
       {childTask=&self:BTTask_Wait_0}
      ]}
    }
   ]
   decorators [HasToken {
     class=FortBTDecorator_HasToken
     properties {
      bInverseCondition=false
      flowAbortMode=Self
      nodeName=HasToken
      activationTagQuery {
       bCanEditDefaultValueType=false
       key=None
       defaultValue {
        _structType=/Script/GameplayTags.GameplayTagQuery
        autoDescription=""
        tokenStreamVersion=0
        userDescription=""
        queryTokenStream []
        tagDictionary []
       }
      }
      blackboardKey {bNoneIsAllowedValue=false, selectedKeyName=AIEvaluator_Global_TargetActor}
      tokenQuery {
       bCanEditDefaultValueType=false
       key=Item_TokenTagQuery
       defaultValue {
        _structType=&GameplayTagQuery
        autoDescription=""
        tokenStreamVersion=0
        userDescription=""
        queryTokenStream []
        tagDictionary []
       }
      }
     }
    }]
   properties {children [
     {childTask=&self:FortBTTask_RandomInt_0}
     {childComposite=&self:BTComposite_Selector_0, childTask=null}
     {childComposite=&self:BTComposite_Selector_3, childTask=null}
     {childComposite=&self:BTComposite_Selector_1, childTask=null}
     {
      childComposite=&self:BTComposite_Sequence_5
      childTask=null
      decorators [&self:BTDecorator_Loop_3]
     }
    ]}
  }]
 decorators []
 properties {
  children [{
    childComposite=&self:BTComposite_Sequence_0
    childTask=null
    decorators [&self:FortBTDecorator_HasToken_0]
   }]
  services [
   &self:FortAthenaBTService_AIEvaluator_55
   &self:FortAthenaBTService_TokenQuery_0
  ]
 }
 services [
  Focus on target {
   class=FortAthenaBTService_AIEvaluator
   properties {interval=0.5, nodeName=Focus on target, aIEvaluatorClass {defaultValue=&FortAthenaAIBotEvaluator_HandleFocusing}}
  }
  Executing...Token Query {
   class=FortAthenaBTService_TokenQuery
   properties {
    bCallTickOnSearchStart=false
    bRestartTimerOnEachActivation=false
    interval=0.5
    nodeName=Executing...Token Query
    randomDeviation=0.1
    targetActorName=AIEvaluator_Global_TargetActor
    weaponKeyName=None
    activationTagQuery {
     bCanEditDefaultValueType=false
     key=None
     defaultValue {
      _structType=&GameplayTagQuery
      autoDescription=""
      tokenStreamVersion=0
      userDescription=""
      queryTokenStream []
      tagDictionary []
     }
    }
    becomeRelevantTokenAction {
     defaultValue=1
     enumType=/Script/FortniteAIServer.EFortAthenaBTServiceTokenQueryAction
     key=None
     nativeEnumTypeName=EFortAthenaBTServiceTokenQueryAction
    }
    ceaseRelevantTokenAction {
     defaultValue=2
     enumType=&EFortAthenaBTServiceTokenQueryAction
     key=None
     nativeEnumTypeName=EFortAthenaBTServiceTokenQueryAction
    }
    defaultTokenSystemTagQuery {
     bCanEditDefaultValueType=false
     key=Item_TokenTagQuery
     defaultValue {
      _structType=&GameplayTagQuery
      autoDescription=""
      tokenStreamVersion=0
      userDescription=""
      queryTokenStream []
      tagDictionary []
     }
    }
    equippedItemTagAssociationDatas []
    targetChangeTokenAction {
     defaultValue=0
     enumType=&EFortAthenaBTServiceTokenQueryAction
     key=None
     nativeEnumTypeName=EFortAthenaBTServiceTokenQueryAction
    }
    tokenReleaseDelay {defaultValue=0, key=None}
    weaponChangeTokenAction {
     defaultValue=0
     enumType=&EFortAthenaBTServiceTokenQueryAction
     key=None
     nativeEnumTypeName=EFortAthenaBTServiceTokenQueryAction
    }
   }
  }
 ]
}
