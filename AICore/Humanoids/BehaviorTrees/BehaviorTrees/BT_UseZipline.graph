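# Behavior tree that lets a humanoid AI bot travel to a zipline, interact with it and ride it.
# A root Selector runs the Ziplining task when usage status is at least ExecutionAllowed, otherwise a Sequence moves the bot to the entry location and interacts.
# The interact task retries a limited number of times and aborts itself if its execution-allowed condition fails.
# A root Zipline service ticks periodically, wired to the status, target, location and interaction blackboard keys.
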BehaviorTree /AICore/Humanoids/BehaviorTrees/BehaviorTrees/BT_UseZipline
blackboard=/AICore/Humanoids/BehaviorTrees/Blackboards/BB_UseZipline
root {
 name=Selector
 class=BTComposite_Selector
 children [
  Ziplining {
   class=FortAthenaBTTask_Zipline
   children []
   decorators [Is on Zipline {
     class=BTDecorator_Blackboard
     properties {
      arithmeticOperation=GreaterOrEqual
      intValue=4
      nodeName=Is on Zipline
      stringValue=ExecutionAllowed
      blackboardKey {selectedKeyName=Zipline_UsageExecutionStatus}
     }
    }]
   properties {
    bIgnoreRestartSelf=false
    nodeName=Ziplining
    usageExecutionStatusName=Zipline_UsageExecutionStatus
    ziplineTargetName=AIEvaluator_Injection_ZiplineTarget
   }
   services []
  }
  Sequence {
   class=BTComposite_Sequence
   children [
    Bot Move To Zipline entry {
     class=FortAthenaBTTask_BotMoveTo
     decorators []
     properties {
      movementResultKeyName=Zipline_MovementState
      nodeName=Bot Move To Zipline entry
      acceptableRadius {defaultValue=50}
      allowRandomWobble {defaultValue=true}
      bAllowStrafe {defaultValue=false}
      blackboardKey {selectedKeyName=AIEvaluator_Injection_ZiplineEntryLocation}
      observedBlackboardValueTolerance {defaultValue=4.75}
     }
    }
    Interact with Zipline {
     class=FortAthenaBTTask_Interact
     children []
     decorators [Blackboard Based Condition {
       class=BTDecorator_Blackboard
       properties {
        arithmeticOperation=GreaterOrEqual
        flowAbortMode=Self
        intValue=4
        nodeName=Blackboard Based Condition
        stringValue=ExecutionAllowed
        blackboardKey {selectedKeyName=Zipline_InteractionExecutionStatus}
       }
      }]
     properties {
      attemptInterval=0.25
      bIgnoreRestartSelf=false
      bShouldFocusOnInteraction=true
      focalPointName=AIEvaluator_Global_FocalPoint
      interactActorName=AIEvaluator_Global_InteractActor
      jumpExecutionStatusName=AIEvaluator_Jump_ExecutionStatus
      maxInteractAttempts=5
      nodeName=Interact with Zipline
      weaponTriggerMeleeName=AIEvaluator_WeaponTriggerMelee_ExecutionStatus
      interactContextInfoKeySelector {bNoneIsAllowedValue=false, selectedKeyName=Zipline_InteractionContextInfo}
      interactExecutionStatusKeySelector {bNoneIsAllowedValue=false, selectedKeyName=Zipline_InteractionExecutionStatus}
     }
     services []
    }
   ]
   decorators []
   properties {children [
     {childTask=&self:FortAthenaBTTask_BotMoveTo_0}
     {
      childTask=&self:FortAthenaBTTask_Interact_0
      decorators [&self:BTDecorator_Blackboard_1]
     }
    ]}
  }
 ]
 decorators []
 properties {
  children [
   {
    childComposite=null
    childTask=&self:FortAthenaBTTask_Zipline_0
    decorators [&self:BTDecorator_Blackboard_4]
   }
   {childComposite=&self:BTComposite_Sequence_0, childTask=null, decorators []}
  ]
  services [&self:FortAthenaBTService_Zipline_0]
 }
 services [Executing... Zipline {
   class=FortAthenaBTService_Zipline
   properties {
    bCallTickOnSearchStart=false
    bRestartTimerOnEachActivation=false
    focalPointName=AIEvaluator_Global_FocalPoint
    interactionContextInfoName=Zipline_InteractionContextInfo
    interactionExecutionStatusName=Zipline_InteractionExecutionStatus
    interval=0.5
    moveExecutionStatusName=Zipline_MoveExecutionStatus
    moveToPathMovementStateName=Zipline_MovementState
    nodeName=Executing... Zipline
    randomDeviation=0.1
    usageExecutionStatusName=Zipline_UsageExecutionStatus
    ziplineEntryLocationName=AIEvaluator_Injection_ZiplineEntryLocation
    ziplineExitLocationKeyName=AIEvaluator_Injection_ZiplineExitLocation
    ziplineTargetName=AIEvaluator_Injection_ZiplineTarget
   }
  }]
}
