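# Behavior tree that drives an AI humanoid's vehicle in combat, starting with aggressive driving at a preset speed.
# A looping selector prioritizes following a group leader, waiting then leaving the vehicle if that cannot run.
# Otherwise it moves to combat positions at circling speed or patrol points at escape speed, leaving the vehicle as last resort.
# Evaluator and patrolling services feed the blackboard keys that gate and target these moves, and handle focusing.
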BehaviorTree /AICore/Humanoids/BehaviorTrees/BehaviorTrees/BT_Vehicle_Driver_Combat
blackboard=/AICore/Humanoids/BehaviorTrees/Blackboards/BB_Vehicle_Driver_Combat
root {
 name=Sequence
 class=BTComposite_Sequence
 children [
  FortAthenaBTTask_SetAggressiveDriving_1 {
   class=FortAthenaBTTask_SetAggressiveDriving
   children []
   decorators []
   properties {bAggressiveDriving=true, bIgnoreRestartSelf=false, nodeName=""}
   services []
  }
  "Speed = 40" {
   class=FortAthenaBTTask_ModulateVehicleSpeed
   children []
   decorators []
   properties {bIgnoreRestartSelf=false, bResetControlPIds=true, nodeName="Speed = 40", drivingSpeed {defaultValue=40, key=None}}
   services []
  }
  Selector {
   class=BTComposite_Selector
   children [
    Selector {
     class=BTComposite_Selector
     children [
      Selector {
       class=BTComposite_Selector
       children [Move To Group Leader {
         class=FortAthenaBTTask_BotMoveTo
         decorators []
         properties {
          movementResultKeyName=AIEvaluator_FollowGroupLeader_MovementState
          nodeName=Move To Group Leader
          acceptableRadius {defaultValue=10}
          allowRandomWobble {defaultValue=true}
          bAllowStrafe {defaultValue=false}
          bReachTestIncludesAgentRadius {defaultValue=false}
          bReachTestIncludesGoalRadius {defaultValue=false}
          blackboardKey {selectedKeyName=AIEvaluator_FollowGroupLeader_Destination}
          observedBlackboardValueTolerance {defaultValue=4.75}
         }
        }]
       decorators [Following leader {
         class=BTDecorator_Blackboard
         properties {
          arithmeticOperation=GreaterOrEqual
          intValue=3
          nodeName=Following leader
          stringValue=EExecutionStatus::ExecutionPending
          blackboardKey {selectedKeyName=AIEvaluator_FollowGroupLeader_ExecutionStatus}
         }
        }]
       properties {
        children [{
          childComposite=null
          childTask=&self:FortAthenaBTTask_BotMoveTo_2
          decorators []
         }]
        services []
       }
       services []
      }
      Sequence {
       class=BTComposite_Sequence
       children [
        Wait {class=BTTask_Wait, properties {waitTime {defaultValue=5}}}
        LeaveVehicle {
         class=FortAthenaBTTask_LeaveVehicle
         children []
         decorators []
         properties {bIgnoreRestartSelf=false, bWaitVehicleStop=false, nodeName=LeaveVehicle}
         services []
        }
       ]
       decorators []
       properties {children [
         {childTask=&self:BTTask_Wait_5}
         {childTask=&self:FortAthenaBTTask_LeaveVehicle_2}
        ]}
      }
     ]
     decorators [Has a leader to follow {
       class=BTDecorator_Blackboard
       properties {
        arithmeticOperation=Equal
        intValue=0
        nodeName=Has a leader to follow
        stringValue=""
        blackboardKey {selectedKeyName=AIEvaluator_FollowGroupLeader_HasLeaderToFollow}
       }
      }]
     properties {
      children [
       {
        childComposite=&self:BTComposite_Selector_1
        childTask=null
        decorators [&self:BTDecorator_Blackboard_2]
       }
       {childComposite=&self:BTComposite_Sequence_8, childTask=null, decorators []}
      ]
      services [&self:FortAthenaBTService_AIEvaluator_4]
     }
     services [Focus on target {
       class=FortAthenaBTService_AIEvaluator
       properties {interval=0.5, nodeName=Focus on target, aIEvaluatorClass {defaultValue=&FortAthenaAIBotEvaluator_HandleFocusing}}
      }]
    }
    Selector {
     class=BTComposite_Selector
     children [
      Sequence {
       class=BTComposite_Sequence
       children [
        FortAthenaBTTask_ModulateVehicleSpeed_2 {
         class=FortAthenaBTTask_ModulateVehicleSpeed
         children []
         decorators []
         properties {bIgnoreRestartSelf=false, bResetControlPIds=true, nodeName="", drivingSpeed {defaultValue=40, key=VehicleDriverCombat_CirclingSpeed}}
         services []
        }
        Nav Position Move To {
         class=FortAthenaBTTask_BotMoveTo
         decorators []
         properties {
          movementResultKeyName=AIEvaluator_VehicleCombatNavigation_MovementState
          nodeName=Nav Position Move To
          acceptableRadius {defaultValue=200}
          allowRandomWobble {defaultValue=false}
          bAllowStrafe {defaultValue=false}
          bStartFromPreviousPath {defaultValue=true}
          blackboardKey {selectedKeyName=AIEvaluator_VehicleCombatNavigation_Destination}
          observedBlackboardValueTolerance {defaultValue=200}
         }
        }
       ]
       decorators [Can move to Combat Position {
         class=BTDecorator_Blackboard
         properties {
          arithmeticOperation=GreaterOrEqual
          intValue=3
          nodeName=Can move to Combat Position
          notifyObserver=ValueChange
          stringValue=ExecutionPending
          blackboardKey {selectedKeyName=AIEvaluator_VehicleCombatNavigation_ExecutionStatus}
         }
        }]
       properties {children [
         {childTask=&self:FortAthenaBTTask_ModulateVehicleSpeed_2}
         {childTask=&self:FortAthenaBTTask_BotMoveTo_3}
        ]}
      }
      Sequence {
       class=BTComposite_Sequence
       children [
        FortAthenaBTTask_ModulateVehicleSpeed_0 {
         class=FortAthenaBTTask_ModulateVehicleSpeed
         children []
         decorators []
         properties {bIgnoreRestartSelf=false, bResetControlPIds=true, nodeName="", drivingSpeed {defaultValue=40, key=VehicleDriverCombat_EscapeSpeed}}
         services []
        }
        MoveTo PatrolPoint {
         class=FortAthenaBTTask_BotMoveTo
         decorators []
         properties {
          movementResultKeyName=AIEvaluator_Patrolling_MovementState
          nodeName=MoveTo PatrolPoint
          acceptableRadius {defaultValue=10}
          allowRandomWobble {defaultValue=false}
          bAllowStrafe {defaultValue=true}
          bStartFromPreviousPath {defaultValue=true}
          bTrackMovingGoal {defaultValue=false}
          blackboardKey {selectedKeyName=AIEvaluator_Patrolling_Destination}
          observedBlackboardValueTolerance {defaultValue=50}
         }
        }
       ]
       decorators [Should Move? {
         class=BTDecorator_Blackboard
         properties {
          arithmeticOperation=GreaterOrEqual
          intValue=3
          nodeName=Should Move?
          stringValue=""
          blackboardKey {selectedKeyName=AIEvaluator_Patrolling_ShouldMove}
         }
        }]
       properties {children [
         {childTask=&self:FortAthenaBTTask_ModulateVehicleSpeed_0}
         {childTask=&self:FortAthenaBTTask_BotMoveTo_1}
        ]}
      }
     ]
     decorators []
     properties {
      children [
       {
        childComposite=&self:BTComposite_Sequence_3
        childTask=null
        decorators [&self:BTDecorator_Blackboard_0]
       }
       {
        childComposite=&self:BTComposite_Sequence_1
        childTask=null
        decorators [&self:BTDecorator_Blackboard_3]
       }
      ]
      services [
       &self:FortAthenaBTService_AIEvaluator_10
       &self:FortAthenaBTService_Patrolling_1
       &self:FortAthenaBTService_AIEvaluator_1
      ]
     }
     services [
      Evaluating...Patrolling {
       class=FortAthenaBTService_AIEvaluator
       properties {
        interval=0.25
        nodeName=Evaluating...Patrolling
        randomDeviation=0
        aIEvaluatorClass {defaultValue=/Script/FortniteAIServer.FortAthenaNpcEvaluator_Patrolling}
       }
      }
      Executing...Patrolling {
       class=FortAthenaBTService_Patrolling
       properties {
        acceptableRadius=5
        bAllowPartialPath=true
        bCallTickOnSearchStart=false
        bProjectGoalLocation=true
        bRestartTimerOnEachActivation=false
        filterClass=null
        interval=0.5
        nodeName=Executing...Patrolling
        patrollingAppendDestinationKeyName=AIEvaluator_Patrolling_AppendDestination
        randomDeviation=0.1
       }
      }
      Evaluating...Vehicle Combat Navigation {
       class=FortAthenaBTService_AIEvaluator
       properties {
        bCallTickOnSearchStart=true
        interval=0.2
        nodeName=Evaluating...Vehicle Combat Navigation
        randomDeviation=0.05
        aIEvaluatorClass {defaultValue=/AICore/Humanoids/BehaviorTrees/Evaluators/AIEvaluator_VehicleCombatNavigation.AIEvaluator_VehicleCombatNavigation_C}
       }
      }
     ]
    }
    LeaveVehicle {
     class=FortAthenaBTTask_LeaveVehicle
     children []
     decorators []
     properties {bIgnoreRestartSelf=false, bWaitVehicleStop=false, nodeName=LeaveVehicle}
     services []
    }
   ]
   decorators [Loop {class=BTDecorator_Loop}]
   properties {
    children [
     {
      childComposite=&self:BTComposite_Selector_6
      childTask=null
      decorators [&self:BTDecorator_Blackboard_4]
     }
     {childComposite=&self:BTComposite_Selector_0, childTask=null, decorators []}
     {
      childComposite=null
      childTask=&self:FortAthenaBTTask_LeaveVehicle_11
      decorators []
     }
    ]
    services [
     &self:FortAthenaBTService_AIEvaluator_0
     &self:FortAthenaBTService_AIEvaluator_2
    ]
   }
   services [
    Focus on target {
     class=FortAthenaBTService_AIEvaluator
     properties {interval=0.5, nodeName=Focus on target, aIEvaluatorClass {defaultValue=&FortAthenaAIBotEvaluator_HandleFocusing}}
    }
    Evaluating... FollowGroupLeader {
     class=FortAthenaBTService_AIEvaluator
     properties {
      interval=0.5
      nodeName=Evaluating... FollowGroupLeader
      aIEvaluatorClass {defaultValue=/AICore/Humanoids/BehaviorTrees/Evaluators/AIEvaluator_VehicleFollowGroupLeader.AIEvaluator_VehicleFollowGroupLeader_C}
     }
    }
   ]
  }
 ]
 decorators []
 properties {children [
   {childTask=&self:FortAthenaBTTask_SetAggressiveDriving_1}
   {childTask=&self:FortAthenaBTTask_ModulateVehicleSpeed_6}
   {
    childComposite=&self:BTComposite_Selector_2
    childTask=null
    decorators [&self:BTDecorator_Loop_4]
   }
  ]}
}
